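# End-of-match HUD widget on native EliminatedOverlay showing who eliminated the player, spectated name, placement, winners and a relocating indicator.
# Native events trigger animations: elimination intro shifts to spectating after a delay, and placement fades out on a timer.
# Multi-winner names cycle by replaying the swap animation whenever it finishes.
# Mobile touch layouts rescale and reposition the eliminated block by aspect ratio, then move it bottom-right for spectating.
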WidgetBlueprint /FortniteCoreUI/HUD/EndGame/EliminatedOverlay
animations [
 EliminatedByIntro {
  binding_count=4
  display_label=""
  length_seconds=0.7475
  object=&self:EliminatedByIntro
  sequence {
   bindings [
    PlayerRankEmblem {
     id=#1
     display_name=OverlayEmblem
     possessed_class=PlayerRankEmblem_C
     tracks [
      {
       type=MovieScene2DTransformTrack
       sections [{
         type=MovieScene2DTransformSection
         channels [
          Translation.X {type=MovieSceneScriptingFloatChannel}
          Translation.Y {type=MovieSceneScriptingFloatChannel}
          Angle {type=MovieSceneScriptingFloatChannel}
          Scale.X {type=MovieSceneScriptingFloatChannel, keys [[28.784,2],[67.784,0.8347],[72.284,1.0565],[84.284,1]]}
          Scale.Y {type=MovieSceneScriptingFloatChannel, keys [[28.784,2],[67.784,0.8347],[72.284,1.0565],[84.284,1]]}
          Shear.X {type=MovieSceneScriptingFloatChannel}
          Shear.Y {type=MovieSceneScriptingFloatChannel}
         ]
        }]
      }
      {
       type=MovieSceneFloatTrack
       display_name=Render Opacity
       sections [{type=MovieSceneFloatSection, channels [None {type=MovieSceneScriptingFloatChannel, keys [[28.784,0],[49.784,1],[89.7,1]]}]}]
      }
     ]
    }
    Text_SpectatorName {
     id=#2
     possessed_class=TextBlock
     tracks [
      {
       type=MovieScene2DTransformTrack
       sections [{
         type=MovieScene2DTransformSection
         channels [
          Translation.X {type=MovieSceneScriptingFloatChannel}
          Translation.Y {type=MovieSceneScriptingFloatChannel}
          Angle {type=MovieSceneScriptingFloatChannel}
          Scale.X {type=MovieSceneScriptingFloatChannel, keys [[18.66,2],[57.66,0.8347],[62.16,1.0565],[74.16,1]]}
          Scale.Y {type=MovieSceneScriptingFloatChannel, keys [[18.66,2],[57.66,0.8347],[62.16,1.0565],[74.16,1]]}
          Shear.X {type=MovieSceneScriptingFloatChannel}
          Shear.Y {type=MovieSceneScriptingFloatChannel}
         ]
        }]
      }
      {
       type=MovieSceneFloatTrack
       display_name=Render Opacity
       sections [{type=MovieSceneFloatSection, channels [None {type=MovieSceneScriptingFloatChannel, keys [[18.66,0],[39.66,1],[79.574,1]]}]}]
      }
     ]
    }
    Text_EliminatedBy {
     id=#3
     possessed_class=TextBlock
     tracks [
      {
       type=MovieScene2DTransformTrack
       sections [{
         type=MovieScene2DTransformSection
         channels [
          Translation.X {type=MovieSceneScriptingFloatChannel}
          Translation.Y {type=MovieSceneScriptingFloatChannel}
          Angle {type=MovieSceneScriptingFloatChannel}
          Scale.X {type=MovieSceneScriptingFloatChannel, keys [[0,3],[39.002,0.8347],[43.502,1.0565],[55.502,1]]}
          Scale.Y {type=MovieSceneScriptingFloatChannel, keys [[0,3],[39.002,0.8347],[43.502,1.0565],[55.502,1]]}
          Shear.X {type=MovieSceneScriptingFloatChannel}
          Shear.Y {type=MovieSceneScriptingFloatChannel}
         ]
        }]
      }
      {
       type=MovieSceneFloatTrack
       display_name=Render Opacity
       sections [{type=MovieSceneFloatSection, channels [None {type=MovieSceneScriptingFloatChannel, keys [[0.002,0],[21.002,1],[60.918,1]]}]}]
      }
     ]
    }
    CommonWidgetSwitcher_phasetype {
     id=#4
     possessed_class=CommonWidgetSwitcherLegacy
     tracks [{
       type=MovieSceneIntegerTrack
       display_name=Active Widget Index
       sections [{type=MovieSceneIntegerSection, channels [None {type=MovieSceneScriptingIntegerChannel, const [0,1]}]}]
      }]
    }
   ]
   display_rate {numerator=120}
  }
 }
 EliminatedToSpectating {
  binding_count=3
  display_label=""
  length_seconds=0.7765
  object=&self:EliminatedToSpectating
  sequence {
   bindings [
    CommonWidgetSwitcher_phasetype {
     id=#5
     possessed_class=CommonWidgetSwitcherLegacy
     tracks [{
       type=MovieSceneIntegerTrack
       display_name=Active Widget Index
       sections [{type=MovieSceneIntegerSection, channels [None {type=MovieSceneScriptingIntegerChannel, const [1,1]}]}]
      }]
    }
    Overlay_PlayerName {
     id=#6
     display_name=OverlayPlayername
     possessed_class=HorizontalBox
     tracks [{
       type=MovieScene2DTransformTrack
       sections [{
         type=MovieScene2DTransformSection
         channels [
          Translation.X {type=MovieSceneScriptingFloatChannel}
          Translation.Y {type=MovieSceneScriptingFloatChannel}
          Angle {type=MovieSceneScriptingFloatChannel}
          Scale.X {type=MovieSceneScriptingFloatChannel, keys [[26.804,1],[69.18,0.7]]}
          Scale.Y {type=MovieSceneScriptingFloatChannel, keys [[26.804,1],[69.18,0.7]]}
          Shear.X {type=MovieSceneScriptingFloatChannel}
          Shear.Y {type=MovieSceneScriptingFloatChannel}
         ]
        }]
      }]
    }
    Widget_KillCounter {
     id=#7
     display_name=Kills
     possessed_class=TextBlock
     tracks [{
       type=MovieSceneFloatTrack
       display_name=Render Opacity
       sections [{type=MovieSceneFloatSection, channels [None {type=MovieSceneScriptingFloatChannel, keys [[26.804,0],[81.182,1],[93.182,1]]}]}]
      }]
    }
   ]
   display_rate {numerator=120}
  }
 }
 PlacementIntro {
  binding_count=1
  display_label=""
  length_seconds=0.3393
  object=&self:PlacementIntro
  sequence {
   bindings [VBox_PlacementText {
     id=#8
     possessed_class=VerticalBox
     tracks [
      {
       type=MovieScene2DTransformTrack
       sections [{
         type=MovieScene2DTransformSection
         end_frame=40
         start_frame=0
         channels [
          Translation.X {type=MovieSceneScriptingFloatChannel}
          Translation.Y {type=MovieSceneScriptingFloatChannel}
          Angle {type=MovieSceneScriptingFloatChannel}
          Scale.X {type=MovieSceneScriptingFloatChannel, keys [[0,2.3491],[31,0.8798],[35.244,1]]}
          Scale.Y {type=MovieSceneScriptingFloatChannel, keys [[0,2.3491],[31,0.8798],[35.244,1]]}
          Shear.X {type=MovieSceneScriptingFloatChannel}
          Shear.Y {type=MovieSceneScriptingFloatChannel}
         ]
        }]
      }
      {
       type=MovieSceneFloatTrack
       display_name=Render Opacity
       sections [{type=MovieSceneFloatSection, end_frame=40, start_frame=0, channels [None {type=MovieSceneScriptingFloatChannel, keys [[0,0],[28.34,1]]}]}]
      }
     ]
    }]
   display_rate {numerator=120}
  }
 }
 PlacementOutro {
  binding_count=1
  display_label=""
  length_seconds=1
  object=&self:PlacementOutro
  sequence {bindings [VBox_PlacementText {
     id=#9
     display_name=YouPlacedBox
     possessed_class=HorizontalBox
     tracks [{
       type=MovieSceneFloatTrack
       display_name=Render Opacity
       sections [{type=MovieSceneFloatSection, end_frame=20, start_frame=0, channels [None {type=MovieSceneScriptingFloatChannel, keys [[0,1],[10,0]]}]}]
      }]
    }]}
 }
 WinnerDetermined {
  binding_count=2
  display_label=""
  length_seconds=0.3083
  object=&self:WinnerDetermined
  sequence {
   bindings [
    VBox_Winner {
     id=#10
     possessed_class=VerticalBox
     tracks [
      {
       type=MovieSceneFloatTrack
       display_name=Render Opacity
       sections [{type=MovieSceneFloatSection, channels [None {type=MovieSceneScriptingFloatChannel, keys [[0,0],[21.54,1]]}]}]
      }
      {
       type=MovieScene2DTransformTrack
       sections [{
         type=MovieScene2DTransformSection
         channels [
          Translation.X {type=MovieSceneScriptingFloatChannel}
          Translation.Y {type=MovieSceneScriptingFloatChannel}
          Angle {type=MovieSceneScriptingFloatChannel}
          Scale.X {type=MovieSceneScriptingFloatChannel, keys [[0,0],[27,1.3246],[33,0.9513],[37,1]]}
          Scale.Y {type=MovieSceneScriptingFloatChannel, keys [[0,0],[27,1.3246],[33,0.9513],[37,1]]}
          Shear.X {type=MovieSceneScriptingFloatChannel}
          Shear.Y {type=MovieSceneScriptingFloatChannel}
         ]
        }]
      }
     ]
    }
    VerticalBoxEliminatedbySpectating {
     id=#11
     possessed_class=VerticalBox
     tracks [{
       type=MovieSceneFloatTrack
       display_name=Render Opacity
       sections [{type=MovieSceneFloatSection, channels [None {type=MovieSceneScriptingFloatChannel, const [0,1]}]}]
      }]
    }
   ]
   display_rate {numerator=120}
  }
 }
 MultiWinnerSwap {
  binding_count=1
  display_label=""
  length_seconds=3.0009
  object=&self:MultiWinnerSwap
  sequence {
   bindings [VerticalBox_MultiWinnerContainer {
     id=#12
     possessed_class=VerticalBox
     tracks [{
       type=MovieScene2DTransformTrack
       sections [{
         type=MovieScene2DTransformSection
         channels [
          Translation.X {type=MovieSceneScriptingFloatChannel}
          Translation.Y {type=MovieSceneScriptingFloatChannel, keys [[0,0],[7.019,71.5928]]}
          Angle {type=MovieSceneScriptingFloatChannel}
          Scale.X {type=MovieSceneScriptingFloatChannel}
          Scale.Y {type=MovieSceneScriptingFloatChannel}
          Shear.X {type=MovieSceneScriptingFloatChannel}
          Shear.Y {type=MovieSceneScriptingFloatChannel}
         ]
        }]
      }]
    }]
   tracks [{type=MovieSceneEventTrack, sections [{type=MovieSceneEventSection}]}]
  }
 }
]
graphs [
 AdjustVertical BoxEliminatedPadding {nodes [
   AdjustVertical BoxEliminatedPadding() {is=title K2Node_FunctionEntry}
   > n1 = Set VerticalBoxSlot(VerticalBoxSlot:Overlay Slot Object Reference=SlotAsOverlaySlot(Widget=Get VerticalBoxEliminatedbySpectating() -> (VerticalBoxEliminatedbySpectating:Vertical Box Object Reference))) -> (Output_Get:Overlay Slot Object Reference)
   > Is Valid(InputObject=@n1)
   > n2 = Sequence()
   > n3 = Branch(Condition=IsMobileGame())
   > Set bAdjustPadding(bAdjustPadding:bool=true) -> (Output_Get:bool)
   n4 = FindParentWidgetOfType!(
    execute=@n3.else
    StartingWidget:Widget Object Reference=Self-Reference()
    Type:Widget Class Reference=/FortniteCoreUI/HUD/AthenaHUD.AthenaHUD_C
   ) -> (then, ReturnValue:Athena HUD Object Reference)
   > Cast To AthenaHUD!(execute, Object=@n4) -> (then, CastFailed, AsAthena HUD:Athena HUD Object Reference) {is=title K2Node_DynamicCast}
   > Branch(Condition=false)
   > Set bAdjustPadding(bAdjustPadding=true)
   Branch(execute=@n2.then_1, Condition=Get bAdjustPadding() -> (bAdjustPadding:bool))
   > SetPadding!(
    execute
    self:Overlay Slot Object Reference=Get VerticalBoxSlot() -> (VerticalBoxSlot:Overlay Slot Object Reference)
    InPadding:Margin Structure=Make Margin(Left=0, Top=150, Right=0, Bottom=0)
   ) -> (then)
  ]}
 DeathCauseToText {nodes [
   n1 = DeathCauseToText!() -> (then, DeathCause:EDeathCause Enum=OutsideSafeZone) {is=title K2Node_FunctionEntry}
   > Return Node!(
    execute
    ReturnValue:Text=Select!(
     OutsideSafeZone:Text="LOCTEXT(\"ELIMINATED BY THE STORM\")"
     FallDamage:Text
     Pistol:Text
     Shotgun:Text
     Rifle:Text
     SMG:Text
     Sniper:Text
     SniperNoScope:Text
     Melee:Text
     InfinityBlade:Text
     Grenade:Text
     C4:Text
     GrenadeLauncher:Text
     RocketLauncher:Text
     Minigun:Text
     Bow:Text
     Trap:Text
     DBNOTimeout:Text
     Banhammer:Text
     RemovedFromGame:Text
     MassiveMelee:Text
     MassiveDiveBomb:Text
     MassiveRanged:Text
     Vehicle:Text
     ShoppingCart:Text
     ATK:Text
     QuadCrasher:Text
     Biplane:Text
     BiplaneGun:Text
     LMG:Text
     GasGrenade:Text
     InstantEnvironmental:Text
     InstantEnvironmentalFellOutOfWorld:Text
     InstantEnvironmentalUnderLandscape:Text
     Turret:Text
     ShipCannon:Text
     Cube:Text
     Balloon:Text
     StormSurge:Text
     Lava:Text
     BasicFiend:Text="LOCTEXT(\"ELIMINATED BY A CUBE FIEND\")"
     EliteFiend:Text="LOCTEXT(\"ELIMINATED BY AN ELITE CUBE FIEND\")"
     RangedFiend:Text
     BasicBrute:Text="LOCTEXT(\"ELIMINATED BY A CUBE BRUTE\")"
     EliteBrute:Text="LOCTEXT(\"ELIMINATED BY AN ELITE CUBE BRUTE\")"
     MegaBrute:Text="LOCTEXT(\"ELIMINATED BY A MEGA CUBE BRUTE\")"
     SilentSwitchingToSpectate:Text="LOCTEXT(\"SPECTATING\")"
     LoggedOut:Text
     TeamSwitchSuicide:Text
     WonMatch:Text
     Unspecified:Text
     Index:EDeathCause Enum=@n1
    ) -> (ReturnValue:Text)
   )
  ]}
 EventGraph {nodes [
   n1 = On UI Extensions Changed!() -> (OutputDelegate, then, Minigame:Fort Minigame Object Reference) {is=title K2Node_CustomEvent}
   > Delay(Duration=1)
   > AdjustVertical BoxEliminatedPadding() {is=title K2Node_CallFunction}
   n2 = PlacementIntroComplete() {is=title K2Node_CustomEvent}
   n3 = PlayPlacementOutro() {is=title K2Node_CustomEvent}
   Set Timer by Event(execute=@n2.then, Delegate=@n3, Time=6, bLooping=false)
   n4 = WinnerSwapComplete() {is=title K2Node_CustomEvent}
   > UpdateMultiwinner(self:Eliminated Overlay Object Reference)
   > PlayAnimation(InAnimation=Get MultiWinnerSwap() -> (MultiWinnerSwap:Widget Animation Object Reference))
   Event OnUpdateMultiwinner() {is=title K2Node_Event}
   > PlayAnimation(InAnimation=Get MultiWinnerSwap())
   n5 = Event SetBannerInfo!() -> (OutputDelegate, then, BannerInfo:"Player Banner Info Structure (by ref)") {is=title K2Node_Event}
   > SetBannerInfo!(
    execute
    self:Player Rank Emblem Object Reference=Get PlayerRankEmblem() -> (PlayerRankEmblem:Player Rank Emblem Object Reference)
    Banner:Player Banner Info Structure=@n5.BannerInfo
   ) -> (then)
   Event OnShowPlacementBox() {is=title K2Node_Event}
   > PlayAnimation(InAnimation=Get PlacementIntro() -> (PlacementIntro:Widget Animation Object Reference))
   n6 = Event SetEliminatorPlayernameAlignments!() -> (
    OutputDelegate
    then
    VerticalAligment:"EVerticalAlignment Enum (by ref)"=VAlign_Fill
    HorizontalAligment:"EHorizontalAlignment Enum (by ref)"=HAlign_Fill
   ) {is=title K2Node_Event}
   > n7 = Cast To OverlaySlot!(execute, Object=Get Slot!(self:Widget Object Reference=Get VerticalBoxEliminatedbySpectating()) -> (Slot:Panel Slot Object Reference)) -> (then, CastFailed, AsOverlay Slot:Overlay Slot Object Reference) {is=title K2Node_DynamicCast}
   > SetVerticalAlignment(self=@n7, InVerticalAlignment=@n6.VerticalAligment)
   > SetHorizontalAlignment(self=@n7, InHorizontalAlignment=@n6.HorizontalAligment)
   Event OnWinnerDetermined() {is=title K2Node_Event}
   > PlayAnimation(InAnimation=Get WinnerDetermined() -> (WinnerDetermined:Widget Animation Object Reference))
   Event Construct()
   > n8 = Cast To FortPlayerControllerAthena(Object=GetOwningPlayer())
   > Bind Event to Spectator Streaming Changed(self:Fort Player Controller Athena Object Reference=@n8, Delegate=Create Event()) {is=title K2Node_AddDelegate}
   > BindToAnimationFinished(Animation=Get MultiWinnerSwap(), Delegate=@n4)
   > Set WinnerNames(WinnerNames:Array of Texts=Make Array!(
     "[0]":Text="LOCTEXT(\"Winner1\")"
     "[1]":Text="LOCTEXT(\"Winner2\")"
     "[2]":Text="LOCTEXT(\"Winner 3\")"
     "[3]":Text="LOCTEXT(\"Winner 4\")"
     "[4]":Text="LOCTEXT(\"Winner 5\")"
    ) -> (Array:Array of Texts)) -> (Output_Get:Array of Texts)
   > BindToAnimationFinished(Animation=Get PlacementIntro(), Delegate=@n2)
   > n9 = Sequence()
   > n10 = GetMinigameForPlayerState!(
    execute
    PlayerState:Player State Object Reference=Get PlayerState!(self:Controller Object Reference=GetOwningPlayer()) -> (PlayerState:Player State Object Reference)
   ) -> (then, ReturnValue:Fort Minigame Object Reference)
   > Is Valid(InputObject=@n10)
   > Assign On UIExtensions Changed!(execute:Exec, self:Fort Minigame Object Reference=@n10, Delegate:"Delegate (by ref)"=@n1) -> (then:Exec) {is=title K2Node_AssignDelegate}
   AdjustVertical BoxEliminatedPadding(execute=@n9.then_1)
   Event OnResetSpectatingForCreative() {is=title K2Node_Event}
   n11 = Get EliminatedToSpectating() -> (EliminatedToSpectating:Widget Animation Object Reference)
   > PlayAnimation(InAnimation=@n11, StartAtTime=GetEndTime(self=@n11), PlayMode=Reverse)
   Event OnEliminatedByIntro() {is=title K2Node_Event}
   n12 = GetOwningPlayer()
   > n13 = Branch(Condition=AND Boolean(A=IsMobileGame(), B=IsUsingTouch(PlayerController=@n12)))
   n14 = SlotAsOverlaySlot(Widget=Get VerticalBoxEliminatedbySpectating())
   n15 = GetViewportSize!() -> (ReturnValue_X:double, ReturnValue_Y:double)
   n16 = float / float(A=@n15, B=@n15.ReturnValue_Y)
   > SetPadding!(
    execute
    self=@n14
    InPadding_Left:float=0
    InPadding_Top:float=Select!(Option 0:double=400, Option 1:double=600, Index:bool="float <= float"(A=@n16, B=1.62) -> (ReturnValue=~)) -> (ReturnValue:double)
    InPadding_Right:float=0
    InPadding_Bottom:float=0
   ) -> (then)
   > SetHorizontalAlignment(self=@n14, InHorizontalAlignment=HAlign_Center)
   > SetVerticalAlignment(self=@n14, InVerticalAlignment=VAlign_Top)
   > n17 = "Play Animation with Finished event (legacy)"!(
    execute:Exec
    Widget:User Widget Object Reference=Self-Reference()
    InAnimation:Widget Animation Object Reference=Get EliminatedByIntro() -> (EliminatedByIntro:Widget Animation Object Reference)
    StartAtTime:float=0
    NumLoopsToPlay:int=1
    PlayMode:EUMGSequencePlayMode Enum
    PlaybackSpeed:float=1
   ) -> (then:Exec, Result:UMGSequence Player Object Reference, Finished:Exec) {is=title K2Node_PlayAnimation}
   RetriggerableDelay(execute=@n17.Finished, Duration=3)
   n18 = SlotAsOverlaySlot(Widget=Get VerticalBoxEliminatedbySpectating())
   > SetPadding(self=@n18, InPadding_Left=0, InPadding_Top=0, InPadding_Right=150, InPadding_Bottom=215)
   > SetHorizontalAlignment(self=@n18, InHorizontalAlignment=HAlign_Right)
   > SetVerticalAlignment(self=@n18, InVerticalAlignment=VAlign_Bottom)
   > PlayAnimation(InAnimation=Get EliminatedToSpectating())
   n19 = "Play Animation with Finished event (legacy)"(execute=@n13.else, Widget=Self-Reference(), InAnimation=Get EliminatedByIntro(), StartAtTime=0, NumLoopsToPlay=1, PlaybackSpeed=1)
   RetriggerableDelay(execute=@n19.Finished, Duration=3)
   > PlayAnimation(InAnimation=Get EliminatedToSpectating())
   Event PreConstruct()
   > SetMobileStyle() {is=title K2Node_CallFunction}
   PlayAnimation(execute=@n3.then, InAnimation=Get PlacementOutro() -> (PlacementOutro:Widget Animation Object Reference))
  ]}
 IsEliminatedByIntroAnimFinished {nodes [
   IsEliminatedByIntroAnimFinished() {is=title K2Node_FunctionEntry}
   n1 = Get EliminatedByIntro()
   > Return Node(ReturnValue:bool=NOT Boolean(A=IsAnimationPlaying(InAnimation=@n1)))
  ]}
 SetMobileStyle {nodes [
   SetMobileStyle() {is=title K2Node_FunctionEntry}
   n1 = GetOwningPlayer()
   > Branch(Condition=AND Boolean(A=IsMobileGame(), B=IsUsingTouch(PlayerController=@n1)))
   > SetRenderScale!(
    execute
    self:Widget Object Reference=Get CommonBorder_0() -> (CommonBorder_0:Common Border Object Reference)
    Scale:Vector 2D Structure=(X=0.85,Y=0.85)
   ) -> (then)
  ]}
 SetPlaceTextFont_TEMP {nodes [
   SetPlaceTextFont_TEMP() {is=title K2Node_FunctionEntry}
   > n1 = GetPlayerState!(execute, self:Eliminated Overlay Object Reference) -> (then, ReturnValue:Fort Player State Athena Object Reference)
   > Is Valid(InputObject=@n1)
   n2 = Get Text_PlaceNumber() -> (Text_PlaceNumber:Common Text Object Reference)
   n3 = Break Slate Font Info!(SlateFontInfo:"Slate Font Info Structure (by ref)"=Get Font!(self:Text Object Reference=@n2) -> (Font:Slate Font Info Structure)) -> (
    FontObject:Object Reference
    FontMaterial:Object Reference
    OutlineSettings:Font Outline Settings Structure
    TypefaceFontName:Name
    Size:float=~
    LetterSpacing:int
    SkewAmount:float=~
    bForceMonospaced:bool=~
    bMaterialIsStencil:bool=~
    MonospacedWidth:float=~
   ) {is=title K2Node_BreakStruct}
   n4 = Get Place!(self:Fort Player State Athena Object Reference=@n1) -> (Place:int=0)
   n5 = "integer < integer"!(A:int=@n4, B:int=4) -> (ReturnValue:bool) {is=title K2Node_CallFunction}
   n6 = Select(Option 0:int=34, Option 1:int=42, Index=@n5) -> (ReturnValue:int)
   > SetFont(
    self=@n2
    InFontInfo=MakeSlateFontInfo!(
     FontObject:Object Reference=@n3
     FontMaterial:Object Reference=Select(Option 0:Object Reference, Option 1:Object Reference=/FortUILibrary/Materials/Foundation/M_Font_Horizon_Dynamic, Index=@n5) -> (ReturnValue:Object Reference)
     OutlineSettings:Font Outline Settings Structure=@n3.OutlineSettings
     TypefaceFontName:Name=@n3.TypefaceFontName
     Size:float="To Float (Integer)"(InInt=@n6)
     LetterSpacing:int=0
     SkewAmount:float=0
     bForceMonospaced:bool=false
     bMaterialIsStencil:bool=false
     MonospacedWidth:float=1
    ) -> (ReturnValue:Slate Font Info Structure)
   )
   n7 = Select(Option 0:int=0, Option 1:int=@n4, Index=@n5) -> (ReturnValue:int)
   > SetColorAndOpacity(
    self=@n2
    InColorAndOpacity=Make Slate Color(SpecifiedColor=Select!(
      Option 0:Linear Color Structure=(R=1,G=1,B=1,A=1)
      Option 1:Linear Color Structure=(R=1,G=0.5841,B=0.0666,A=1)
      Option 2:Linear Color Structure=(R=0.6584,G=0.7605,B=1,A=1)
      Option 3:Linear Color Structure=(R=1,G=0.4179,B=0.0931,A=1)
      Index:int=@n7
     ) -> (ReturnValue:Linear Color Structure))
   )
  ]}
]
variables []
widgets {
 info {inheritedWidgetCount=19, parentClass=/Script/FortniteUI.EliminatedOverlay, rootWidgetClass=&Overlay, widgetCount=39}
 widgets [
  Overlay_0 {widgetClassPath=&Overlay, widget=Overlay_0}
  VerticalBoxEliminatedbySpectating {
   widgetClassPath=&VerticalBox
   bIsVariable=true
   parent=Overlay_0
   widget=VerticalBoxEliminatedbySpectating
   slot {class=OverlaySlot, properties {horizontalAlignment=HAlign_Center, padding=<ltrb 0 5 0 0>}}
  }
  CommonWidgetSwitcher_phasetype {
   widgetClassPath=&CommonWidgetSwitcherLegacy
   bIsVariable=true
   parent=VerticalBoxEliminatedbySpectating
   widget=CommonWidgetSwitcher_phasetype
   properties {transitionDuration=0.8, transitionType=Horizontal}
   slot {class=VerticalBoxSlot, properties {horizontalAlignment=HAlign_Center, padding=<ltrb 0 0 0 -20>}}
  }
  Text_EliminatedBy {
   widgetClassPath=&TextBlock
   bInherited=true
   bIsVariable=true
   parent=CommonWidgetSwitcher_phasetype
   widget=Text_EliminatedBy
   properties {
    text=ELIMINATED BY
    font {
     bForceMonospaced=false
     bMaterialIsStencil=false
     fontMaterial=null
     fontObject=&Burbank
     letterSpacing=0
     monospacedWidth=1
     size=24
     skewAmount=0
     typefaceFontName=Burbank-PlayerPower
     outlineSettings {
      bApplyOutlineToDropShadows=false
      bMiteredCorners=false
      bSeparateFillAlpha=false
      outlineMaterial=/Game/UI/Foundation/Materials/M-FontShadow
      outlineSize=3
      outlineColor=<rgba 0.0052 0.0052 0.0056 0.2>
     }
    }
    renderTransformPivot=<0.5 0>
    shadowColorAndOpacity=<rgba 0.0052 0.0052 0.0056 0.5>
    shadowOffset=<4 4>
   }
   slot {class=WidgetSwitcherSlot}
  }
  Text_Spectating {
   widgetClassPath=&TextBlock
   bIsVariable=true
   parent=CommonWidgetSwitcher_phasetype
   widget=Text_Spectating
   properties {
    text=Spectating
    textTransformPolicy=ToUpper
    font {
     bForceMonospaced=false
     bMaterialIsStencil=false
     fontMaterial=null
     fontObject=&Burbank
     letterSpacing=0
     monospacedWidth=1
     size=24
     skewAmount=0
     typefaceFontName=Burbank-PlayerPower
     outlineSettings {
      bApplyOutlineToDropShadows=false
      bMiteredCorners=false
      bSeparateFillAlpha=false
      outlineMaterial=&M-FontShadow
      outlineSize=6
      outlineColor=<rgba 0.0052 0.0052 0.0056 0.5>
     }
    }
    renderTransformPivot=<0.5 0>
    shadowColorAndOpacity=<rgba 0.0052 0.0052 0.0056 0.5>
    shadowOffset=<4 4>
   }
   slot {class=WidgetSwitcherSlot}
  }
  Overlay_PlayerName {
   widgetClassPath=&Overlay
   bInherited=true
   bIsVariable=true
   parent=VerticalBoxEliminatedbySpectating
   widget=Overlay_PlayerName
   properties {renderTransformPivot=<0.5 0.2>}
   slot {class=VerticalBoxSlot, properties {horizontalAlignment=HAlign_Center}}
  }
  Widget_KillCounter {
   widgetClassPath=/Game/Athena/HUD/AthenaKillsWidget
   bInherited=true
   bIsVariable=true
   parent=Overlay_PlayerName
   widget=Widget_KillCounter
   properties {renderOpacity=0, renderTransform {angle=0, scale=<1.25 1.25>, shear=<-12 0>, translation=<90 4>}}
   slot {class=OverlaySlot, properties {horizontalAlignment=HAlign_Right, verticalAlignment=VAlign_Center}}
  }
  PlayerNameBox {
   widgetClassPath=&HorizontalBox
   parent=Overlay_PlayerName
   widget=PlayerNameBox
   slot {class=OverlaySlot, properties {horizontalAlignment=HAlign_Center}}
  }
  ScaleBoxEmblem {
   widgetClassPath=&ScaleBox
   parent=PlayerNameBox
   widget=ScaleBoxEmblem
   properties {stretch=UserSpecified, userSpecifiedScale=0.5, visibility=Collapsed}
   slot {class=HorizontalBoxSlot, properties {verticalAlignment=VAlign_Center}}
  }
  OverlayEmblem {
   widgetClassPath=&Overlay
   parent=ScaleBoxEmblem
   widget=OverlayEmblem
   properties {renderTransformPivot=<1 0.5>}
   slot {class=ScaleBoxSlot}
  }
  PlayerRankEmblem_C_0 {
   widgetClassPath=/Game/Athena/UI/Progression/PlayerRankEmblem
   parent=OverlayEmblem
   widget=PlayerRankEmblem_C_0
   properties {
    bindToLocalPlayerByDefault=false
    colorAndOpacity=<rgba 0 0 0 0.8095>
    renderTransform {angle=0, scale=<1 1>, shear=<0 0>, translation=<3 3>}
   }
   slot {class=OverlaySlot}
  }
  PlayerRankEmblem {
   widgetClassPath=&PlayerRankEmblem
   bIsVariable=true
   parent=OverlayEmblem
   widget=PlayerRankEmblem
   properties {bindToLocalPlayerByDefault=false}
   slot {class=OverlaySlot}
  }
  CommonBorder_0 {
   widgetClassPath=&CommonBorder
   bIsVariable=true
   parent=PlayerNameBox
   widget=CommonBorder_0
   properties {horizontalAlignment=HAlign_Center, padding=<ltrb 0 0 0 0>, renderTransform {angle=0, scale=<1 1>, shear=<-12 0>, translation=<0 0>}}
   slot {class=HorizontalBoxSlot, properties {horizontalAlignment=HAlign_Center, verticalAlignment=VAlign_Top, padding=<ltrb 0 -2 0 -25>}}
  }
  Text_SpectatorName {
   widgetClassPath=&TextBlock
   bInherited=true
   bIsVariable=true
   parent=CommonBorder_0
   widget=Text_SpectatorName
   properties {
    text="{playername}"
    textTransformPolicy=ToUpper
    colorAndOpacity {colorUseRule=UseColor_Specified, specifiedColor=<rgba 1 0.7795 0 1>}
    font {
     bForceMonospaced=false
     bMaterialIsStencil=false
     fontMaterial=null
     fontObject=&Burbank
     letterSpacing=0
     monospacedWidth=1
     size=48
     skewAmount=0
     typefaceFontName=BurbankRegular-Black
     outlineSettings {
      bApplyOutlineToDropShadows=false
      bMiteredCorners=false
      bSeparateFillAlpha=false
      outlineMaterial=&M-FontShadow
      outlineSize=6
      outlineColor=<rgba 0.0052 0.0052 0.0056 0.5>
     }
    }
    renderTransformPivot=<0.5 0>
    shadowColorAndOpacity=<rgba 0.0052 0.0052 0.0056 0.4>
    shadowOffset=<8 8>
   }
   slot {class=BorderSlot, properties {horizontalAlignment=HAlign_Center, padding=<ltrb 0 0 0 0>}}
  }
  Text_UserId {
   widgetClassPath=&TextBlock
   bInherited=true
   bIsVariable=true
   parent=Overlay_PlayerName
   widget=Text_UserId
   properties {
    justification=Center
    text=USER ID
    colorAndOpacity {colorUseRule=UseColor_Specified, specifiedColor=<rgba 1 1 1 0.4>}
    font {
     bForceMonospaced=false
     bMaterialIsStencil=false
     fontMaterial=null
     fontObject=&NotoSans
     letterSpacing=0
     monospacedWidth=1
     size=10
     skewAmount=0
     typefaceFontName=NotoSans-Regular
     outlineSettings {
      bApplyOutlineToDropShadows=false
      bMiteredCorners=false
      bSeparateFillAlpha=false
      outlineMaterial=null
      outlineSize=0
      outlineColor=<rgba 0 0 0 1>
     }
    }
    renderTransform {angle=0, scale=<1 1>, shear=<0 0>, translation=<0 20>}
   }
   slot {class=OverlaySlot, properties {horizontalAlignment=HAlign_Center, verticalAlignment=VAlign_Bottom}}
  }
  Spacer_0 {
   widgetClassPath=&Spacer
   parent=VerticalBoxEliminatedbySpectating
   widget=Spacer_0
   slot {class=VerticalBoxSlot, properties {verticalAlignment=VAlign_Center, size {sizeRule=Fill, value=1}}}
  }
  VerticalBoxWinnerPlacementInfo {
   widgetClassPath=&VerticalBox
   parent=Overlay_0
   widget=VerticalBoxWinnerPlacementInfo
   slot {class=OverlaySlot, properties {horizontalAlignment=HAlign_Center, verticalAlignment=VAlign_Center}}
  }
  VBox_Winner {
   widgetClassPath=&VerticalBox
   bInherited=true
   bIsVariable=true
   parent=VerticalBoxWinnerPlacementInfo
   widget=VBox_Winner
   properties {visibility=Collapsed, renderTransformPivot=<0.5 0.8>}
   slot {class=VerticalBoxSlot, properties {padding=<ltrb 0 0 0 20>}}
  }
  Switcher_MultiWinners {
   widgetClassPath=&CommonWidgetSwitcherLegacy
   bInherited=true
   bIsVariable=true
   parent=VBox_Winner
   widget=Switcher_MultiWinners
   slot {class=VerticalBoxSlot}
  }
  WinnerName {
   widgetClassPath=&HorizontalBox
   parent=Switcher_MultiWinners
   widget=WinnerName
   properties {visibility=Visible}
   slot {class=WidgetSwitcherSlot, properties {horizontalAlignment=HAlign_Center}}
  }
  Text_Winner {
   widgetClassPath=/Script/FortniteUI.FortRichTextBlockLegacy
   bInherited=true
   bIsVariable=true
   parent=WinnerName
   widget=Text_Winner
   properties {
    styleSet=/FortniteCoreUI/Athena/HUD/EndGame/EliminatedOverlaySingleWinnerStyles
    text="<WinnerName>WWWWW</> won the game!"
    renderTransform {angle=0, scale=<1 1>, shear=<-12 0>, translation=<0 0>}
   }
   slot {class=HorizontalBoxSlot}
  }
  MultiWinnernames {
   widgetClassPath=&HorizontalBox
   parent=Switcher_MultiWinners
   widget=MultiWinnernames
   properties {clipping=ClipToBounds, visibility=Visible}
   slot {class=WidgetSwitcherSlot, properties {horizontalAlignment=HAlign_Center}}
  }
  VerticalBox_MultiWinnerContainer {
   widgetClassPath=&VerticalBox
   parent=MultiWinnernames
   widget=VerticalBox_MultiWinnerContainer
   slot {class=HorizontalBoxSlot, properties {verticalAlignment=VAlign_Bottom, padding=<ltrb 0 -70 10 0>}}
  }
  Text_MultiWinner2 {
   widgetClassPath=&CommonTextBlock
   bInherited=true
   bIsVariable=true
   parent=VerticalBox_MultiWinnerContainer
   widget=Text_MultiWinner2
   properties {
    justification=Right
    text=WWWWWinner1
    colorAndOpacity {colorUseRule=UseColor_Specified, specifiedColor=<rgba 1 0.7913 0.1356 1>}
    font {
     bForceMonospaced=false
     bMaterialIsStencil=false
     fontMaterial=null
     fontObject=&Burbank
     letterSpacing=0
     monospacedWidth=1
     size=40
     skewAmount=0
     typefaceFontName=Burbank-Black
     outlineSettings {
      bApplyOutlineToDropShadows=false
      bMiteredCorners=false
      bSeparateFillAlpha=false
      outlineMaterial=null
      outlineSize=1
      outlineColor=<rgba 0.0052 0.0052 0.0056 0.149>
     }
    }
    renderTransform {angle=0, scale=<1 1>, shear=<0 0>, translation=<10 0>}
    renderTransformPivot=<0 0.5>
    shadowColorAndOpacity=<rgba 0.0052 0.0052 0.0056 0.822>
    shadowOffset=<3 3>
   }
   slot {class=VerticalBoxSlot}
  }
  Text_MultiWinner1 {
   widgetClassPath=&CommonTextBlock
   bInherited=true
   bIsVariable=true
   parent=VerticalBox_MultiWinnerContainer
   widget=Text_MultiWinner1
   properties {
    justification=Right
    text=WWWWW
    colorAndOpacity {colorUseRule=UseColor_Specified, specifiedColor=<rgba 1 0.7913 0.1356 1>}
    font {
     bForceMonospaced=false
     bMaterialIsStencil=false
     fontMaterial=&M_Font_Horizon_Dynamic
     fontObject=&Burbank
     letterSpacing=0
     monospacedWidth=1
     size=40
     skewAmount=0
     typefaceFontName=Burbank-Black
     outlineSettings {
      bApplyOutlineToDropShadows=false
      bMiteredCorners=false
      bSeparateFillAlpha=false
      outlineMaterial=null
      outlineSize=1
      outlineColor=<rgba 0.0052 0.0052 0.0056 0.149>
     }
    }
    renderTransform {angle=0, scale=<1 1>, shear=<0 0>, translation=<10 0>}
    renderTransformPivot=<0 0.5>
    shadowColorAndOpacity=<rgba 0.0052 0.0052 0.0056 0.822>
    shadowOffset=<3 3>
   }
   slot {class=VerticalBoxSlot}
  }
  Text_MultiWinner {
   widgetClassPath=&TextBlock
   bInherited=true
   bIsVariable=true
   parent=MultiWinnernames
   widget=Text_MultiWinner
   properties {
    text="and [x] others won the game!"
    textTransformPolicy=ToUpper
    font {
     bForceMonospaced=false
     bMaterialIsStencil=false
     fontMaterial=null
     fontObject=&Burbank
     letterSpacing=0
     monospacedWidth=1
     size=30
     skewAmount=0
     typefaceFontName=Burbank-PlayerPower
     outlineSettings {
      bApplyOutlineToDropShadows=false
      bMiteredCorners=false
      bSeparateFillAlpha=false
      outlineMaterial=null
      outlineSize=1
      outlineColor=<rgba 0.0052 0.0052 0.0056 0.149>
     }
    }
    renderTransform {angle=0, scale=<1 1>, shear=<-8 0>, translation=<0 0>}
    shadowColorAndOpacity=<rgba 0.0052 0.0052 0.0056 0.822>
    shadowOffset=<3 3>
   }
   slot {class=HorizontalBoxSlot, properties {verticalAlignment=VAlign_Center}}
  }
  Text_WinnerPoints {
   widgetClassPath=&TextBlock
   bInherited=true
   bIsVariable=true
   parent=VBox_Winner
   widget=Text_WinnerPoints
   properties {
    text="1,000 points"
    visibility=Collapsed
    font {
     bForceMonospaced=false
     bMaterialIsStencil=false
     fontMaterial=null
     fontObject=&Burbank
     letterSpacing=0
     monospacedWidth=1
     size=34
     skewAmount=0
     typefaceFontName=Burbank-Black
     outlineSettings {
      bApplyOutlineToDropShadows=false
      bMiteredCorners=false
      bSeparateFillAlpha=false
      outlineMaterial=null
      outlineSize=1
      outlineColor=<rgba 0.0052 0.0052 0.0056 0.149>
     }
    }
    shadowColorAndOpacity=<rgba 0.0052 0.0052 0.0056 0.822>
    shadowOffset=<3 3>
   }
   slot {class=VerticalBoxSlot, properties {horizontalAlignment=HAlign_Center, padding=<ltrb 0 -15 0 0>}}
  }
  VBox_PlacementText {
   widgetClassPath=&VerticalBox
   bInherited=true
   bIsVariable=true
   parent=VerticalBoxWinnerPlacementInfo
   widget=VBox_PlacementText
   properties {renderTransformPivot=<0.5 0>}
   slot {class=VerticalBoxSlot, properties {horizontalAlignment=HAlign_Center, verticalAlignment=VAlign_Top, padding=<ltrb 0 0 0 25>}}
  }
  HorizontalBox {
   widgetClassPath=&HorizontalBox
   parent=VBox_PlacementText
   widget=HorizontalBox
   properties {visibility=Visible}
   slot {class=VerticalBoxSlot}
  }
  Text_Placement {
   widgetClassPath=&TextBlock
   bInherited=true
   bIsVariable=true
   parent=HorizontalBox
   widget=Text_Placement
   properties {
    text="YOU PLACED "
    font {
     bForceMonospaced=false
     bMaterialIsStencil=false
     fontMaterial=null
     fontObject=&Burbank
     letterSpacing=0
     monospacedWidth=1
     size=34
     skewAmount=0
     typefaceFontName=Burbank-Black
     outlineSettings {
      bApplyOutlineToDropShadows=false
      bMiteredCorners=false
      bSeparateFillAlpha=false
      outlineMaterial=null
      outlineSize=1
      outlineColor=<rgba 0.0052 0.0052 0.0056 0.149>
     }
    }
    shadowColorAndOpacity=<rgba 0.0052 0.0052 0.0056 0.822>
    shadowOffset=<3 3>
   }
   slot {class=HorizontalBoxSlot, properties {verticalAlignment=VAlign_Center}}
  }
  Text_PlaceNumber {
   widgetClassPath=&CommonTextBlock
   bInherited=true
   bIsVariable=true
   parent=HorizontalBox
   widget=Text_PlaceNumber
   properties {
    justification=Center
    text="##{Place}"
    font {
     bForceMonospaced=false
     bMaterialIsStencil=false
     fontMaterial=null
     fontObject=&Burbank
     letterSpacing=0
     monospacedWidth=1
     size=40
     skewAmount=0
     typefaceFontName=Burbank-Black
     outlineSettings {
      bApplyOutlineToDropShadows=false
      bMiteredCorners=false
      bSeparateFillAlpha=false
      outlineMaterial=null
      outlineSize=1
      outlineColor=<rgba 0.0052 0.0052 0.0056 0.149>
     }
    }
    renderTransform {angle=2, scale=<1 1>, shear=<0 0>, translation=<0 -4>}
    renderTransformPivot=<0 0.5>
    shadowColorAndOpacity=<rgba 0.0052 0.0052 0.0056 0.822>
    shadowOffset=<3 3>
   }
   slot {class=HorizontalBoxSlot, properties {verticalAlignment=VAlign_Center}}
  }
  Text_PlacePoints {
   widgetClassPath=&TextBlock
   bInherited=true
   bIsVariable=true
   parent=VBox_PlacementText
   widget=Text_PlacePoints
   properties {
    text=500 points
    visibility=Collapsed
    font {
     bForceMonospaced=false
     bMaterialIsStencil=false
     fontMaterial=null
     fontObject=&Burbank
     letterSpacing=0
     monospacedWidth=1
     size=34
     skewAmount=0
     typefaceFontName=Burbank-Black
     outlineSettings {
      bApplyOutlineToDropShadows=false
      bMiteredCorners=false
      bSeparateFillAlpha=false
      outlineMaterial=null
      outlineSize=1
      outlineColor=<rgba 0.0052 0.0052 0.0056 0.149>
     }
    }
    shadowColorAndOpacity=<rgba 0.0052 0.0052 0.0056 0.822>
    shadowOffset=<3 3>
   }
   slot {class=VerticalBoxSlot, properties {horizontalAlignment=HAlign_Center, padding=<ltrb 0 -15 0 0>}}
  }
  AthenaWinnersList {
   widgetClassPath=/Game/Athena/HUD/EndGame/AthenaWinnerList
   bInherited=true
   bIsVariable=true
   parent=VerticalBoxWinnerPlacementInfo
   widget=AthenaWinnersList
   properties {visibility=Collapsed}
   slot {class=VerticalBoxSlot, properties {padding=<ltrb 0 0 0 -100>}}
  }
  HBox_LoadingBox {
   widgetClassPath=&HorizontalBox
   bInherited=true
   bIsVariable=true
   parent=VerticalBoxWinnerPlacementInfo
   widget=HBox_LoadingBox
   properties {visibility=Collapsed}
   slot {class=VerticalBoxSlot, properties {horizontalAlignment=HAlign_Center, verticalAlignment=VAlign_Center, padding=<ltrb 0 0 0 -65>}}
  }
  SizeBox_1 {
   widgetClassPath=&SizeBox
   parent=HBox_LoadingBox
   widget=SizeBox_1
   properties {bOverride_HeightOverride=true, heightOverride=65}
   slot {class=HorizontalBoxSlot, properties {horizontalAlignment=HAlign_Center}}
  }
  HorizontalBox_0 {widgetClassPath=&HorizontalBox, parent=SizeBox_1, widget=HorizontalBox_0, slot {class=SizeBoxSlot}}
  ImageSpinner {
   widgetClassPath=&Image
   parent=HorizontalBox_0
   widget=ImageSpinner
   properties {brush {
     bIsDynamicallyLoaded=false
     drawAs=Image
     imageType=NoImage
     mirroring=NoMirror
     resourceName=None
     resourceObject=/FortUILibrary/UIKit/Throbber/MI_UIKit_Monochrome_Throbber
     tiling=NoTile
     imageSize=<42 42>
     margin=<ltrb 0 0 0 0>
     outlineSettings {
      bUseBrushTransparency=false
      roundingType=HalfHeightRadius
      width=0
      color {colorUseRule=UseColor_Specified, specifiedColor=<rgba 0 0 0 0>}
      cornerRadii=<0 0 0 0>
     }
     tintColor {colorUseRule=UseColor_Specified, specifiedColor=<rgba 1 1 1 1>}
     uVRegion {bIsValid=false, max=<0 0>, min=<0 0>}
    }}
   slot {class=HorizontalBoxSlot, properties {horizontalAlignment=HAlign_Right, verticalAlignment=VAlign_Center, padding=<ltrb 0 0 20 0>}}
  }
  TextBlock_0 {
   widgetClassPath=&TextBlock
   parent=HorizontalBox_0
   widget=TextBlock_0
   properties {
    text=Relocating...
    font {
     bForceMonospaced=false
     bMaterialIsStencil=false
     fontMaterial=null
     fontObject=&Burbank
     letterSpacing=0
     monospacedWidth=1
     size=34
     skewAmount=0
     typefaceFontName=Burbank-Black
     outlineSettings {
      bApplyOutlineToDropShadows=false
      bMiteredCorners=false
      bSeparateFillAlpha=false
      outlineMaterial=null
      outlineSize=1
      outlineColor=<rgba 0.0052 0.0052 0.0056 0.149>
     }
    }
    shadowColorAndOpacity=<rgba 0.0052 0.0052 0.0056 0.822>
    shadowOffset=<3 3>
   }
   slot {class=HorizontalBoxSlot, properties {horizontalAlignment=HAlign_Left}}
  }
  TutorialHighlight {widgetClassPath=&TutorialHighlight}
 ]
}
