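# SpectatorModeWidgetContainer is an end-game HUD container for spectator mode UI, derived from AthenaSpectatorWidgetContainer.
# Construct asynchronously creates the content widget and view model, and Destruct signals that the content widget is destroyed.
# When the view model is created, it passes the owning player's FortPlayerStateAthena to it after validity checks.
# GetContentWidgetContainer returns the root Named Slot, Content, as the content host.
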WidgetBlueprint /Game/Athena/HUD/EndGame/SpectatorModeWidgetContainer
graphs [
 EventGraph {nodes [
   Event Construct()
   > AsyncCreateContentWidgetAndViewModel(self:Spectator Widget Container Object Reference)
   Event OnViewModelCreated() {is=title K2Node_Event}
   n1 = GetOwningPlayer()
   > Is Valid(InputObject=@n1)
   > n2 = Cast To FortPlayerStateAthena!(execute, Object=Get PlayerState!(self:Controller Object Reference=@n1) -> (PlayerState:Player State Object Reference)) -> (then, CastFailed, AsFort Player State Athena:Fort Player State Athena Object Reference) {is=title K2Node_DynamicCast}
   n3 = Get ViewModel() -> (ViewModel:Spectator Mode Viewmodel Object Reference)
   > Is Valid(InputObject=@n3)
   > SetPlayerState!(execute, self:Spectator Mode Viewmodel Object Reference=@n3, PlayerState:Fort Player State Athena Object Reference=@n2) -> (then)
   Event Destruct()
   > OnContentWidgetDestroyed(self:Spectator Widget Container Object Reference)
   Event PreConstruct()
   Event Tick()
  ]}
 GetContentWidgetContainer {nodes [
   GetContentWidgetContainer() {is=title K2Node_FunctionEntry}
   > Return Node!(execute, ReturnValue:Content Widget Object Reference=Get Content() -> (Content:Named Slot Object Reference))
  ]}
]
variables []
widgets {
 info {inheritedWidgetCount=0, parentClass=/Script/FortniteUI.AthenaSpectatorWidgetContainer, rootWidgetClass=&NamedSlot, widgetCount=1}
 widgets [Content {widgetClassPath=&NamedSlot, bIsVariable=true, widget=Content}]
}
