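# Battle Royale (Athena) game-over HUD widget, laying out the elimination and spectator screen for the native AthenaGameOverDisplay parent.
# A SafeZone without bottom padding wraps an overlay stacking battle map tooltip, timeline and click handler, the EliminatedOverlay and the spectator widget container.
# A bottom-right ShadowPlay Highlights button starts collapsed and is tied to the AthenaOpenShadowPlay input action.
# The event graph is empty, and the overlay, button and container are inherited bindings.
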WidgetBlueprint /FortniteCoreUI/HUD/EndGame/AthenaGameOverWidget
graphs [EventGraph {links [], nodes []}]
variables []
widgets {
 info {inheritedWidgetCount=3, parentClass=/Script/FortniteUI.AthenaGameOverDisplay, rootWidgetClass=&SafeZone, widgetCount=10}
 widgets [
  SafeZone_0 {widgetClassPath=&SafeZone, widget=SafeZone_0, properties {padBottom=false}}
  Overlay_1 {widgetClassPath=&Overlay, parent=SafeZone_0, widget=Overlay_1, slot {class=SafeZoneSlot}}
  AthenaBattleMapNodeTooltip {
   widgetClassPath=/Game/Athena/UI/Spectator/BattleMap/AthenaBattleMapNodeTooltip
   parent=Overlay_1
   widget=AthenaBattleMapNodeTooltip
   slot {class=OverlaySlot}
  }
  AthenaBattleMapTimelineWidget {
   widgetClassPath=/Game/Athena/UI/Spectator/BattleMap/AthenaBattleMapTimelineWidget
   parent=Overlay_1
   widget=AthenaBattleMapTimelineWidget
   slot {class=OverlaySlot}
  }
  AthenaBattleMapClickHandler {
   widgetClassPath=/Game/Athena/UI/Spectator/BattleMap/AthenaBattleMapClickHandler
   parent=Overlay_1
   widget=AthenaBattleMapClickHandler
   slot {class=OverlaySlot, properties {horizontalAlignment=HAlign_Fill, verticalAlignment=VAlign_Fill}}
  }
  EliminatedOverlay {
   widgetClassPath=&/EliminatedOverlay
   bInherited=true
   bIsVariable=true
   parent=Overlay_1
   widget=EliminatedOverlay
   properties {visibility=HitTestInvisible}
   slot {class=OverlaySlot, properties {horizontalAlignment=HAlign_Fill, verticalAlignment=VAlign_Fill}}
  }
  Overlay_0 {
   widgetClassPath=&Overlay
   parent=Overlay_1
   widget=Overlay_0
   slot {class=OverlaySlot, properties {horizontalAlignment=HAlign_Fill, verticalAlignment=VAlign_Bottom, padding=<ltrb 70 0 70 60>}}
  }
  VerticalBox_0 {
   widgetClassPath=&VerticalBox
   parent=Overlay_0
   widget=VerticalBox_0
   slot {class=OverlaySlot, properties {horizontalAlignment=HAlign_Right, verticalAlignment=VAlign_Bottom}}
  }
  Button_Highlights {
   widgetClassPath=&IconTextButton
   bInherited=true
   bIsVariable=true
   parent=VerticalBox_0
   widget=Button_Highlights
   properties {
    bHideInputActionWithKeyboard=true
    buttonText=""
    onButtonBaseClicked=(&self:WidgetTree.Button_Highlights.HandleOnButtonClicked)
    onButtonBaseDoubleClicked=(&self:WidgetTree.Button_Highlights.HandleOnButtonDoubleClicked)
    onButtonBaseHovered=(&self:WidgetTree.Button_Highlights.HandleOnButtonHovered)
    onButtonBaseUnhovered=(&self:WidgetTree.Button_Highlights.HandleOnButtonUnhovered)
    onSelectedChangedBase=(&self:WidgetTree.Button_Highlights.HandleOnSelectedChanged)
    style=/FortUILibrary/Foundation/Buttons/ButtonStyles/Primary/ButtonStyle-Primary-L.ButtonStyle-Primary-L_C
    visibility=Collapsed
    iconBrush {
     bIsDynamicallyLoaded=false
     drawAs=Image
     imageType=NoImage
     mirroring=NoMirror
     resourceName=None
     resourceObject=/Game/Athena/HUD/Art/ShadowPlayHighlightsButton
     tiling=NoTile
     imageSize=<225 32>
     margin=<ltrb 0 0 0 0>
     outlineSettings {
      bUseBrushTransparency=false
      roundingType=HalfHeightRadius
      width=0
      color {colorUseRule=UseColor_Specified, specifiedColor=<rgba 0 0 0 0>}
      cornerRadii=<0 0 0 1>
     }
     tintColor {colorUseRule=UseColor_Specified, specifiedColor=<rgba 0 0 0 1>}
     uVRegion {bIsValid=false, max=<0 0>, min=<0 0>}
    }
    triggeringInputAction {dataTable=&InputActionDataTable, rowName=AthenaOpenShadowPlay}
   }
   slot {class=VerticalBoxSlot}
  }
  SpectatorWidgetContainer {
   widgetClassPath=/Game/Athena/HUD/EndGame/SpectatorModeWidgetContainer
   bInherited=true
   bIsVariable=true
   parent=Overlay_1
   widget=SpectatorWidgetContainer
   slot {class=OverlaySlot, properties {horizontalAlignment=HAlign_Fill, verticalAlignment=VAlign_Fill}}
  }
 ]
}
