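# Ranked quest notification widget, derived from FortNotificationEntry_Quest, showing quest progress, completion, assists and a timer.
# The intro resets visuals, then UpdateProgress feeds count ratios into the bar's dynamic material and plays FillProgress.
# Animation delegates trigger completion effects and sound, then finish or advance staged objectives through retriggerable delays; a rushed path skips the fill.
# Icons load asynchronously as portrait textures or materials, and clicking tracks the quest.
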WidgetBlueprint /FortniteCoreUI/UI/InGame/EventNotifications/RankedQuestNotification
animations [
 FillProgress {
  binding_count=4
  length_seconds=0.5008
  object=&self:FillProgress
  sequence {bindings [
    Image_textglow {
     id=#1
     possessed_class=Image
     tracks [
      {
       type=MovieScene2DTransformTrack
       sections [{
         type=MovieScene2DTransformSection
         channels [
          Translation.X {type=MovieSceneScriptingFloatChannel}
          Translation.Y {type=MovieSceneScriptingFloatChannel}
          Angle {type=MovieSceneScriptingFloatChannel}
          Scale.X {type=MovieSceneScriptingFloatChannel, keys [[0,1],[10.0153,4]]}
          Scale.Y {type=MovieSceneScriptingFloatChannel, keys [[0,1],[10.0153,4]]}
          Shear.X {type=MovieSceneScriptingFloatChannel}
          Shear.Y {type=MovieSceneScriptingFloatChannel}
         ]
        }]
      }
      {
       type=MovieSceneFloatTrack
       display_name=Render Opacity
       sections [{type=MovieSceneFloatSection, channels [None {type=MovieSceneScriptingFloatChannel, keys [[0,0],[2,1],[10.0153,0]]}]}]
      }
     ]
    }
    RichText_ObjectivesCompletedOLD {
     id=#2
     display_name=RichText_ObjectivesCompleted
     possessed_class=RichTextBlock
     tracks [{
       type=MovieScene2DTransformTrack
       sections [{
         type=MovieScene2DTransformSection
         end_frame=10
         start_frame=0
         channels [
          Translation.X {type=MovieSceneScriptingFloatChannel}
          Translation.Y {type=MovieSceneScriptingFloatChannel}
          Angle {type=MovieSceneScriptingFloatChannel}
          Scale.X {type=MovieSceneScriptingFloatChannel, keys [[0,1],[2,1.05],[10,1]]}
          Scale.Y {type=MovieSceneScriptingFloatChannel, keys [[0,1],[2,1.05],[10,1]]}
          Shear.X {type=MovieSceneScriptingFloatChannel}
          Shear.Y {type=MovieSceneScriptingFloatChannel}
         ]
        }]
      }]
    }
    ProgressFill {
     id=#3
     possessed_class=Image
     tracks [{
       type=MovieSceneWidgetMaterialTrack
       sections [{type=MovieSceneParameterSection, channels [FillAmount {type=MovieSceneScriptingFloatChannel, keys [[0,0],[10.015,1]]}]}]
      }]
    }
    Image_textglow_2 {
     id=#4
     possessed_class=Image
     tracks [
      {
       type=MovieScene2DTransformTrack
       sections [{
         type=MovieScene2DTransformSection
         end_frame=10
         start_frame=0
         channels [
          Angle {type=MovieSceneScriptingFloatChannel, keys [[0,-2],[10,7]]}
          Scale.X {type=MovieSceneScriptingFloatChannel, keys [[0,0.14],[10,1]]}
          Scale.Y {type=MovieSceneScriptingFloatChannel, keys [[0,0.5],[10,1]]}
         ]
        }]
      }
      {
       type=MovieSceneFloatTrack
       display_name=Render Opacity
       sections [{
         type=MovieSceneFloatSection
         end_frame=10
         start_frame=0
         channels [None {type=MovieSceneScriptingFloatChannel, keys [[0,0],[2,0.25],[10.0153,0]]}]
        }]
      }
     ]
    }
   ]}
 }
 QuestComplete {
  binding_count=6
  length_seconds=0.6
  object=&self:QuestComplete
  sequence {bindings [
    Check {
     id=#5
     possessed_class=Image
     tracks [
      {
       type=MovieSceneFloatTrack
       display_name=Render Opacity
       sections [{type=MovieSceneFloatSection, channels [None {type=MovieSceneScriptingFloatChannel, keys [[0,0],[7.2397,1]]}]}]
      }
      {
       type=MovieScene2DTransformTrack
       sections [{
         type=MovieScene2DTransformSection
         channels [
          Translation.X {type=MovieSceneScriptingFloatChannel}
          Translation.Y {type=MovieSceneScriptingFloatChannel, const [-15,1]}
          Angle {type=MovieSceneScriptingFloatChannel}
          Scale.X {type=MovieSceneScriptingFloatChannel, keys [[0,2.3531],[6,0.45],[8,0.835]]}
          Scale.Y {type=MovieSceneScriptingFloatChannel, keys [[0,2.3531],[6,0.45],[8,0.835]]}
          Shear.X {type=MovieSceneScriptingFloatChannel}
          Shear.Y {type=MovieSceneScriptingFloatChannel}
         ]
        }]
      }
     ]
    }
    Image_Glow {
     id=#6
     possessed_class=Image
     tracks [
      {
       type=MovieSceneFloatTrack
       display_name=Render Opacity
       sections [{type=MovieSceneFloatSection, channels [None {type=MovieSceneScriptingFloatChannel, keys [[6,0],[8,1],[12,0]]}]}]
      }
      {
       type=MovieScene2DTransformTrack
       sections [{
         type=MovieScene2DTransformSection
         channels [
          Translation.X {type=MovieSceneScriptingFloatChannel}
          Translation.Y {type=MovieSceneScriptingFloatChannel}
          Angle {type=MovieSceneScriptingFloatChannel}
          Scale.X {type=MovieSceneScriptingFloatChannel, keys [[6,1],[12,1.3]]}
          Scale.Y {type=MovieSceneScriptingFloatChannel, keys [[6,1],[12,1.3]]}
          Shear.X {type=MovieSceneScriptingFloatChannel}
          Shear.Y {type=MovieSceneScriptingFloatChannel}
         ]
        }]
      }
     ]
    }
    Image_Spark {
     id=#7
     possessed_class=Image
     tracks [
      {
       type=MovieSceneFloatTrack
       display_name=Render Opacity
       sections [{type=MovieSceneFloatSection, channels [None {type=MovieSceneScriptingFloatChannel, keys [[6,0],[8,1],[12,0]]}]}]
      }
      {
       type=MovieScene2DTransformTrack
       sections [{
         type=MovieScene2DTransformSection
         channels [
          Translation.X {type=MovieSceneScriptingFloatChannel}
          Translation.Y {type=MovieSceneScriptingFloatChannel}
          Angle {type=MovieSceneScriptingFloatChannel}
          Scale.X {type=MovieSceneScriptingFloatChannel, keys [[6,0.75],[12,1]]}
          Scale.Y {type=MovieSceneScriptingFloatChannel, keys [[6,0.75],[12,1]]}
          Shear.X {type=MovieSceneScriptingFloatChannel}
          Shear.Y {type=MovieSceneScriptingFloatChannel}
         ]
        }]
      }
     ]
    }
    GreenFill {
     id=#8
     possessed_class=Image
     tracks [{
       type=MovieSceneFloatTrack
       display_name=Render Opacity
       sections [{type=MovieSceneFloatSection, channels [None {type=MovieSceneScriptingFloatChannel, const [1,1]}]}]
      }]
    }
    WidgetSwitcher_Complete {
     id=#9
     possessed_class=WidgetSwitcher
     tracks [{
       type=MovieSceneIntegerTrack
       display_name=Active Widget Index
       sections [{type=MovieSceneIntegerSection, channels [None {type=MovieSceneScriptingIntegerChannel, const [1,1]}]}]
      }]
    }
    Overlay_Completiontext {
     id=#10
     possessed_class=Overlay
     tracks [{
       type=MovieSceneByteTrack
       display_name=Visibility
       sections [{type=MovieSceneByteSection, channels [None {type=MovieSceneScriptingByteChannel, const [3,1]}]}]
      }]
    }
   ]}
 }
 ResetQuestComplete {
  binding_count=7
  length_seconds=0.2447
  object=&self:ResetQuestComplete
  sequence {bindings [
    Image_Glow {
     id=#11
     possessed_class=Image
     tracks [{
       type=MovieSceneFloatTrack
       display_name=Render Opacity
       sections [{type=MovieSceneFloatSection, channels [None {type=MovieSceneScriptingFloatChannel, const [0,1]}]}]
      }]
    }
    GreenFill {
     id=#12
     possessed_class=Image
     tracks [{
       type=MovieSceneFloatTrack
       display_name=Render Opacity
       sections [{type=MovieSceneFloatSection, channels [None {type=MovieSceneScriptingFloatChannel, const [0,1]}]}]
      }]
    }
    Overlay_Completiontext {
     id=#13
     possessed_class=Overlay
     tracks [{
       type=MovieSceneByteTrack
       display_name=Visibility
       sections [{type=MovieSceneByteSection, channels [None {type=MovieSceneScriptingByteChannel, const [1,1]}]}]
      }]
    }
    Check {
     id=#14
     possessed_class=Image
     tracks [{
       type=MovieSceneFloatTrack
       display_name=Render Opacity
       sections [{type=MovieSceneFloatSection, channels [None {type=MovieSceneScriptingFloatChannel, const [0,1]}]}]
      }]
    }
    Image_Spark {
     id=#15
     possessed_class=Image
     tracks [{
       type=MovieSceneFloatTrack
       display_name=Render Opacity
       sections [{type=MovieSceneFloatSection, channels [None {type=MovieSceneScriptingFloatChannel, const [0,1]}]}]
      }]
    }
    WidgetSwitcher_Complete {
     id=#16
     possessed_class=WidgetSwitcher
     tracks [{
       type=MovieSceneIntegerTrack
       display_name=Active Widget Index
       sections [{type=MovieSceneIntegerSection, channels [None {type=MovieSceneScriptingIntegerChannel, const [0,1]}]}]
      }]
    }
    ProgressFill {
     id=#17
     possessed_class=Image
     tracks [{
       type=MovieSceneWidgetMaterialTrack
       sections [{type=MovieSceneParameterSection, channels [FillAmount {type=MovieSceneScriptingFloatChannel, const [-0.5824,1]}]}]
      }]
    }
   ]}
 }
 Intro {
  binding_count=0
  length_seconds=0.5
  object=&self:Intro
  sequence {bindings []}
 }
 TurnInNotificationStandin {
  binding_count=1
  length_seconds=0.85
  object=&self:TurnInNotificationStandin
  sequence {bindings [ImageTurnInStandin {
     id=#18
     possessed_class=Image
     tracks [{
       type=MovieSceneFloatTrack
       display_name=Render Opacity
       sections [{type=MovieSceneFloatSection, channels [None {type=MovieSceneScriptingFloatChannel, keys [[0,1],[17,0]]}]}]
      }]
    }]}
 }
]
graphs [
 EventGraph {nodes [
   "On Clicked (Button_MobileClickCatcher)"() {is=title K2Node_ComponentBoundEvent}
   > TrackCurrentQuest(self:Fort Notification Entry Quest Object Reference)
   EventFillComplete() {is=title K2Node_CustomEvent}
   > n1 = Branch(Condition="integer >= integer"!(A:int=Get CurrentCount() -> (CurrentCount:int=0), B:int=Get GoalCount() -> (GoalCount:int=0)) -> (ReturnValue:bool) {is=title K2Node_CallFunction})
   > PlayAnimation(InAnimation=Get QuestComplete() -> (QuestComplete:Widget Animation Object Reference))
   FinishNotificationIntro(execute=@n1.else, self:Fort Notification Entry Object Reference)
   n2 = EventCompleteAnim() {is=title K2Node_CustomEvent}
   > n3 = Branch(Condition=Get bNeedsStagedNotification() -> (bNeedsStagedNotification:bool))
   > TurnInNotification()
   FinishNotificationIntro(execute=@n3.else)
   EventReset() {is=title K2Node_CustomEvent}
   > n4 = GetDynamicMaterial(self=Get ProgressFill_Current() -> (ProgressFill_Current:Image Object Reference))
   n5 = Get GoalCount() -> (GoalCount=0)
   n6 = float / float(A="To Float (Integer)"(InInt=Get PreviousCount() -> (PreviousCount:int=0)), B="To Float (Integer)"(InInt="Max (Integer)"(A=@n5, B=1))) {is=title K2Node_CallFunction}
   > SetScalarParameterValue(self=@n4, ParameterName=StartValue, Value=@n6)
   > n7 = GetDynamicMaterial(self=Get ProgressFill() -> (ProgressFill:Image Object Reference))
   > SetScalarParameterValue(self=@n7, ParameterName=FillAmount, Value=0)
   > SetScalarParameterValue(self=@n7, ParameterName=StartValue, Value=@n6)
   > SetRenderOpacity(
    self=[
     Get Check() -> (Check:Image Object Reference)
     Get Image_Glow() -> (Image_Glow:Image Object Reference)
     Get GreenFill() -> (GreenFill:Image Object Reference)
     Get Image_Spark() -> (Image_Spark:Image Object Reference)
    ]
    InOpacity=0
   )
   > SetActiveWidgetIndex(self=Get WidgetSwitcher_Complete() -> (WidgetSwitcher_Complete:Widget Switcher Object Reference), Index=0)
   n8 = Get CurrentCount() -> (CurrentCount=0)
   n9 = Get GoalCount() -> (GoalCount=0)
   > SetVisibility(
    self=Get Overlay_Completiontext() -> (Overlay_Completiontext:Overlay Object Reference)
    InVisibility=Select(Option 0=Collapsed, Option 1=Hidden, Index="integer >= integer"(A=@n8, B=@n9))
   )
   n10 = EventOnStartComplete() {is=title K2Node_CustomEvent}
   > PlaySound2D(Sound=/Game/Sounds/Fort_UI_Sounds/S15/Quest_Complete_Cue)
   > Branch(Condition=Get bIsEndOfMatch() -> (bIsEndOfMatch:bool))
   TurnInNotification() {is=title K2Node_CustomEvent}
   > RetriggerableDelay(Duration=1)
   > PlayAnimation(InAnimation=Get TurnInNotificationStandin() -> (TurnInNotificationStandin:Widget Animation Object Reference))
   > ShowNextStagedObjectiveInformation(self:Fort Notification Entry Quest Object Reference)
   > EventReset() {is=title K2Node_CallFunction}
   > RetriggerableDelay(Duration=4)
   > FinishNotificationIntro()
   n11 = LoadTextureForDefaultMaterial!() -> (OutputDelegate, then, Asset:Soft Object Reference) {is=title K2Node_CustomEvent}
   > SetBrushFromMaterial(
    self=Get Image_Icon() -> (Image_Icon:Fort Lazy Image Object Reference)
    Material=Get DefaultIconMaterial() -> (DefaultIconMaterial:Material Instance Object Reference)
   )
   > SetIconParameters()
   > n12 = Async Load Asset(Asset=@n11.Asset)
   n13 = Cast To Texture2D(execute=@n12.Completed, Object=@n12)
   > n14 = GetDynamicMaterial(self=Get Image_Icon())
   > SetTextureParameterValue(self=@n14, ParameterName=NPC-Portrait, Value=@n13)
   n15 = Cast To MaterialInterface!(execute=@n13.CastFailed, Object=@n12) -> (then, CastFailed, AsMaterial Interface:Material Interface Object Reference) {is=title K2Node_DynamicCast}
   > SetBrushFromMaterial(self=Get Image_Icon(), Material=@n15)
   > SetIconParameters()
   n16 = LoadMaterial!() -> (OutputDelegate, then, SoftMaterial:Material Interface Soft Object Reference) {is=title K2Node_CustomEvent}
   > n17 = Async Load Asset(Asset=@n16.SoftMaterial)
   n18 = Cast To MaterialInterface(execute=@n17.Completed, Object=@n17)
   > SetBrushFromMaterial(self=Get Image_Icon(), Material=@n18)
   > SetIconParameters()
   n19 = PlayIntroAnimation() {is=title K2Node_CustomEvent}
   EventReset(execute=[@n19.then, @n41.then])
   n20 = Get Intro()
   > PlayAnimation(InAnimation=@n20)
   > RetriggerableDelay(Duration=GetEndTime(self=@n20))
   > UpdateProgress()
   n21 = Event PreConstruct()
   > Parent: PreConstruct(IsDesignTime=@n21.IsDesignTime)
   > SetWidthOverride(self=Get SizeBox_main() -> (SizeBox_main:Size Box Object Reference), InWidthOverride=Get In Width Override() -> (In Width Override:double))
   Event Construct()
   > Parent: Construct()
   n22 = GetOwningPlayer()
   > Is Valid(InputObject=@n22)
   > n23 = GetDynamicMaterial(self:Border Object Reference=Get CommonBorder_MainContent() -> (CommonBorder_MainContent:Common Border Object Reference))
   n24 = Get Class(Object=@n22) -> (ReturnValue:Player Controller Class Reference)
   > SetScalarParameterValue(
    self=@n23
    ParameterName=Opacity
    Value=Select(Option 0:float=0.5, Option 1:float=0.85, Index=ClassIsChildOf(TestClass=@n24, ParentClass=/Script/FortniteGame.FortPlayerControllerFrontEnd)) -> (ReturnValue:float)
   )
   n25 = Event OnUpdateNotificationBackground!() -> (OutputDelegate, then, AssetToDisplay:"Soft Object Reference (by ref)", bUseRarityContainer:bool, bIsDailyPunchcardQuest:bool) {is=title K2Node_Event}
   > n26 = Branch(Condition=@n25.bIsDailyPunchcardQuest)
   > n27 = LoadMaterial!(
    execute
    self
    SoftMaterial:Material Interface Soft Object Reference=Get PunchcardSoftMaterial() -> (PunchcardSoftMaterial:Material Interface Soft Object Reference)
   ) -> (then) {is=title K2Node_CallFunction}
   SetVisibility(
    execute=[@n27.then, @n28.then]
    InVisibility=Select(Option 0:ESlateVisibility Enum=Collapsed, Option 1:ESlateVisibility Enum=Visible, Index=@n25.bUseRarityContainer) -> (ReturnValue:ESlateVisibility Enum)
   )
   > SetVisibility(
    self=Get BorderXPBar() -> (BorderXPBar:Common Border Object Reference)
    InVisibility=Select(Option 0=SelfHitTestInvisible, Option 1=Hidden, Index=@n25.bUseRarityContainer)
   )
   > SetActiveWidget(
    self:Widget Switcher Object Reference=Get Switcher_TimeOrBar() -> (Switcher_TimeOrBar:Widget Switcher Object Reference)
    Widget=Select(
     Option 0:Widget Object Reference=Get BorderXPBar()
     Option 1:Widget Object Reference=Get Border_TimerContainer() -> (Border_TimerContainer:Border Object Reference)
     Index=@n25.bUseRarityContainer
    ) -> (ReturnValue:Widget Object Reference)
   )
   n28 = LoadTextureForDefaultMaterial!(execute=@n26.else, self, Asset:Soft Object Reference=@n25.AssetToDisplay) -> (then) {is=title K2Node_CallFunction}
   n29 = Event OnEnableCompleteFlavorText!() -> (OutputDelegate, then, IsEnabled:bool) {is=title K2Node_Event}
   > n30 = Branch(Condition=@n29.IsEnabled)
   n31 = Get WidgetSwitcher_Complete()
   > SetActiveWidgetIndex(self=@n31, Index=1)
   n32 = Get Overlay_Completiontext()
   > SetVisibility(self=@n32, InVisibility=HitTestInvisible)
   SetActiveWidgetIndex(execute=@n30.else, self=@n31, Index=0)
   > SetVisibility(self=@n32, InVisibility=Collapsed)
   n33 = Event OnPlayRushedNotificationIntro!() -> (OutputDelegate, then, Rate:float) {is=title K2Node_Event}
   > EventReset()
   n34 = "Max (Float)"(A=@n33.Rate, B=1)
   n35 = Get Intro()
   > PlayAnimation(InAnimation=@n35, PlaybackSpeed=@n34)
   > RetriggerableDelay(Duration=GetEndTime(self=@n35))
   > n36 = Branch(Condition="integer >= integer"(A=Get CurrentCount() -> (CurrentCount=0), B=Get GoalCount() -> (GoalCount=0)))
   n37 = Get QuestComplete()
   > PlayAnimation(InAnimation=@n37, PlaybackSpeed=@n34)
   > RetriggerableDelay(Duration=GetEndTime(self=@n37))
   > FinishNotificationOutro(self:Fort Notification Entry Object Reference)
   FinishNotificationOutro(execute=@n36.else)
   n38 = Event SetAssistData!() -> (OutputDelegate, then, bWasAssisted:bool, OtherPlayersName:String) {is=title K2Node_Event}
   > n39 = Branch(Condition=@n38.bWasAssisted)
   n40 = Get Text_AssistBlock() -> (Text_AssistBlock:Common Rich Text Block Object Reference)
   > SetText(
    self=@n40
    InText=Format Text!(Format="LOCTEXT(\"<text case=\\\"upper\\\">Assisted by</> {PlayerName}\")", PlayerName:String=@n38.OtherPlayersName) -> (Result)
   )
   SetText(
    execute=@n39.else
    self=@n40
    InText=Format Text(Format="LOCTEXT(\"<text case=\\\"upper\\\">Assisted</> {PlayerName}\")", PlayerName=@n38.OtherPlayersName)
   )
   n41 = Event StartNotificationIntro() {is=title K2Node_Event}
   Event StartNotificationOutro() {is=title K2Node_Event}
   > FinishNotificationOutro()
   > StopListeningForAllInputActions(self:User Widget Object Reference=[
     Get KeybindWidget() -> (KeybindWidget:Keybind Widget Object Reference)
     Get KeybindWidget_1() -> (KeybindWidget_1:Keybind Widget Object Reference)
    ])
   Event HideBindWidgets() {is=title K2Node_Event}
   n42 = Get HorizontalBoxHotkeyPrompt() -> (HorizontalBoxHotkeyPrompt:Horizontal Box Object Reference)
   > SetVisibility(self=@n42, InVisibility=Collapsed)
   Event ShowBindWidgets() {is=title K2Node_Event}
   > SetVisibility(self=@n42, InVisibility=Visible)
   Event OnInitialized()
   > BindToAnimationFinished(Animation=Get FillProgress() -> (FillProgress:Widget Animation Object Reference), Delegate=Create Event())
   > BindToAnimationFinished(Animation=Get QuestComplete(), Delegate=@n2)
   > BindToAnimationStarted!(execute, self:User Widget Object Reference, Animation:Widget Animation Object Reference=Get QuestComplete(), Delegate:Delegate=@n10) -> (then)
   n43 = Event On Timer Set!() -> (OutputDelegate, then, TimerState:EAthenaChallengeTimerState Enum=Hidden) {is=title K2Node_Event}
   > n44 = Branch(Condition="Equal (Enum)"(A:EAthenaChallengeTimerState Enum=@n43.TimerState, B:EAthenaChallengeTimerState Enum=Hidden))
   n45 = Get Border_TimerContainer()
   > SetVisibility(self=@n45, InVisibility=Hidden)
   SetVisibility(execute=@n44.else, self=@n45, InVisibility=SelfHitTestInvisible)
   > SetContentColorAndOpacity!(
    execute
    self:Border Object Reference=Get Border_TimerContainer()
    InContentColorAndOpacity:Linear Color Structure=Select!(
     Option 0:Linear Color Structure
     Option 1:Linear Color Structure
     Option 2:Linear Color Structure
     Option 3:Linear Color Structure
     Option 4:Linear Color Structure
     Option 5:Linear Color Structure
     Index:Byte=@n43.TimerState
    ) -> (ReturnValue:Linear Color Structure)
   ) -> (then)
   > SetActiveWidget(
    self=Get WidgetSwitcher_Urgent() -> (WidgetSwitcher_Urgent:Widget Switcher Object Reference)
    Widget=Select!(
     Option 0:Widget Object Reference=Get DateTimeText_TimeRemaining() -> (DateTimeText_TimeRemaining:Athena Challenge Date Time Text Block Object Reference)
     Option 1:Widget Object Reference=Get Text_TimeRemainingUrgent() -> (Text_TimeRemainingUrgent:Common Text Object Reference)
     Index:bool="Equal (Enum)"(A=@n43.TimerState, B=Urgent)
    ) -> (ReturnValue:Widget Object Reference)
   )
   Event On Initialize Quest Notification Finished() {is=title K2Node_Event}
   > Branch(Condition=Get bIsEndOfMatch())
   n46 = Get Text_CategoryDisplay() -> (Text_CategoryDisplay:Common Text Object Reference)
   > Branch(Condition=TextIsEmpty(InText="GetText (Text)"!(self:Text Object Reference=@n46) -> (ReturnValue:Text)))
   > SetVisibility(self=@n46, InVisibility=Collapsed)
  ]}
 SetIconParameters {nodes [
   SetIconParameters() {is=title K2Node_FunctionEntry}
   > n1 = GetDynamicMaterial(self:Image Object Reference=Get Image_Icon())
   > SetScalarParameterValue(self=@n1, ParameterName=Corner roundness, Value=0.15)
   > SetScalarParameterValue(self=@n1, ParameterName=Left aligned, Value=0)
  ]}
 UpdateProgress {nodes [
   UpdateProgress() {is=title K2Node_FunctionEntry}
   > Branch(Condition="Not Equal (Integer)"(A=Get CurrentCount() -> (CurrentCount=0), B=Get PreviousCount() -> (PreviousCount=0)))
   > n1 = GetDynamicMaterial(self=Get ProgressFill())
   n2 = Get PreviousCount() -> (PreviousCount=0)
   n3 = "Max (Integer)"(A=Get GoalCount() -> (GoalCount=0), B=1)
   > SetScalarParameterValue(self=@n1, ParameterName=StartValue, Value=float / float(A="To Float (Integer)"(InInt=@n2), B="To Float (Integer)"(InInt=@n3)))
   n4 = Get CurrentCount() -> (CurrentCount=0)
   n5 = "Max (Integer)"(A=Get GoalCount() -> (GoalCount=0), B=1)
   > SetScalarParameterValue(self=@n1, ParameterName=EndValue, Value=float / float(A="To Float (Integer)"(InInt=@n4), B="To Float (Integer)"(InInt=@n5)))
   > PlayAnimation(InAnimation=Get FillProgress())
  ]}
]
variables [
 bIsEndOfMatch {type=boolean}
 DefaultIconMaterial {type=MaterialInstance, default=/FortniteCoreUI/HUD/MiniMap/Art/QuestBundle/MI_UI_QuestBundle_QuestNotification_Default}
 Flavor Text Visibility {type=ESlateVisibility, default=SelfHitTestInvisible}
 In Width Override {type=number, default=480}
 MID_MissionHeader {type=MaterialInstanceDynamic}
 PunchcardSoftMaterial {type=MaterialInterface, default=/FortniteCoreUI/HUD/MiniMap/Art/QuestBundle/M_UI_QuestBundle_RustyBolt}
]
widgets {
 info {inheritedWidgetCount=9, parentClass=/Script/FortniteUI.FortNotificationEntry_Quest, rootWidgetClass=&Overlay, widgetCount=65}
 widgets [
  Overlay_5 {widgetClassPath=&Overlay, widget=Overlay_5}
  VerticalBox_1 {
   widgetClassPath=&VerticalBox
   parent=Overlay_5
   widget=VerticalBox_1
   slot {class=OverlaySlot, properties {horizontalAlignment=HAlign_Fill, verticalAlignment=VAlign_Fill}}
  }
  Text_CategoryDisplay {
   widgetClassPath=&CommonTextBlock
   bInherited=true
   bIsVariable=true
   parent=VerticalBox_1
   widget=Text_CategoryDisplay
   properties {
    autoWrapText=true
    colorAndOpacity {colorUseRule=UseColor_Specified, specifiedColor=<rgba 0.1088 0.3619 0.6788 0.6>}
    font {
     bForceMonospaced=false
     bMaterialIsStencil=false
     fontMaterial=null
     fontObject=&BurbankSmall
     letterSpacing=0
     monospacedWidth=1
     size=24
     skewAmount=0
     typefaceFontName=Bold
     outlineSettings {
      bApplyOutlineToDropShadows=false
      bMiteredCorners=false
      bSeparateFillAlpha=false
      outlineMaterial=null
      outlineSize=0
      outlineColor=<rgba 0 0 0 1>
     }
    }
    shadowOffset=<0 0>
   }
   slot {class=VerticalBoxSlot}
  }
  CommonBorder_MainContent {
   widgetClassPath=&CommonBorder
   bIsVariable=true
   parent=VerticalBox_1
   widget=CommonBorder_MainContent
   properties {bReducePaddingBySafezone=true, verticalAlignment=VAlign_Top, padding=<ltrb 0 0 0 0>}
   slot {class=VerticalBoxSlot}
  }
  GridPanel_1 {
   widgetClassPath=&GridPanel
   parent=CommonBorder_MainContent
   widget=GridPanel_1
   properties {columnFill [1], rowFill [1]}
   slot {class=BorderSlot, properties {verticalAlignment=VAlign_Top, padding=<ltrb 0 0 0 0>}}
  }
  UIKit_Backplate {
   widgetClassPath=&WBP_UIKit_Backplate
   parent=GridPanel_1
   widget=UIKit_Backplate
   properties {brightness=Dark}
   slot {class=GridSlot}
  }
  VerticalBox_69 {widgetClassPath=&VerticalBox, parent=GridPanel_1, widget=VerticalBox_69, slot {class=GridSlot, properties {layer=1}}}
  SizeBox_main {
   widgetClassPath=&SizeBox
   bIsVariable=true
   parent=VerticalBox_69
   widget=SizeBox_main
   properties {bOverride_MinDesiredHeight=true, maxDesiredWidth=360, minDesiredHeight=100}
   slot {class=VerticalBoxSlot}
  }
  Overlay_3 {widgetClassPath=&Overlay, parent=SizeBox_main, widget=Overlay_3, slot {class=SizeBoxSlot}}
  HorizontalBox_4 {
   widgetClassPath=&HorizontalBox
   parent=Overlay_3
   widget=HorizontalBox_4
   slot {class=OverlaySlot, properties {horizontalAlignment=HAlign_Fill, verticalAlignment=VAlign_Fill}}
  }
  VerticalBox_0 {
   widgetClassPath=&VerticalBox
   parent=HorizontalBox_4
   widget=VerticalBox_0
   slot {class=HorizontalBoxSlot, properties {padding=<ltrb 14 6 14 6>, size {sizeRule=Fill, value=1}}}
  }
  Panel_AssistedBy {
   widgetClassPath=&Border
   bInherited=true
   bIsVariable=true
   parent=VerticalBox_0
   widget=Panel_AssistedBy
   properties {
    background {
     bIsDynamicallyLoaded=false
     drawAs=NoDrawType
     imageType=NoImage
     mirroring=NoMirror
     resourceName=None
     resourceObject=/FortniteCoreUI/Athena/UI/Challenges/Art/M_UI_DefaultChallengeHeader
     tiling=NoTile
     imageSize=<32 32>
     margin=<ltrb 0 0 0 0>
     outlineSettings {
      bUseBrushTransparency=false
      roundingType=HalfHeightRadius
      width=0
      color {colorUseRule=UseColor_Specified, specifiedColor=<rgba 0 0 0 0>}
      cornerRadii=<0 0 0 0>
     }
     tintColor {colorUseRule=UseColor_Specified, specifiedColor=<rgba 1 1 1 1>}
     uVRegion {bIsValid=false, max=<0 0>, min=<0 0>}
    }
    padding=<ltrb 4 0 0 4>
   }
   slot {class=VerticalBoxSlot}
  }
  HorizontalBox_3 {
   widgetClassPath=&HorizontalBox
   parent=Panel_AssistedBy
   widget=HorizontalBox_3
   slot {class=BorderSlot, properties {padding=<ltrb 4 0 0 4>}}
  }
  Overlay_Assist {
   widgetClassPath=&Overlay
   parent=HorizontalBox_3
   widget=Overlay_Assist
   properties {renderTransform {angle=0, scale=<2 2>, shear=<0 0>, translation=<0 0>}}
   slot {class=HorizontalBoxSlot, properties {horizontalAlignment=HAlign_Left, verticalAlignment=VAlign_Center}}
  }
  Image_GotHelped {
   widgetClassPath=&Image
   parent=Overlay_Assist
   widget=Image_GotHelped
   properties {
    visibility=SelfHitTestInvisible
    brush {
     bIsDynamicallyLoaded=false
     drawAs=Image
     imageType=NoImage
     mirroring=NoMirror
     resourceName=None
     resourceObject=/Game/Athena/UI/Challenges/Art/T_UI_PartyAssist_Outgoing_64
     tiling=NoTile
     imageSize=<16 16>
     margin=<ltrb 0 0 0 0>
     outlineSettings {
      bUseBrushTransparency=false
      roundingType=HalfHeightRadius
      width=0
      color {colorUseRule=UseColor_Specified, specifiedColor=<rgba 0 0 0 0>}
      cornerRadii=<0 0 0 0>
     }
     tintColor {colorUseRule=UseColor_Specified, specifiedColor=<rgba 1 1 1 1>}
     uVRegion {bIsValid=false, max=<0 0>, min=<0 0>}
    }
   }
   slot {class=OverlaySlot, properties {horizontalAlignment=HAlign_Center, verticalAlignment=VAlign_Center}}
  }
  AttentionAnim {
   widgetClassPath=&Image
   parent=Overlay_Assist
   widget=AttentionAnim
   properties {
    visibility=Collapsed
    brush {
     bIsDynamicallyLoaded=false
     drawAs=Image
     imageType=NoImage
     mirroring=NoMirror
     resourceName=None
     resourceObject=/FortniteCoreUI/Athena/UI/Challenges/Art/M_UI_ChallengeTile_PartyAssistAttention
     tiling=NoTile
     imageSize=<16 16>
     margin=<ltrb 0 0 0 0>
     outlineSettings {
      bUseBrushTransparency=false
      roundingType=HalfHeightRadius
      width=0
      color {colorUseRule=UseColor_Specified, specifiedColor=<rgba 0 0 0 0>}
      cornerRadii=<0 0 0 0>
     }
     tintColor {colorUseRule=UseColor_Specified, specifiedColor=<rgba 1 1 1 1>}
     uVRegion {bIsValid=false, max=<0 0>, min=<0 0>}
    }
   }
   slot {class=OverlaySlot, properties {horizontalAlignment=HAlign_Center, verticalAlignment=VAlign_Center}}
  }
  Text_AssistBlock {
   widgetClassPath=&CommonRichTextBlock
   bIsVariable=true
   parent=HorizontalBox_3
   widget=Text_AssistBlock
   properties {
    bOverrideDefaultStyle=true
    mobileTextBlockScale=1.1
    scrollStyle=&TextScrollStyle-Base_C
    text="<text case=\"upper\">Assisted by</> {PlayerName}"
    wrapTextAt=400
    defaultTextStyleOverride {
     overflowPolicy=Clip
     transformPolicy=None
     colorAndOpacity {colorUseRule=UseColor_Specified, specifiedColor=<rgba 1 1 1 1>}
     font {
      bForceMonospaced=false
      bMaterialIsStencil=false
      fontMaterial=null
      fontObject=&BurbankSmall
      letterSpacing=0
      monospacedWidth=1
      size=14
      skewAmount=0
      typefaceFontName=Black
      outlineSettings {
       bApplyOutlineToDropShadows=false
       bMiteredCorners=false
       bSeparateFillAlpha=false
       outlineMaterial=null
       outlineSize=0
       outlineColor=<rgba 0 0 0 1>
      }
     }
     highlightColor {colorUseRule=UseColor_Specified, specifiedColor=<rgba 0 0 0 1>}
     highlightShape {
      bIsDynamicallyLoaded=false
      drawAs=Image
      imageType=NoImage
      mirroring=NoMirror
      resourceName=None
      resourceObject=null
      tiling=NoTile
      imageSize=<32 32>
      margin=<ltrb 0 0 0 0>
      outlineSettings {
       bUseBrushTransparency=false
       roundingType=HalfHeightRadius
       width=0
       color {colorUseRule=UseColor_Specified, specifiedColor=<rgba 0 0 0 0>}
       cornerRadii=<0 0 0 0>
      }
      tintColor {colorUseRule=UseColor_Specified, specifiedColor=<rgba 1 1 1 1>}
      uVRegion {bIsValid=false, max=<0 0>, min=<0 0>}
     }
     selectedBackgroundColor {colorUseRule=UseColor_Foreground, specifiedColor=<rgba 1 0 1 1>}
     shadowColorAndOpacity=<rgba 0 0 0 1>
     shadowOffset=<0 0>
     strikeBrush {
      bIsDynamicallyLoaded=false
      drawAs=Image
      imageType=NoImage
      mirroring=NoMirror
      resourceName=None
      resourceObject=null
      tiling=NoTile
      imageSize=<32 32>
      margin=<ltrb 0 0 0 0>
      outlineSettings {
       bUseBrushTransparency=false
       roundingType=HalfHeightRadius
       width=0
       color {colorUseRule=UseColor_Specified, specifiedColor=<rgba 0 0 0 0>}
       cornerRadii=<0 0 0 0>
      }
      tintColor {colorUseRule=UseColor_Specified, specifiedColor=<rgba 1 1 1 1>}
      uVRegion {bIsValid=false, max=<0 0>, min=<0 0>}
     }
     underlineBrush {
      bIsDynamicallyLoaded=false
      drawAs=Image
      imageType=NoImage
      mirroring=NoMirror
      resourceName=None
      resourceObject=null
      tiling=NoTile
      imageSize=<32 32>
      margin=<ltrb 0 0 0 0>
      outlineSettings {
       bUseBrushTransparency=false
       roundingType=HalfHeightRadius
       width=0
       color {colorUseRule=UseColor_Specified, specifiedColor=<rgba 0 0 0 0>}
       cornerRadii=<0 0 0 0>
      }
      tintColor {colorUseRule=UseColor_Specified, specifiedColor=<rgba 1 1 1 1>}
      uVRegion {bIsValid=false, max=<0 0>, min=<0 0>}
     }
    }
    renderTransform {angle=0, scale=<1 1>, shear=<-8 0>, translation=<0 0>}
   }
   slot {class=HorizontalBoxSlot, properties {verticalAlignment=VAlign_Center, padding=<ltrb 12 0 0 0>, size {sizeRule=Fill, value=1}}}
  }
  Text_Description {
   widgetClassPath=&CommonTextBlock
   bInherited=true
   bIsVariable=true
   parent=VerticalBox_0
   widget=Text_Description
   properties {
    autoWrapText=true
    lineHeightPercentage=0.9
    mobileFontSizeMultiplier=1.4
    text=Do the task
    font {
     bForceMonospaced=false
     bMaterialIsStencil=false
     fontMaterial=null
     fontObject=&BurbankSmall
     letterSpacing=0
     monospacedWidth=1
     size=18
     skewAmount=0
     typefaceFontName=Bold
     outlineSettings {
      bApplyOutlineToDropShadows=false
      bMiteredCorners=false
      bSeparateFillAlpha=false
      outlineMaterial=null
      outlineSize=0
      outlineColor=<rgba 0 0 0 1>
     }
    }
    shadowOffset=<0 0>
   }
   slot {class=VerticalBoxSlot, properties {verticalAlignment=VAlign_Top}}
  }
  Overlay_Completiontext {
   widgetClassPath=&Overlay
   bIsVariable=true
   parent=VerticalBox_0
   widget=Overlay_Completiontext
   properties {visibility=Collapsed}
   slot {class=VerticalBoxSlot, properties {verticalAlignment=VAlign_Top, padding=<ltrb 0 0 0 4>}}
  }
  Text_FlavorText {
   widgetClassPath=&CommonTextBlock
   bInherited=true
   bIsVariable=true
   parent=Overlay_Completiontext
   widget=Text_FlavorText
   properties {
    autoWrapText=true
    bAutoCollapseWithEmptyText=true
    lineHeightPercentage=0.9
    mobileFontSizeMultiplier=1.3
    text=Completion text.
    colorAndOpacity {colorUseRule=UseColor_Specified, specifiedColor=<rgba 0.6038 0.8879 1 1>}
    font {
     bForceMonospaced=false
     bMaterialIsStencil=false
     fontMaterial=null
     fontObject=&BurbankSmall
     letterSpacing=0
     monospacedWidth=1
     size=15
     skewAmount=0
     typefaceFontName=Bold
     outlineSettings {
      bApplyOutlineToDropShadows=false
      bMiteredCorners=false
      bSeparateFillAlpha=false
      outlineMaterial=null
      outlineSize=0
      outlineColor=<rgba 0 0 0 1>
     }
    }
    shadowOffset=<0 0>
   }
   slot {class=OverlaySlot, properties {horizontalAlignment=HAlign_Fill}}
  }
  HorizontalBox_5 {
   widgetClassPath=&HorizontalBox
   parent=VerticalBox_0
   widget=HorizontalBox_5
   slot {class=VerticalBoxSlot, properties {verticalAlignment=VAlign_Bottom, padding=<ltrb 0 6 0 0>, size {sizeRule=Fill, value=1}}}
  }
  Switcher_TimeOrBar {
   widgetClassPath=&WidgetSwitcher
   bIsVariable=true
   parent=HorizontalBox_5
   widget=Switcher_TimeOrBar
   slot {class=HorizontalBoxSlot, properties {horizontalAlignment=HAlign_Left, verticalAlignment=VAlign_Bottom, padding=<ltrb 0 0 12 0>}}
  }
  Border_TimerContainer {
   widgetClassPath=&Border
   bIsVariable=true
   parent=Switcher_TimeOrBar
   widget=Border_TimerContainer
   properties {
    background {
     bIsDynamicallyLoaded=false
     drawAs=NoDrawType
     imageType=NoImage
     mirroring=NoMirror
     resourceName=None
     resourceObject=null
     tiling=NoTile
     imageSize=<32 32>
     margin=<ltrb 0 0 0 0>
     outlineSettings {
      bUseBrushTransparency=false
      roundingType=HalfHeightRadius
      width=0
      color {colorUseRule=UseColor_Specified, specifiedColor=<rgba 0 0 0 0>}
      cornerRadii=<0 0 0 0>
     }
     tintColor {colorUseRule=UseColor_Specified, specifiedColor=<rgba 1 1 1 1>}
     uVRegion {bIsValid=false, max=<0 0>, min=<0 0>}
    }
    contentColorAndOpacity=<rgba 0 0.8796 0.8632 1>
    padding=<ltrb 0 0 0 0>
   }
   slot {class=WidgetSwitcherSlot}
  }
  TimerContainer {
   widgetClassPath=&HorizontalBox
   parent=Border_TimerContainer
   widget=TimerContainer
   properties {flowDirectionPreference=Culture}
   slot {class=BorderSlot, properties {padding=<ltrb 0 0 0 0>}}
  }
  ImageTimer {
   widgetClassPath=&Image
   parent=TimerContainer
   widget=ImageTimer
   properties {brush {
     bIsDynamicallyLoaded=false
     drawAs=Image
     imageType=NoImage
     mirroring=NoMirror
     resourceName=None
     resourceObject=/FortniteCoreUI/Athena/HUD/Quests/Art/M_UI_Timer_Icon_Quest
     tiling=NoTile
     imageSize=<24 24>
     margin=<ltrb 0 0 0 0>
     outlineSettings {
      bUseBrushTransparency=false
      roundingType=HalfHeightRadius
      width=0
      color {colorUseRule=UseColor_Specified, specifiedColor=<rgba 0 0 0 0>}
      cornerRadii=<0 0 0 0>
     }
     tintColor {colorUseRule=UseColor_Specified, specifiedColor=<rgba 1 1 1 1>}
     uVRegion {bIsValid=false, max=<0 0>, min=<0 0>}
    }}
   slot {class=HorizontalBoxSlot, properties {horizontalAlignment=HAlign_Left, verticalAlignment=VAlign_Center, padding=<ltrb -6 -12 0 -12>}}
  }
  WidgetSwitcher_Urgent {
   widgetClassPath=&WidgetSwitcher
   bIsVariable=true
   parent=TimerContainer
   widget=WidgetSwitcher_Urgent
   slot {class=HorizontalBoxSlot, properties {verticalAlignment=VAlign_Center, size {sizeRule=Fill, value=1}}}
  }
  DateTimeText_TimeRemaining {
   widgetClassPath=/Script/FortniteUI.AthenaChallengeDateTimeTextBlock
   bInherited=true
   bIsVariable=true
   parent=WidgetSwitcher_Urgent
   widget=DateTimeText_TimeRemaining
   properties {
    mobileFontSizeMultiplier=1.2
    text=Text Block
    font {
     bForceMonospaced=false
     bMaterialIsStencil=false
     fontMaterial=/FortniteCoreUI/Athena/HUD/Quests/Art/M_UI_Timer_Font_Quest
     fontObject=&BurbankSmall
     letterSpacing=0
     monospacedWidth=1
     size=15
     skewAmount=0
     typefaceFontName=Bold
     outlineSettings {
      bApplyOutlineToDropShadows=false
      bMiteredCorners=false
      bSeparateFillAlpha=false
      outlineMaterial=null
      outlineSize=0
      outlineColor=<rgba 0 0 0 1>
     }
    }
    shadowOffset=<0 0>
   }
   slot {class=WidgetSwitcherSlot, properties {verticalAlignment=VAlign_Bottom}}
  }
  Text_TimeRemainingUrgent {
   widgetClassPath=&CommonTextBlock
   bIsVariable=true
   parent=WidgetSwitcher_Urgent
   widget=Text_TimeRemainingUrgent
   properties {
    mobileFontSizeMultiplier=1.2
    text="This Match Only!"
    textTransformPolicy=ToUpper
    visibility=Collapsed
    font {
     bForceMonospaced=false
     bMaterialIsStencil=false
     fontMaterial=&M_UI_Timer_Font_Quest
     fontObject=&BurbankSmall
     letterSpacing=0
     monospacedWidth=1
     size=15
     skewAmount=0
     typefaceFontName=Bold
     outlineSettings {
      bApplyOutlineToDropShadows=false
      bMiteredCorners=false
      bSeparateFillAlpha=false
      outlineMaterial=null
      outlineSize=0
      outlineColor=<rgba 0 0 0 1>
     }
    }
    shadowOffset=<0 0>
   }
   slot {class=WidgetSwitcherSlot}
  }
  BorderXPBar {
   widgetClassPath=&CommonBorder
   bIsVariable=true
   parent=Switcher_TimeOrBar
   widget=BorderXPBar
   properties {
    style=/FortUILibrary/Styles/Border/Border_StatRow_WhiteMidOpacity.Border_StatRow_WhiteMidOpacity_C
    visibility=Hidden
    brushColor=<rgba 1 1 1 0>
    padding=<ltrb 0 0 0 0>
   }
   slot {class=WidgetSwitcherSlot, properties {verticalAlignment=VAlign_Center}}
  }
  SizeBox_1 {
   widgetClassPath=&SizeBox
   parent=BorderXPBar
   widget=SizeBox_1
   properties {bOverride_HeightOverride=true, bOverride_MinDesiredWidth=true, heightOverride=6, minDesiredWidth=8}
   slot {class=BorderSlot, properties {padding=<ltrb 0 0 0 0>}}
  }
  Overlay_1 {widgetClassPath=&Overlay, parent=SizeBox_1, widget=Overlay_1, slot {class=SizeBoxSlot}}
  ProgressFill {
   widgetClassPath=&Image
   bIsVariable=true
   parent=Overlay_1
   widget=ProgressFill
   properties {brush {
     bIsDynamicallyLoaded=false
     drawAs=Image
     imageType=NoImage
     mirroring=NoMirror
     resourceName=None
     resourceObject=/FortniteCoreUI/Athena/UI/Progression/MI_Quest_Bar_Dynamic_Toast
     tiling=NoTile
     imageSize=<6 6>
     margin=<ltrb 0 0 0 0>
     outlineSettings {
      bUseBrushTransparency=false
      roundingType=HalfHeightRadius
      width=0
      color {colorUseRule=UseColor_Specified, specifiedColor=<rgba 0 0 0 0>}
      cornerRadii=<0 0 0 0>
     }
     tintColor {colorUseRule=UseColor_Specified, specifiedColor=<rgba 1 1 1 1>}
     uVRegion {bIsValid=false, max=<0 0>, min=<0 0>}
    }}
   slot {class=OverlaySlot, properties {horizontalAlignment=HAlign_Fill, verticalAlignment=VAlign_Fill}}
  }
  ProgressFill_Current {
   widgetClassPath=&Image
   bIsVariable=true
   parent=Overlay_1
   widget=ProgressFill_Current
   properties {brush {
     bIsDynamicallyLoaded=false
     drawAs=Image
     imageType=NoImage
     mirroring=NoMirror
     resourceName=None
     resourceObject=/FortniteCoreUI/Athena/UI/Progression/MI_Quest_Bar_Dynamic_Inst
     tiling=NoTile
     imageSize=<6 6>
     margin=<ltrb 0 0 0 0>
     outlineSettings {
      bUseBrushTransparency=false
      roundingType=HalfHeightRadius
      width=0
      color {colorUseRule=UseColor_Specified, specifiedColor=<rgba 0 0 0 0>}
      cornerRadii=<0 0 0 0>
     }
     tintColor {colorUseRule=UseColor_Specified, specifiedColor=<rgba 1 1 1 1>}
     uVRegion {bIsValid=false, max=<0 0>, min=<0 0>}
    }}
   slot {class=OverlaySlot, properties {horizontalAlignment=HAlign_Fill, verticalAlignment=VAlign_Fill}}
  }
  GreenFill {
   widgetClassPath=&Image
   bIsVariable=true
   parent=Overlay_1
   widget=GreenFill
   properties {
    renderOpacity=0
    brush {
     bIsDynamicallyLoaded=false
     drawAs=Image
     imageType=NoImage
     mirroring=NoMirror
     resourceName=None
     resourceObject=null
     tiling=NoTile
     imageSize=<6 6>
     margin=<ltrb 0 0 0 0>
     outlineSettings {
      bUseBrushTransparency=false
      roundingType=HalfHeightRadius
      width=0
      color {colorUseRule=UseColor_Specified, specifiedColor=<rgba 0 0 0 0>}
      cornerRadii=<0 0 0 0>
     }
     tintColor {colorUseRule=UseColor_Specified, specifiedColor=<rgba 1 1 1 1>}
     uVRegion {bIsValid=false, max=<0 0>, min=<0 0>}
    }
    colorAndOpacity=<rgba 0.0343 1 0.227 1>
   }
   slot {class=OverlaySlot, properties {horizontalAlignment=HAlign_Fill, verticalAlignment=VAlign_Fill}}
  }
  WidgetSwitcher_Complete {
   widgetClassPath=&WidgetSwitcher
   bIsVariable=true
   parent=HorizontalBox_5
   widget=WidgetSwitcher_Complete
   slot {class=HorizontalBoxSlot, properties {horizontalAlignment=HAlign_Right, verticalAlignment=VAlign_Bottom, size {sizeRule=Fill, value=1}}}
  }
  Progress {widgetClassPath=&HorizontalBox, parent=WidgetSwitcher_Complete, widget=Progress, slot {class=WidgetSwitcherSlot}}
  Overlay_4 {
   widgetClassPath=&Overlay
   parent=Progress
   widget=Overlay_4
   slot {class=HorizontalBoxSlot, properties {horizontalAlignment=HAlign_Right, verticalAlignment=VAlign_Bottom, padding=<ltrb 12 0 0 0>}}
  }
  RichText_ObjectivesCompleted {
   widgetClassPath=&CommonRichTextBlock
   bInherited=true
   bIsVariable=true
   parent=Overlay_4
   widget=RichText_ObjectivesCompleted
   properties {
    bOverrideDefaultStyle=true
    mobileTextBlockScale=1.2
    text="<UnselectedStyle>100</> / 100"
    textStyleSet=/FortniteCoreUI/HUD/MiniMap/AthenaChallengeObjectiveProgressStyles
    defaultTextStyleOverride {
     overflowPolicy=Clip
     transformPolicy=None
     colorAndOpacity {colorUseRule=UseColor_Specified, specifiedColor=<rgba 0 0.8796 0.8632 1>}
     font {
      bForceMonospaced=false
      bMaterialIsStencil=false
      fontMaterial=null
      fontObject=&HeadingNow
      letterSpacing=0
      monospacedWidth=1
      size=12.75
      skewAmount=0
      typefaceFontName=76 Bold
      outlineSettings {
       bApplyOutlineToDropShadows=false
       bMiteredCorners=false
       bSeparateFillAlpha=false
       outlineMaterial=null
       outlineSize=1
       outlineColor=<rgba 0.0104 0.0104 0.0104 0.6>
      }
     }
     highlightColor {colorUseRule=UseColor_Specified, specifiedColor=<rgba 0 0 0 1>}
     highlightShape {
      bIsDynamicallyLoaded=false
      drawAs=Image
      imageType=NoImage
      mirroring=NoMirror
      resourceName=None
      resourceObject=null
      tiling=NoTile
      imageSize=<32 32>
      margin=<ltrb 0 0 0 0>
      outlineSettings {
       bUseBrushTransparency=false
       roundingType=HalfHeightRadius
       width=0
       color {colorUseRule=UseColor_Specified, specifiedColor=<rgba 0 0 0 0>}
       cornerRadii=<0 0 0 0>
      }
      tintColor {colorUseRule=UseColor_Specified, specifiedColor=<rgba 1 1 1 1>}
      uVRegion {bIsValid=false, max=<0 0>, min=<0 0>}
     }
     selectedBackgroundColor {colorUseRule=UseColor_Foreground, specifiedColor=<rgba 1 0 1 1>}
     shadowColorAndOpacity=<rgba 0 0 0 1>
     shadowOffset=<0 0>
     strikeBrush {
      bIsDynamicallyLoaded=false
      drawAs=Image
      imageType=NoImage
      mirroring=NoMirror
      resourceName=None
      resourceObject=null
      tiling=NoTile
      imageSize=<32 32>
      margin=<ltrb 0 0 0 0>
      outlineSettings {
       bUseBrushTransparency=false
       roundingType=HalfHeightRadius
       width=0
       color {colorUseRule=UseColor_Specified, specifiedColor=<rgba 0 0 0 0>}
       cornerRadii=<0 0 0 0>
      }
      tintColor {colorUseRule=UseColor_Specified, specifiedColor=<rgba 1 1 1 1>}
      uVRegion {bIsValid=false, max=<0 0>, min=<0 0>}
     }
     underlineBrush {
      bIsDynamicallyLoaded=false
      drawAs=Image
      imageType=NoImage
      mirroring=NoMirror
      resourceName=None
      resourceObject=null
      tiling=NoTile
      imageSize=<32 32>
      margin=<ltrb 0 0 0 0>
      outlineSettings {
       bUseBrushTransparency=false
       roundingType=HalfHeightRadius
       width=0
       color {colorUseRule=UseColor_Specified, specifiedColor=<rgba 0 0 0 0>}
       cornerRadii=<0 0 0 0>
      }
      tintColor {colorUseRule=UseColor_Specified, specifiedColor=<rgba 1 1 1 1>}
      uVRegion {bIsValid=false, max=<0 0>, min=<0 0>}
     }
    }
   }
   slot {class=OverlaySlot}
  }
  Image_textglow {
   widgetClassPath=&Image
   parent=Overlay_4
   widget=Image_textglow
   properties {
    visibility=HitTestInvisible
    brush {
     bIsDynamicallyLoaded=false
     drawAs=Image
     imageType=NoImage
     mirroring=NoMirror
     resourceName=None
     resourceObject=/Game/Athena/HUD/LTM/Wax/Materials/M_UI_TextGlow_Add
     tiling=NoTile
     imageSize=<16 16>
     margin=<ltrb 0 0 0 0>
     outlineSettings {
      bUseBrushTransparency=false
      roundingType=HalfHeightRadius
      width=0
      color {colorUseRule=UseColor_Specified, specifiedColor=<rgba 0 0 0 0>}
      cornerRadii=<0 0 0 0>
     }
     tintColor {colorUseRule=UseColor_Specified, specifiedColor=<rgba 0.25 0.25 0.25 1>}
     uVRegion {bIsValid=false, max=<0 0>, min=<0 0>}
    }
    colorAndOpacity=<rgba 0 0.3465 0.9911 1>
    renderTransform {angle=0, scale=<0 0>, shear=<0 0>, translation=<0 0>}
   }
   slot {class=OverlaySlot, properties {horizontalAlignment=HAlign_Center, verticalAlignment=VAlign_Center}}
  }
  Image_textglow_2 {
   widgetClassPath=&Image
   parent=Overlay_4
   widget=Image_textglow_2
   properties {
    renderOpacity=0
    visibility=SelfHitTestInvisible
    brush {
     bIsDynamicallyLoaded=false
     drawAs=Image
     imageType=NoImage
     mirroring=NoMirror
     resourceName=None
     resourceObject=&M_UI_TextGlow_Add
     tiling=NoTile
     imageSize=<16 10>
     margin=<ltrb 0 0 0 0>
     outlineSettings {
      bUseBrushTransparency=false
      roundingType=HalfHeightRadius
      width=0
      color {colorUseRule=UseColor_Specified, specifiedColor=<rgba 0 0 0 0>}
      cornerRadii=<0 0 0 0>
     }
     tintColor {colorUseRule=UseColor_Specified, specifiedColor=<rgba 0.75 0.75 0.75 1>}
     uVRegion {bIsValid=false, max=<0 0>, min=<0 0>}
    }
    colorAndOpacity=<rgba 0 0.7053 0.9911 1>
    renderTransform {angle=7, scale=<1 1>, shear=<0 0>, translation=<0 0>}
   }
   slot {class=OverlaySlot, properties {horizontalAlignment=HAlign_Fill, verticalAlignment=VAlign_Fill, padding=<ltrb -40 2 -40 2>}}
  }
  Complete {
   widgetClassPath=&HorizontalBox
   parent=WidgetSwitcher_Complete
   widget=Complete
   slot {class=WidgetSwitcherSlot, properties {horizontalAlignment=HAlign_Right, verticalAlignment=VAlign_Center}}
  }
  SizeBox_2 {
   widgetClassPath=&SizeBox
   parent=Complete
   widget=SizeBox_2
   properties {bOverride_HeightOverride=true, bOverride_WidthOverride=true, heightOverride=24, minDesiredWidth=72, widthOverride=75}
   slot {class=HorizontalBoxSlot, properties {verticalAlignment=VAlign_Bottom}}
  }
  Overlay_2 {widgetClassPath=&Overlay, parent=SizeBox_2, widget=Overlay_2, slot {class=SizeBoxSlot}}
  Check {
   widgetClassPath=&Image
   bIsVariable=true
   parent=Overlay_2
   widget=Check
   properties {
    renderOpacity=0
    brush {
     bIsDynamicallyLoaded=false
     drawAs=Image
     imageType=NoImage
     mirroring=NoMirror
     resourceName=None
     resourceObject=/Game/UI/Foundation/Textures/Icons/Checks/T-Icon-Check-64
     tiling=NoTile
     imageSize=<60 60>
     margin=<ltrb 0 0 0 0>
     outlineSettings {
      bUseBrushTransparency=false
      roundingType=HalfHeightRadius
      width=0
      color {colorUseRule=UseColor_Specified, specifiedColor=<rgba 0 0 0 0>}
      cornerRadii=<0 0 0 0>
     }
     tintColor {colorUseRule=UseColor_Specified, specifiedColor=<rgba 1 1 1 1>}
     uVRegion {bIsValid=false, max=<0 0>, min=<0 0>}
    }
    colorAndOpacity=<rgba 0.0343 1 0.227 1>
    renderTransform {angle=0, scale=<1.0414 1.0414>, shear=<0 0>, translation=<5 -19>}
    renderTransformPivot=<0.5 0.8907>
   }
   slot {class=OverlaySlot, properties {horizontalAlignment=HAlign_Center, verticalAlignment=VAlign_Center, padding=<ltrb -10 -32 -10 -32>}}
  }
  Image_Glow {
   widgetClassPath=&Image
   bIsVariable=true
   parent=Overlay_2
   widget=Image_Glow
   properties {
    renderOpacity=0
    brush {
     bIsDynamicallyLoaded=false
     drawAs=Image
     imageType=NoImage
     mirroring=NoMirror
     resourceName=None
     resourceObject=&M_UI_TextGlow_Add
     tiling=NoTile
     imageSize=<130 130>
     margin=<ltrb 0 0 0 0>
     outlineSettings {
      bUseBrushTransparency=false
      roundingType=HalfHeightRadius
      width=0
      color {colorUseRule=UseColor_Specified, specifiedColor=<rgba 0 0 0 0>}
      cornerRadii=<0 0 0 0>
     }
     tintColor {colorUseRule=UseColor_Specified, specifiedColor=<rgba 0 1 0.1443 1>}
     uVRegion {bIsValid=false, max=<0 0>, min=<0 0>}
    }
   }
   slot {class=OverlaySlot, properties {horizontalAlignment=HAlign_Center, verticalAlignment=VAlign_Center, padding=<ltrb -100 -100 -100 -100>}}
  }
  Image_Spark {
   widgetClassPath=&Image
   bIsVariable=true
   parent=Overlay_2
   widget=Image_Spark
   properties {
    renderOpacity=0
    brush {
     bIsDynamicallyLoaded=false
     drawAs=Image
     imageType=NoImage
     mirroring=NoMirror
     resourceName=None
     resourceObject=/Game/UI/InGame/Quest/Materials/M_UI_SparkStar
     tiling=NoTile
     imageSize=<120 120>
     margin=<ltrb 0 0 0 0>
     outlineSettings {
      bUseBrushTransparency=false
      roundingType=HalfHeightRadius
      width=0
      color {colorUseRule=UseColor_Specified, specifiedColor=<rgba 0 0 0 0>}
      cornerRadii=<0 0 0 0>
     }
     tintColor {colorUseRule=UseColor_Specified, specifiedColor=<rgba 1 1 1 1>}
     uVRegion {bIsValid=false, max=<0 0>, min=<0 0>}
    }
   }
   slot {class=OverlaySlot, properties {horizontalAlignment=HAlign_Center, verticalAlignment=VAlign_Center, padding=<ltrb -75 -75 -75 -75>}}
  }
  SizeBox {
   widgetClassPath=&SizeBox
   parent=HorizontalBox_4
   widget=SizeBox
   properties {heightOverride=94, widthOverride=94}
   slot {class=HorizontalBoxSlot, properties {horizontalAlignment=HAlign_Right}}
  }
  Image_Icon {
   widgetClassPath=&FortLazyImage
   bIsVariable=true
   parent=SizeBox
   widget=Image_Icon
   properties {
    materialTextureParamName=NPC-Portrait
    brush {
     bIsDynamicallyLoaded=false
     drawAs=Image
     imageType=NoImage
     mirroring=NoMirror
     resourceName=None
     resourceObject=/Game/2dAssets/Sprays/Season28/PreviewImages/T-T-Sprays-Season28-PreviewImages-S28-Ranked-03
     tiling=NoTile
     imageSize=<128 128>
     margin=<ltrb 0.25 0.25 0.25 0.25>
     outlineSettings {
      bUseBrushTransparency=false
      roundingType=HalfHeightRadius
      width=0
      color {colorUseRule=UseColor_Specified, specifiedColor=<rgba 0 0 0 0>}
      cornerRadii=<0 0 0 0>
     }
     tintColor {colorUseRule=UseColor_Specified, specifiedColor=<rgba 1 1 1 1>}
     uVRegion {bIsValid=false, max=<0 0>, min=<0 0>}
    }
    loadingBackgroundBrush {
     bIsDynamicallyLoaded=false
     drawAs=Image
     imageType=NoImage
     mirroring=NoMirror
     resourceName=None
     resourceObject=/FortUILibrary/UIKit/Throbber/MI_UIKit_Monochrome_Throbber
     tiling=NoTile
     imageSize=<128 128>
     margin=<ltrb 0 0 0 0>
     outlineSettings {
      bUseBrushTransparency=false
      roundingType=HalfHeightRadius
      width=0
      color {colorUseRule=UseColor_Specified, specifiedColor=<rgba 0 0 0 0>}
      cornerRadii=<0 0 0 0>
     }
     tintColor {colorUseRule=UseColor_Specified, specifiedColor=<rgba 1 1 1 1>}
     uVRegion {bIsValid=false, max=<0 0>, min=<0 0>}
    }
   }
   slot {class=SizeBoxSlot, properties {horizontalAlignment=HAlign_Center, verticalAlignment=VAlign_Center}}
  }
  ImageTurnInStandin {
   widgetClassPath=&Image
   parent=Overlay_3
   widget=ImageTurnInStandin
   properties {
    renderOpacity=0
    visibility=SelfHitTestInvisible
    brush {
     bIsDynamicallyLoaded=false
     drawAs=Box
     imageType=NoImage
     mirroring=NoMirror
     resourceName=None
     resourceObject=/Game/Athena/HUD/MiniMap/Art/M_UI_Quest_Container_S19_Fill_White
     tiling=NoTile
     imageSize=<64 64>
     margin=<ltrb 0.5 0.5 0.5 0.5>
     outlineSettings {
      bUseBrushTransparency=false
      roundingType=HalfHeightRadius
      width=0
      color {colorUseRule=UseColor_Specified, specifiedColor=<rgba 0 0 0 0>}
      cornerRadii=<0 0 0 0>
     }
     tintColor {colorUseRule=UseColor_Specified, specifiedColor=<rgba 0.2122 0.6514 1 1>}
     uVRegion {bIsValid=false, max=<0 0>, min=<0 0>}
    }
   }
   slot {class=OverlaySlot, properties {horizontalAlignment=HAlign_Fill, verticalAlignment=VAlign_Fill}}
  }
  XpSubNotification {
   widgetClassPath=&/XpSubNotification
   bInherited=true
   bIsVariable=true
   parent=VerticalBox_1
   widget=XpSubNotification
   properties {visibility=Collapsed}
   slot {class=VerticalBoxSlot, properties {horizontalAlignment=HAlign_Center}}
  }
  HorizontalBoxHotkeyPrompt {
   widgetClassPath=&HorizontalBox
   bIsVariable=true
   parent=VerticalBox_1
   widget=HorizontalBoxHotkeyPrompt
   properties {visibility=Collapsed}
   slot {class=VerticalBoxSlot, properties {horizontalAlignment=HAlign_Center, verticalAlignment=VAlign_Bottom}}
  }
  SizeBox_Hotkey {
   widgetClassPath=&SizeBox
   parent=HorizontalBoxHotkeyPrompt
   widget=SizeBox_Hotkey
   properties {bOverride_HeightOverride=true, bOverride_WidthOverride=true, heightOverride=32, widthOverride=32}
   slot {class=HorizontalBoxSlot, properties {horizontalAlignment=HAlign_Center, verticalAlignment=VAlign_Bottom}}
  }
  ScaleBox_0 {widgetClassPath=&ScaleBox, parent=SizeBox_Hotkey, widget=ScaleBox_0, slot {class=SizeBoxSlot}}
  Overlay {widgetClassPath=&Overlay, parent=ScaleBox_0, widget=Overlay, slot {class=ScaleBoxSlot}}
  Gamepad {
   widgetClassPath=&CommonHardwareVisibilityBorder
   parent=Overlay
   widget=Gamepad
   properties {visibilityQuery {
     autoDescription=" ANY( Input.Gamepad )"
     tokenStreamVersion=0
     userDescription=""
     queryTokenStream [0,1,1,1,0]
     tagDictionary [{tagName=Input.Gamepad}]
    }}
   slot {class=OverlaySlot}
  }
  KeybindWidget {
   widgetClassPath=/FortUILibrary/Input/KeybindWidget
   bIsVariable=true
   parent=Gamepad
   widget=KeybindWidget
   properties {boundAction=GamepadToggleFullScreenMap, forcedHoldKeybindStatus=ForcedHold}
   slot {class=BorderSlot, properties {padding=<ltrb 0 0 0 0>}}
  }
  MouseKeyboard {
   widgetClassPath=&CommonHardwareVisibilityBorder
   parent=Overlay
   widget=MouseKeyboard
   properties {visibilityQuery {
     autoDescription=" ANY( Input.MouseAndKeyboard )"
     tokenStreamVersion=0
     userDescription=""
     queryTokenStream [0,1,1,1,0]
     tagDictionary [{tagName=Input.MouseAndKeyboard}]
    }}
   slot {class=OverlaySlot, properties {horizontalAlignment=HAlign_Fill, verticalAlignment=VAlign_Center}}
  }
  KeybindWidget_1 {
   widgetClassPath=&KeybindWidget
   bIsVariable=true
   parent=MouseKeyboard
   widget=KeybindWidget_1
   properties {boundAction=ToggleFullScreenMap, forcedHoldKeybindStatus=ForcedHold, boundActionFallbacks [GamepadToggleFullScreenMap]}
   slot {class=BorderSlot, properties {padding=<ltrb 0 0 0 0>}}
  }
  Visibility_MobileCollectionText {
   widgetClassPath=&CommonHardwareVisibilityBorder
   parent=HorizontalBoxHotkeyPrompt
   widget=Visibility_MobileCollectionText
   properties {visibilityQuery {
     autoDescription=" ANY( Input.Touch )"
     tokenStreamVersion=0
     userDescription=""
     queryTokenStream [0,1,1,1,0]
     tagDictionary [{tagName=Input.Touch}]
    }}
   slot {class=HorizontalBoxSlot, properties {horizontalAlignment=HAlign_Center, verticalAlignment=VAlign_Center}}
  }
  Switcher_MobileClueText {
   widgetClassPath=&CommonWidgetSwitcherLegacy
   bIsVariable=true
   parent=Visibility_MobileCollectionText
   widget=Switcher_MobileClueText
   properties {transitionDuration=0}
   slot {class=BorderSlot, properties {padding=<ltrb 0 0 0 0>}}
  }
  Text_MobileCollectionPrompt {
   widgetClassPath=&CommonTextBlock
   bIsVariable=true
   parent=Switcher_MobileClueText
   widget=Text_MobileCollectionPrompt
   properties {
    mobileFontSizeMultiplier=1.75
    text=TAP TO TRACK QUEST
    textTransformPolicy=ToUpper
    font {
     bForceMonospaced=false
     bMaterialIsStencil=false
     fontMaterial=null
     fontObject=&Burbank
     letterSpacing=0
     monospacedWidth=1
     size=12
     skewAmount=0
     typefaceFontName=Burbank-PlayerPower
     outlineSettings {
      bApplyOutlineToDropShadows=false
      bMiteredCorners=false
      bSeparateFillAlpha=false
      outlineMaterial=null
      outlineSize=0
      outlineColor=<rgba 0 0 0 1>
     }
    }
    renderTransform {angle=0, scale=<1 1>, shear=<-12 0>, translation=<0 0>}
    shadowOffset=<0 0>
   }
   slot {class=WidgetSwitcherSlot}
  }
  Visibility_CollectionText {
   widgetClassPath=&CommonHardwareVisibilityBorder
   parent=HorizontalBoxHotkeyPrompt
   widget=Visibility_CollectionText
   properties {
    padding=<ltrb 4 2 4 2>
    visibilityQuery {
     autoDescription=" ANY( Input.MouseAndKeyboard, Input.Gamepad )"
     tokenStreamVersion=0
     userDescription=""
     queryTokenStream [0,1,1,2,0,1]
     tagDictionary [{tagName=Input.MouseAndKeyboard}, {tagName=Input.Gamepad}]
    }
   }
   slot {class=HorizontalBoxSlot, properties {horizontalAlignment=HAlign_Center, verticalAlignment=VAlign_Center}}
  }
  Text_MapPrompt {
   widgetClassPath=&CommonTextBlock
   bIsVariable=true
   parent=Visibility_CollectionText
   widget=Text_MapPrompt
   properties {
    mobileFontSizeMultiplier=1.2
    text=Hold to Track Quest
    textTransformPolicy=ToUpper
    font {
     bForceMonospaced=false
     bMaterialIsStencil=false
     fontMaterial=null
     fontObject=&BurbankSmall
     letterSpacing=0
     monospacedWidth=1
     size=13
     skewAmount=0
     typefaceFontName=Black
     outlineSettings {
      bApplyOutlineToDropShadows=false
      bMiteredCorners=false
      bSeparateFillAlpha=false
      outlineMaterial=null
      outlineSize=2
      outlineColor=<rgba 0 0 0 0.125>
     }
    }
    shadowOffset=<0 0>
   }
   slot {class=BorderSlot}
  }
  Button_MobileClickCatcher {
   widgetClassPath=&Button
   bInherited=true
   parent=Overlay_5
   widget=Button_MobileClickCatcher
   properties {isFocusable=false, backgroundColor=<rgba 1 1 1 0>, colorAndOpacity=<rgba 1 1 1 0>}
   slot {class=OverlaySlot, properties {horizontalAlignment=HAlign_Fill, verticalAlignment=VAlign_Fill}}
  }
  Text_Header {widgetClassPath=&CommonTextBlock, bInherited=true, widget=null}
 ]
}
