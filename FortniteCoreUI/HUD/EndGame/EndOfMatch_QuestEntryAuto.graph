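# End-of-match quest entry widget in the Fortnite HUD, derived from the native EndOfMatchQuestEntryWidget_Auto class.
# Its only widget is an inherited RankedQuestNotification root, configured for end-of-match display and to prefer reward icons.
# PlayIntroAnimation casts the quest notification to RankedQuestNotification and forwards the call to that widget's own intro animation.
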WidgetBlueprint /FortniteCoreUI/HUD/EndGame/EndOfMatch_QuestEntryAuto
graphs [
 EventGraph {links [], nodes []}
 PlayIntroAnimation {nodes [
   PlayIntroAnimation() {is=title K2Node_FunctionEntry}
   > n1 = Cast To RankedQuestNotification!(execute, Object=Get QuestNotification() -> (QuestNotification:Fort Notification Entry Quest Object Reference)) -> (then, CastFailed, AsRanked Quest Notification:Ranked Quest Notification Object Reference) {is=title K2Node_DynamicCast}
   > PlayIntroAnimation(self:Ranked Quest Notification Object Reference=@n1) {is=title K2Node_CallFunction}
  ]}
]
variables []
widgets {
 info {
  inheritedWidgetCount=1
  parentClass=/Script/FortniteUI.EndOfMatchQuestEntryWidget_Auto
  rootWidgetClass=/FortniteCoreUI/UI/InGame/EventNotifications/RankedQuestNotification.RankedQuestNotification_C
  widgetCount=1
 }
 widgets [QuestNotification {
   widgetClassPath=/FortniteCoreUI/UI/InGame/EventNotifications/RankedQuestNotification
   bInherited=true
   bIsVariable=true
   widget=QuestNotification
   properties {bIsEndOfMatch=true, bPreferRewardIcon=true}
  }]
}
