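# Quest notification widget, built on FortNotificationEntry_Quest, showing quest progress, completion, timer, assists and a track-quest prompt that follows input type.
# The intro leads to a fill animation, or for community quests a "+N" count-up via animated CountLerp.
# Completed quests then play QuestComplete and can reveal the next staged objective; a rushed intro skips progress animations.
# Icons and backgrounds load asynchronously, with dynamic materials copying category colors for non-community quests.
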WidgetBlueprint /FortniteCoreUI/UI/InGame/EventNotifications/QuestNotification
animations [
 FillProgress {
  binding_count=5
  length_seconds=0.5008
  object=&self:FillProgress
  sequence {bindings [
    Image_textglow {
     id=#1
     possessed_class=Image
     tracks [
      {
       type=MovieScene2DTransformTrack
       sections [{
         type=MovieScene2DTransformSection
         channels [
          Translation.X {type=MovieSceneScriptingFloatChannel}
          Translation.Y {type=MovieSceneScriptingFloatChannel}
          Angle {type=MovieSceneScriptingFloatChannel}
          Scale.X {type=MovieSceneScriptingFloatChannel, keys [[0,1],[10.0153,4]]}
          Scale.Y {type=MovieSceneScriptingFloatChannel, keys [[0,1],[10.0153,4]]}
          Shear.X {type=MovieSceneScriptingFloatChannel}
          Shear.Y {type=MovieSceneScriptingFloatChannel}
         ]
        }]
      }
      {
       type=MovieSceneFloatTrack
       display_name=Render Opacity
       sections [{type=MovieSceneFloatSection, channels [None {type=MovieSceneScriptingFloatChannel, keys [[0,0],[2,1],[10.0153,0]]}]}]
      }
     ]
    }
    RichText_ObjectivesCompletedOLD {
     id=#2
     display_name=RichText_ObjectivesCompleted
     possessed_class=RichTextBlock
     tracks [{
       type=MovieScene2DTransformTrack
       sections [{
         type=MovieScene2DTransformSection
         end_frame=10
         start_frame=0
         channels [
          Translation.X {type=MovieSceneScriptingFloatChannel}
          Translation.Y {type=MovieSceneScriptingFloatChannel}
          Angle {type=MovieSceneScriptingFloatChannel}
          Scale.X {type=MovieSceneScriptingFloatChannel, keys [[0,1],[2,1.05],[10,1]]}
          Scale.Y {type=MovieSceneScriptingFloatChannel, keys [[0,1],[2,1.05],[10,1]]}
          Shear.X {type=MovieSceneScriptingFloatChannel}
          Shear.Y {type=MovieSceneScriptingFloatChannel}
         ]
        }]
      }]
    }
    HorizontalBoxHotkeyPrompt {
     id=#3
     possessed_class=HorizontalBox
     tracks [{
       type=MovieSceneFloatTrack
       display_name=Render Opacity
       sections [{type=MovieSceneFloatSection, end_frame=0, start_frame=0, channels [None {type=MovieSceneScriptingFloatChannel, const [1,1]}]}]
      }]
    }
    SizeBox_main {
     id=#4
     possessed_class=SizeBox
     tracks [{
       type=MovieSceneFloatTrack
       display_name=Render Opacity
       sections [{type=MovieSceneFloatSection, end_frame=0, start_frame=0, channels [None {type=MovieSceneScriptingFloatChannel, const [1,1]}]}]
      }]
    }
    Image_textglow_2 {
     id=#5
     possessed_class=Image
     tracks [
      {
       type=MovieScene2DTransformTrack
       sections [{
         type=MovieScene2DTransformSection
         end_frame=10
         start_frame=0
         channels [
          Angle {type=MovieSceneScriptingFloatChannel, keys [[0,-2],[10,7]]}
          Scale.X {type=MovieSceneScriptingFloatChannel, keys [[0,0.14],[10,1]]}
          Scale.Y {type=MovieSceneScriptingFloatChannel, keys [[0,0.5],[10,1]]}
         ]
        }]
      }
      {
       type=MovieSceneFloatTrack
       display_name=Render Opacity
       sections [{
         type=MovieSceneFloatSection
         end_frame=10
         start_frame=0
         channels [None {type=MovieSceneScriptingFloatChannel, keys [[0,0],[2,0.25],[10.0153,0]]}]
        }]
      }
     ]
    }
   ]}
 }
 QuestComplete {
  binding_count=7
  length_seconds=0.6
  object=&self:QuestComplete
  sequence {bindings [
    Image_Check {
     id=#6
     possessed_class=Image
     tracks [
      {
       type=MovieSceneFloatTrack
       display_name=Render Opacity
       sections [{type=MovieSceneFloatSection, channels [None {type=MovieSceneScriptingFloatChannel, keys [[0,0],[8,1]]}]}]
      }
      {
       type=MovieScene2DTransformTrack
       sections [{
         type=MovieScene2DTransformSection
         channels [
          Translation.X {type=MovieSceneScriptingFloatChannel}
          Angle {type=MovieSceneScriptingFloatChannel}
          Scale.X {type=MovieSceneScriptingFloatChannel, keys [[0,2.3531],[6,0.45],[8,0.6]]}
          Scale.Y {type=MovieSceneScriptingFloatChannel, keys [[0,2.3531],[6,0.45],[8,0.6]]}
          Shear.X {type=MovieSceneScriptingFloatChannel}
          Shear.Y {type=MovieSceneScriptingFloatChannel}
         ]
        }]
      }
     ]
    }
    Image_Glow {
     id=#7
     possessed_class=Image
     tracks [
      {
       type=MovieSceneFloatTrack
       display_name=Render Opacity
       sections [{type=MovieSceneFloatSection, channels [None {type=MovieSceneScriptingFloatChannel, keys [[6,0],[8,1],[12,0]]}]}]
      }
      {
       type=MovieScene2DTransformTrack
       sections [{
         type=MovieScene2DTransformSection
         channels [
          Translation.X {type=MovieSceneScriptingFloatChannel}
          Angle {type=MovieSceneScriptingFloatChannel}
          Scale.X {type=MovieSceneScriptingFloatChannel, keys [[6,1],[12,1.3]]}
          Scale.Y {type=MovieSceneScriptingFloatChannel, keys [[6,1],[12,1.3]]}
          Shear.X {type=MovieSceneScriptingFloatChannel}
          Shear.Y {type=MovieSceneScriptingFloatChannel}
         ]
        }]
      }
     ]
    }
    Text_FlavorText {
     id=#8
     possessed_class=CommonTextBlock
     tracks [{
       type=MovieSceneByteTrack
       display_name=Visibility
       sections [{type=MovieSceneByteSection, channels [None {type=MovieSceneScriptingByteChannel, const [3,1]}]}]
      }]
    }
    Image_Spark {
     id=#9
     possessed_class=Image
     tracks [
      {
       type=MovieSceneFloatTrack
       display_name=Render Opacity
       sections [{type=MovieSceneFloatSection, channels [None {type=MovieSceneScriptingFloatChannel, keys [[6,0],[8,1],[12,0]]}]}]
      }
      {
       type=MovieScene2DTransformTrack
       sections [{
         type=MovieScene2DTransformSection
         channels [
          Translation.X {type=MovieSceneScriptingFloatChannel}
          Angle {type=MovieSceneScriptingFloatChannel}
          Scale.X {type=MovieSceneScriptingFloatChannel, keys [[6,0.75],[12,1]]}
          Scale.Y {type=MovieSceneScriptingFloatChannel, keys [[6,0.75],[12,1]]}
          Shear.X {type=MovieSceneScriptingFloatChannel}
          Shear.Y {type=MovieSceneScriptingFloatChannel}
         ]
        }]
      }
     ]
    }
    WidgetSwitcher_Complete {
     id=#10
     possessed_class=WidgetSwitcher
     tracks [{
       type=MovieSceneIntegerTrack
       display_name=Active Widget Index
       sections [{type=MovieSceneIntegerSection, channels [None {type=MovieSceneScriptingIntegerChannel, const [1,1]}]}]
      }]
    }
    HorizontalBoxHotkeyPrompt {
     id=#11
     possessed_class=HorizontalBox
     tracks [{
       type=MovieSceneFloatTrack
       display_name=Render Opacity
       sections [{type=MovieSceneFloatSection, end_frame=0, start_frame=0, channels [None {type=MovieSceneScriptingFloatChannel, const [1,1]}]}]
      }]
    }
    SizeBox_main {
     id=#12
     possessed_class=SizeBox
     tracks [{
       type=MovieSceneFloatTrack
       display_name=Render Opacity
       sections [{type=MovieSceneFloatSection, end_frame=0, start_frame=0, channels [None {type=MovieSceneScriptingFloatChannel, const [1,1]}]}]
      }]
    }
   ]}
 }
 Outro {
  binding_count=3
  length_seconds=0.25
  object=&self:Outro
  sequence {bindings [
    SizeBox_main {
     id=#13
     possessed_class=SizeBox
     tracks [{
       type=MovieSceneFloatTrack
       display_name=Render Opacity
       sections [{type=MovieSceneFloatSection, channels [None {type=MovieSceneScriptingFloatChannel, keys [[0,1],[5,0]]}]}]
      }]
    }
    HorizontalBoxHotkeyPrompt {
     id=#14
     possessed_class=HorizontalBox
     tracks [{
       type=MovieSceneFloatTrack
       display_name=Render Opacity
       sections [{type=MovieSceneFloatSection, channels [None {type=MovieSceneScriptingFloatChannel, keys [[0,1],[5,0]]}]}]
      }]
    }
    Image_Icon_CQ {
     id=#15
     possessed_class=Image
     tracks [{
       type=MovieSceneFloatTrack
       display_name=Render Opacity
       sections [{type=MovieSceneFloatSection, end_frame=5, start_frame=0, channels [None {type=MovieSceneScriptingFloatChannel, keys [[0,1],[4,0],[5,0]]}]}]
      }]
    }
   ]}
 }
 AddProgress {
  binding_count=4
  length_seconds=0.95
  object=&self:AddProgress
  sequence {bindings [
    SizeBox_main {
     id=#4
     possessed_class=SizeBox
     tracks [{
       type=MovieSceneFloatTrack
       display_name=Render Opacity
       sections [{type=MovieSceneFloatSection, end_frame=0, start_frame=0, channels [None {type=MovieSceneScriptingFloatChannel, const [1,1]}]}]
      }]
    }
    RichText_Objectives_Current {
     id=#16
     possessed_class=CommonRichTextBlock
     tracks [
      {
       type=MovieSceneFloatTrack
       display_name=Wrap Text At
       sections [{type=MovieSceneFloatSection, end_frame=4, start_frame=4, channels [None {type=MovieSceneScriptingFloatChannel, const [0,1]}]}]
      }
      {
       type=MovieSceneFloatTrack
       display_name=Render Opacity
       sections [{
         type=MovieSceneFloatSection
         end_frame=18
         start_frame=0
         channels [None {type=MovieSceneScriptingFloatChannel, keys [[0,0],[9,0],[16.7143,1]]}]
        }]
      }
     ]
    }
    RichText_ObjectiveAdded {
     id=#17
     possessed_class=CommonRichTextBlock
     tracks [
      {
       type=MovieSceneFloatTrack
       display_name=Wrap Text At
       sections [{type=MovieSceneFloatSection, end_frame=0, start_frame=0, channels [None {type=MovieSceneScriptingFloatChannel, const [0,1]}]}]
      }
      {
       type=MovieSceneFloatTrack
       display_name=Render Opacity
       sections [{
         type=MovieSceneFloatSection
         end_frame=18
         start_frame=-1
         channels [None {type=MovieSceneScriptingFloatChannel, keys [[-1,0],[5.3333,1],[16.7333,1],[18,0]]}]
        }]
      }
      {
       type=MovieScene2DTransformTrack
       sections [{
         type=MovieScene2DTransformSection
         end_frame=18
         start_frame=-1
         channels [
          Translation.X {type=MovieSceneScriptingFloatChannel, keys [[-1,-15],[9.1333,-15],[18,60]]}
          Angle {type=MovieSceneScriptingFloatChannel}
          Scale.X {type=MovieSceneScriptingFloatChannel, keys [[-1,0],[4.0667,1.3],[7.8667,0.9],[9.1333,1],[18,0]]}
          Scale.Y {type=MovieSceneScriptingFloatChannel, keys [[-1,0],[4.0667,1.3],[7.8667,0.9],[9.1333,1],[18,0]]}
          Shear.X {type=MovieSceneScriptingFloatChannel}
          Shear.Y {type=MovieSceneScriptingFloatChannel}
         ]
        }]
      }
     ]
    }
    Image_Icon_CQ {
     id=#18
     possessed_class=Image
     tracks [{
       type=MovieSceneFloatTrack
       display_name=Render Opacity
       sections [{type=MovieSceneFloatSection, end_frame=0, start_frame=-1, channels [None {type=MovieSceneScriptingFloatChannel, const [1,1]}]}]
      }]
    }
   ]}
 }
 CountProgress {
  binding_count=2
  length_seconds=1
  object=&self:CountProgress
  sequence {bindings [
    RichText_Objectives_Current {
     id=#19
     possessed_class=CommonRichTextBlock
     tracks [{
       type=MovieSceneFloatTrack
       display_name=Render Opacity
       sections [{type=MovieSceneFloatSection, end_frame=0, start_frame=0, channels [None {type=MovieSceneScriptingFloatChannel, const [1,1]}]}]
      }]
    }
    QuestNotification_C_30 {
     id=#20
     possessed_class=QuestNotification_C
     tracks [{
       type=MovieSceneDoubleTrack
       display_name=Count Lerp
       sections [{type=MovieSceneDoubleSection, end_frame=20, start_frame=0, channels [None {type=MovieSceneScriptingDoubleChannel, keys [[0,0],[20,1]]}]}]
      }]
    }
   ]}
 }
]
graphs [
 AddedProgress {nodes [
   n1 = AddedProgress!() -> (then, StartValue:int, TargetValue:int, ProgressTime:double=~) {is=title K2Node_FunctionEntry}
   n2 = Lerp(A="To Float (Integer)"(InInt=@n1), B="To Float (Integer)"(InInt=@n1.TargetValue), Alpha=@n1.ProgressTime)
   > Return Node!(execute, TextProgress:Text="To Text (Integer)"(Value=Round(A=@n2) {is=title K2Node_CallFunction}))
  ]}
 EventGraph {nodes [
   "On Clicked (Button_MobileClickCatcher)"() {is=title K2Node_ComponentBoundEvent}
   > TrackCurrentQuest(self:Fort Notification Entry Quest Object Reference)
   EventFillComplete() {is=title K2Node_CustomEvent}
   > n1 = Branch(Condition="integer >= integer"!(A:int=Get CurrentCount() -> (CurrentCount:int=0), B:int=Get GoalCount() -> (GoalCount:int=0)) -> (ReturnValue:bool) {is=title K2Node_CallFunction})
   > PlayAnimation(InAnimation=Get QuestComplete() -> (QuestComplete:Widget Animation Object Reference))
   FinishNotificationIntro(execute=@n1.else, self:Fort Notification Entry Object Reference)
   n2 = EventCompleteAnim() {is=title K2Node_CustomEvent}
   > n3 = Branch(Condition=Get bNeedsStagedNotification() -> (bNeedsStagedNotification:bool))
   > TurnInNotification()
   FinishNotificationIntro(execute=@n3.else)
   EventReset() {is=title K2Node_CustomEvent}
   > SetRenderOpacity(
    self=[
     Get Image_Check() -> (Image_Check:Image Object Reference)
     Get Image_Glow() -> (Image_Glow:Image Object Reference)
     Get Image_Spark() -> (Image_Spark:Image Object Reference)
    ]
    InOpacity=0
   )
   > n4 = GetQuestOwnedGameplayTags!(execute, self:Fort Notification Entry Quest Object Reference) -> (then, ReturnValue:Gameplay Tag Container Structure)
   > SetVisibility(
    self=Get Overlay_Progress() -> (Overlay_Progress:Overlay Object Reference)
    InVisibility=Select(
     Option 0=HitTestInvisible
     Option 1=Collapsed
     Index=HasTag!(
      TagContainer:"Gameplay Tag Container Structure (by ref)"=@n4
      Tag:Gameplay Tag Structure=(TagName="Athena.Quest.HideCount")
      bExactMatch:bool=false
     ) -> (ReturnValue:bool)
    )
   )
   > n5 = Branch(Condition="integer >= integer"(A=Get CurrentCount() -> (CurrentCount=0), B=Get GoalCount() -> (GoalCount=0)))
   n6 = Get WidgetSwitcher_Complete() -> (WidgetSwitcher_Complete:Widget Switcher Object Reference)
   > SetActiveWidget(self:Widget Switcher Object Reference=@n6, Widget=Get SizeBox_Complete() -> (SizeBox_Complete:Size Box Object Reference))
   > n7 = SetVisibility(
    self=Get Spacer_Flavor() -> (Spacer_Flavor:Spacer Object Reference)
    InVisibility=Select(Option 0=SelfHitTestInvisible, Option 1=Collapsed, Index=Get IsCQ() -> (IsCQ:bool))
   )
   SetVisibility(execute=[@n7.then, @n10.then], self=Get Text_FlavorText() -> (Text_FlavorText:Common Text Object Reference), InVisibility=Hidden)
   n8 = Get Image_Icon_CQ() -> (Image_Icon_CQ:Image Object Reference)
   > SetVisibility(self=@n8, InVisibility=Collapsed)
   SetVisibility(execute=@n5.else, self=Get Spacer_Flavor(), InVisibility=Collapsed)
   > n9 = Branch(Condition=Get IsCQ())
   > SetActiveWidget(self=@n6, Widget=Get Overlay_AddedProgress() -> (Overlay_AddedProgress:Overlay Object Reference))
   > SetVisibility(self=@n8, InVisibility=HitTestInvisible)
   n10 = SetActiveWidget(execute=@n9.else, self=@n6, Widget=Get Overlay_Progress())
   n11 = EventOnStartComplete() {is=title K2Node_CustomEvent}
   > PlaySound2D(Sound=/Game/Sounds/Fort_UI_Sounds/S15/Quest_Complete_Cue)
   n12 = HandleOnInputMethodChanged!() -> (OutputDelegate, then, bNewInputType:ECommonInputType Enum) {is=title K2Node_CustomEvent}
   > SetVisibility(
    self=Get KeybindWidget() -> (KeybindWidget:Keybind Widget Object Reference)
    InVisibility=Select!(
     MouseAndKeyboard:ESlateVisibility Enum=SelfHitTestInvisible
     Gamepad:ESlateVisibility Enum=SelfHitTestInvisible
     Touch:ESlateVisibility Enum=Visible
     Count:ESlateVisibility Enum=Collapsed
     Index:ECommonInputType Enum=@n12.bNewInputType
    ) -> (ReturnValue)
   )
   > SetBoundAction!(
    execute
    self:Fort Keybind Widget Object Reference=Get KeybindWidget()
    NewBoundAction:Name=Select(
     MouseAndKeyboard:Name=ToggleFullScreenMap
     Gamepad:Name=GamepadToggleFullScreenMap
     Touch:Name=None
     Count:Name=None
     Index=@n12.bNewInputType
    ) -> (ReturnValue:Name)
   ) -> (then)
   n13 = Get Text_Map_Prompt() -> (Text_Map_Prompt:Text)
   > Set Text(
    Text=Select(
     MouseAndKeyboard:Text=@n13
     Gamepad:Text=@n13
     Touch:Text=Get Text_Mobile_Prompt() -> (Text_Mobile_Prompt:Text)
     Count:Text
     Index=@n12.bNewInputType
    ) -> (ReturnValue:Text)
    self=Get Text_MapPrompt() -> (Text_MapPrompt:Common Text Object Reference)
   )
   TurnInNotification() {is=title K2Node_CustomEvent}
   > RetriggerableDelay(Duration=1)
   > ShowNextStagedObjectiveInformation(self:Fort Notification Entry Quest Object Reference)
   > EventReset() {is=title K2Node_CallFunction}
   > RetriggerableDelay(Duration=4)
   > FinishNotificationIntro()
   n14 = LoadTextureForDefaultMaterial!() -> (OutputDelegate, then, Asset:Soft Object Reference) {is=title K2Node_CustomEvent}
   > SetBrushFromMaterial(
    self=Get Image_Main_CategoryBG() -> (Image_Main_CategoryBG:Image Object Reference)
    Material=Get DefaultIconMaterial() -> (DefaultIconMaterial:Material Instance Object Reference)
   )
   > SetIconParameters()
   > n15 = Async Load Asset(Asset=@n14.Asset)
   n16 = Cast To Texture2D(execute=@n15.Completed, Object=@n15)
   > n17 = GetDynamicMaterial(self=Get Image_Main_CategoryBG())
   > SetScalarParameterValue(self=@n17, ParameterName=ShowItemImage, Value=1)
   > SetTextureParameterValue(self=@n17, ParameterName=Texture, Value=@n16)
   n18 = Cast To MaterialInterface!(execute=@n16.CastFailed, Object=@n15) -> (then, CastFailed, AsMaterial Interface:Material Interface Object Reference) {is=title K2Node_DynamicCast}
   > SetBrushFromMaterial(self=Get Image_Main_CategoryBG(), Material=@n18)
   > n19 = GetDynamicMaterial(self=Get Image_Main_CategoryBG())
   > SetScalarParameterValue(self=@n19, ParameterName=IsNotification, Value=1)
   > SetScalarParameterValue(self=@n19, ParameterName=IsAnimated, Value=1)
   > SetIconParameters()
   n20 = Get Image_Assisted_BG() -> (Image_Assisted_BG:Image Object Reference)
   > SetBrushFromMaterial(self=@n20, Material=@n18)
   > n21 = GetDynamicMaterial(self=@n20)
   > SetScalarParameterValue(self=@n21, ParameterName=ShowItemImage, Value=0)
   n22 = LoadMaterial!() -> (OutputDelegate, then, SoftMaterial:Material Interface Soft Object Reference) {is=title K2Node_CustomEvent}
   > n23 = Async Load Asset(Asset=@n22.SoftMaterial)
   n24 = Cast To MaterialInterface(execute=@n23.Completed, Object=@n23)
   > SetBrushFromMaterial(self=Get Image_Icon() -> (Image_Icon:Fort Lazy Image Object Reference), Material=@n24)
   > SetIconParameters()
   n25 = Event PreConstruct()
   > Parent: PreConstruct(IsDesignTime=@n25.IsDesignTime)
   > SetWidthOverride(self=Get SizeBox_main() -> (SizeBox_main:Size Box Object Reference), InWidthOverride=Get In Width Override() -> (In Width Override:double))
   > n26 = Branch(Condition="float > float"(A=Get TextMaxHeightOverride() -> (TextMaxHeightOverride:double)))
   n27 = Get SizeBox_Description() -> (SizeBox_Description:Size Box Object Reference)
   > n28 = SetMaxDesiredHeight!(execute, self:Size Box Object Reference=@n27, InMaxDesiredHeight:float=Get TextMaxHeightOverride()) -> (then)
   SetBrushSize!(execute=[@n28.then, @n29.then], self:Fort Keybind Widget Object Reference=Get KeybindWidget(), InBrushSize:EFortBrushSize Enum=VeryVerySmall) -> (then)
   n29 = ClearMaxDesiredHeight(execute=@n26.else, self:Size Box Object Reference=@n27)
   Event Construct()
   > Parent: Construct()
   n30 = Get CommonInputSubsystem()
   > Bind Event to On Input Method Changed(self=@n30, Delegate=Create Event())
   > HandleOnInputMethodChanged!(execute, self, bNewInputType:ECommonInputType Enum=GetCurrentInputType(self=@n30)) -> (then) {is=title K2Node_CallFunction}
   > SetMobileVisuals()
   Event Destruct()
   > Parent: Destruct()
   > Unbind Event from On Input Method Changed(self=Get CommonInputSubsystem(), Delegate=Create Event())
   n31 = Event OnUpdateNotificationBackground!() -> (OutputDelegate, then, AssetToDisplay:"Soft Object Reference (by ref)", bUseRarityContainer:bool, bIsDailyPunchcardQuest:bool) {is=title K2Node_Event}
   > n32 = Branch(Condition=@n31.bIsDailyPunchcardQuest)
   > n33 = LoadMaterial!(
    execute
    self
    SoftMaterial:Material Interface Soft Object Reference=Get PunchcardSoftMaterial() -> (PunchcardSoftMaterial:Material Interface Soft Object Reference)
   ) -> (then) {is=title K2Node_CallFunction}
   SetVisibility(
    execute=[@n33.then, @n53.then]
    InVisibility=Select!(Option 0:ESlateVisibility Enum=Hidden, Option 1:ESlateVisibility Enum=SelfHitTestInvisible, Index:bool=@n31.bUseRarityContainer) -> (ReturnValue:ESlateVisibility Enum)
   )
   > n34 = GetQuestGiverIcon!(execute, self:Fort Notification Entry Quest Object Reference) -> (then, ReturnValue:Soft Object Reference)
   > n35 = GetQuestOwnedGameplayTags()
   n36 = Get CurrentCount() -> (CurrentCount=0)
   > Set IsCQ(IsCQ:bool=AND Boolean(A=HasTag(TagContainer=@n35, Tag=(TagName="Athena.Quest.Community.LocalProgress"), bExactMatch=true), B="integer > integer"(A=@n36))) -> (Output_Get:bool)
   n37 = IsAssetNull!(Asset:Soft Object Reference=@n34) -> (ReturnValue:bool)
   > SetVisibility(self=Get Image_Icon(), InVisibility=Select(Option 0=HitTestInvisible, Option 1=Collapsed, Index=@n37))
   > SetVisibility(
    self=Get Spacer_Story() -> (Spacer_Story:Spacer Object Reference)
    InVisibility=Select(Option 0=Collapsed, Option 1=HitTestInvisible, Index=@n37)
   )
   > n38 = Branch(Condition=@n37)
   n39 = Get Switcher_MainBG() -> (Switcher_MainBG:Common Visibility Switcher Object Reference)
   > SetActiveWidgetIndex(self:Common Visibility Switcher Object Reference=@n39, Index=1)
   > SetBrush(self=Get Image_Assisted_BG(), InBrush=Get Brush!(self:Image Object Reference=Get Image_Main_CategoryBG()) -> (Brush:Slate Brush Structure))
   SetActiveWidgetIndex(execute=@n38.else, self=@n39, Index=0)
   > n40 = Async Load Asset(Asset=@n34)
   n41 = Cast To MaterialInstance!(execute=@n40.Completed, Object=@n40) -> (then, CastFailed, AsMaterial Instance:Material Instance Object Reference) {is=title K2Node_DynamicCast}
   > n42 = SetBrushFromMaterial(self=Get Image_Icon(), Material=@n41)
   n43 = Get IsCQ()
   SetBrushFromMaterial(
    execute=[@n42.then, @n52.then]
    self=Get Image_Main_DefaultBG() -> (Image_Main_DefaultBG:Image Object Reference)
    Material=Select(
     Option 0:Material Interface Object Reference=Get Default_BG_Material() -> (Default_BG_Material:Material Instance Object Reference)
     Option 1:Material Interface Object Reference=Get CQ_Material() -> (CQ_Material:Material Instance Object Reference)
     Index=@n43
    ) -> (ReturnValue:Material Interface Object Reference)
   )
   > SetBrushFromMaterial(
    self:Border Object Reference=Get Border_Icon() -> (Border_Icon:Common Border Object Reference)
    Material=Select(
     Option 0=Get CQ_Material_DefaultPortrait_BG() -> (CQ_Material_DefaultPortrait_BG:Material Instance Object Reference)
     Option 1=Get CQ_Material_Portrait_BG() -> (CQ_Material_Portrait_BG:Material Instance Object Reference)
     Index=@n43
    )
   )
   > n44 = Branch(Condition=@n43)
   n45 = GetDynamicMaterial(execute=@n44.else, self=Get Image_Main_CategoryBG())
   > n46 = Get Vector Parameter Value!(execute, self:Material Instance Dynamic Object Reference=@n45, ParameterName:Name=BackgroundColor) -> (then, ReturnValue:Linear Color Structure) {is=title K2Node_CallFunction}
   > n47 = Get Vector Parameter Value(self=@n45, ParameterName=ColorSecond)
   > n48 = GetDynamicMaterial(self=Get Image_Main_DefaultBG())
   > n49 = GetDynamicMaterial(self:Border Object Reference=Get Border_Icon())
   > SetVectorParameterValue(self=[@n48, @n49], ParameterName=RGB1, Value=@n46)
   > SetVectorParameterValue(self=[@n48, @n49], ParameterName=RGB2, Value=@n47)
   n50 = Cast To Texture!(execute=@n41.CastFailed, Object=@n40) -> (then, CastFailed, AsTexture:Texture Object Reference) {is=title K2Node_DynamicCast}
   > n51 = GetDynamicMaterial(self:Image Object Reference=Get Image_Icon())
   > n52 = SetTextureParameterValue(self=@n51, ParameterName=Item, Value=@n50)
   n53 = LoadTextureForDefaultMaterial!(execute=@n32.else, self, Asset:Soft Object Reference=@n31.AssetToDisplay) -> (then) {is=title K2Node_CallFunction}
   n54 = Event OnEnableCompleteFlavorText!() -> (OutputDelegate, then, IsEnabled:bool) {is=title K2Node_Event}
   > SetVisibility(self=Get Spacer_Flavor(), InVisibility=Collapsed)
   > n55 = Branch(Condition=@n54.IsEnabled)
   n56 = Get WidgetSwitcher_Complete()
   > SetActiveWidgetIndex(self:Widget Switcher Object Reference=@n56, Index=1)
   n57 = Get Text_FlavorText()
   > SetVisibility(self=@n57, InVisibility=HitTestInvisible)
   SetActiveWidgetIndex(execute=@n55.else, self=@n56, Index=0)
   > SetVisibility(self=@n57, InVisibility=Hidden)
   n58 = Event OnPlayRushedNotificationIntro!() -> (OutputDelegate, then, Rate:float) {is=title K2Node_Event}
   > EventReset()
   n59 = "Max (Float)"(A=@n58.Rate, B=1)
   > PlayAnimationByName!(
    execute
    self:Fort Simple Widget Animations Panel Object Reference=Get SimpleAnimations() -> (SimpleAnimations:Fort Simple Widget Animations Panel Object Reference)
    Name:"Name (by ref)"=Get SimpleAnim_Intro() -> (SimpleAnim_Intro:Name=None)
    Speed:float=@n59
    StartAtTime:float=0
   ) -> (then, ReturnValue:bool)
   > PlaySound2D(Sound=/Game/Sounds/Fort_UI_Sounds/S15/Quest_Notification_Popup_Cue)
   > RetriggerableDelay(Duration=float / float(A=Get SimpleAnim_IntroDuration() -> (SimpleAnim_IntroDuration:float), B=@n59))
   > n60 = Branch(Condition="integer >= integer"(A=Get CurrentCount() -> (CurrentCount=0), B=Get GoalCount() -> (GoalCount=0)))
   n61 = Get QuestComplete()
   > PlayAnimation(InAnimation=@n61, PlaybackSpeed=@n59)
   > RetriggerableDelay(Duration=GetEndTime(self=@n61))
   > FinishNotificationOutro(self:Fort Notification Entry Object Reference)
   FinishNotificationOutro(execute=@n60.else)
   n62 = Event SetAssistData!() -> (OutputDelegate, then, bWasAssisted:bool, OtherPlayersName:String) {is=title K2Node_Event}
   > n63 = Branch()
   n64 = Get Text_AssistBlock() -> (Text_AssistBlock:Common Rich Text Block Object Reference)
   > SetText(
    self=@n64
    InText=Format Text!(Format="LOCTEXT(\"<text case=\\\"upper\\\">Assisted by</> {PlayerName}\")", PlayerName:String=@n62.OtherPlayersName) -> (Result)
   )
   SetText(
    execute=@n63.else
    self=@n64
    InText=Format Text(Format="LOCTEXT(\"<text case=\\\"upper\\\">Assisted</> {PlayerName}\")", PlayerName=@n62.OtherPlayersName)
   )
   Event StartNotificationIntro() {is=title K2Node_Event}
   > EventReset()
   > PlayAnimationByName(self=Get SimpleAnimations(), Name=Get SimpleAnim_Intro() -> (SimpleAnim_Intro=None), Speed=1, StartAtTime=0)
   > PlaySound2D(Sound=&Quest_Notification_Popup_Cue)
   > RetriggerableDelay(Duration=Get SimpleAnim_IntroDuration())
   > n65 = Branch(Condition="integer >= integer"(A=Get CurrentCount() -> (CurrentCount=0), B=Get GoalCount() -> (GoalCount=0)))
   n66 = Play Animation with Finished event!(
    execute:Exec=[@n67.else, @n65.then]
    Widget:User Widget Object Reference=Self-Reference()
    InAnimation:Widget Animation Object Reference=Get FillProgress() -> (FillProgress:Widget Animation Object Reference)
    StartAtTime:float=0
    NumLoopsToPlay:int=1
    PlayMode:EUMGSequencePlayMode Enum
    PlaybackSpeed:float=1
   ) -> (then:Exec, Result:Widget Animation Handle Structure, Finished:Exec) {is=title K2Node_PlayAnimation2}
   Call Progress Completed!(execute:Exec=@n66.Finished, self:Quest Notification Object Reference) -> (then:Exec) {is=title K2Node_CallDelegate}
   n67 = Branch(execute=@n65.else, Condition=Get IsCQ())
   n68 = Get PreviousCount() -> (PreviousCount:int=0)
   n69 = int - int(A=Get CurrentCount() -> (CurrentCount=0), B=@n68)
   > Set Text(
    Text=Format Text!(Format="LOCTEXT(\"+ {added}\")", added:Text="To Text (Integer)"(Value=@n69)) -> (Result)
    self:Rich Text Block Object Reference=Get RichText_ObjectiveAdded() -> (RichText_ObjectiveAdded:Common Rich Text Block Object Reference)
   )
   n70 = Get RichText_Objectives_Current() -> (RichText_Objectives_Current:Common Rich Text Block Object Reference)
   > Set Text(Text="To Text (Integer)"(Value=@n68), self=@n70)
   > Set RenderOpacity!(execute, RenderOpacity:float=0, self:Widget Object Reference=@n70) -> (then, Output_Get:float)
   > StopAnimation(InAnimation=Get CountProgress() -> (CountProgress:Widget Animation Object Reference))
   > n71 = Play Animation with Finished event(
    Widget=Self-Reference()
    InAnimation=Get AddProgress() -> (AddProgress:Widget Animation Object Reference)
    StartAtTime=0
    NumLoopsToPlay=1
    PlaybackSpeed=1.2
   )
   n72 = Play Animation with Finished event(execute=@n71.Finished, Widget=Self-Reference(), InAnimation=Get CountProgress(), StartAtTime=0, NumLoopsToPlay=1, PlaybackSpeed=0.7)
   Call Progress Completed(execute=@n72.Finished)
   Event StartNotificationOutro() {is=title K2Node_Event}
   > n73 = Branch(Condition="integer >= integer"(A=Get CurrentCount() -> (CurrentCount=0), B=Get GoalCount() -> (GoalCount=0)))
   FinishNotificationOutro(execute=[@n75.then, @n73.then])
   > StopListeningForAllInputActions(self:User Widget Object Reference=Get KeybindWidget())
   n74 = Get Outro() -> (Outro:Widget Animation Object Reference)
   PlayAnimation(execute=@n73.else, InAnimation=@n74, bRestoreState=true)
   > n75 = RetriggerableDelay(Duration=GetEndTime(self=@n74))
   Event HideBindWidgets() {is=title K2Node_Event}
   n76 = Get HorizontalBoxHotkeyPrompt() -> (HorizontalBoxHotkeyPrompt:Horizontal Box Object Reference)
   > SetVisibility(self=@n76, InVisibility=Collapsed)
   Event ShowBindWidgets() {is=title K2Node_Event}
   > n77 = IsQuestTrackingEnabled!(execute, self:Fort Notification Entry Quest Object Reference) -> (then, ReturnValue:bool)
   > Branch(Condition=@n77)
   > SetVisibility(self=@n76, InVisibility=SelfHitTestInvisible)
   Event OnInitialized()
   > Bind Event to Progress Completed(self:Quest Notification Object Reference, Delegate=Create Event()) {is=title K2Node_AddDelegate}
   > BindToAnimationFinished(Animation=Get QuestComplete(), Delegate=@n2)
   > BindToAnimationStarted!(execute, self:User Widget Object Reference, Animation:Widget Animation Object Reference=Get QuestComplete(), Delegate:Delegate=@n11) -> (then)
   > SetVisibility(self=Get Image_Icon_CQ(), InVisibility=Collapsed)
   n78 = Event On Timer Set!() -> (OutputDelegate, then, TimerState:EAthenaChallengeTimerState Enum=Hidden) {is=title K2Node_Event}
   > n79 = Branch(Condition="Equal (Enum)"(A:EAthenaChallengeTimerState Enum=@n78.TimerState, B:EAthenaChallengeTimerState Enum=Hidden))
   n80 = Get Border_TimerContainer() -> (Border_TimerContainer:Border Object Reference)
   > SetVisibility(self=@n80, InVisibility=Hidden)
   SetVisibility(execute=@n79.else, self=@n80, InVisibility=SelfHitTestInvisible)
   > SetContentColorAndOpacity!(
    execute
    self:Border Object Reference=Get Border_TimerContainer()
    InContentColorAndOpacity:Linear Color Structure=Select!(
     Option 0:Linear Color Structure
     Option 1:Linear Color Structure
     Option 2:Linear Color Structure
     Option 3:Linear Color Structure
     Option 4:Linear Color Structure
     Option 5:Linear Color Structure
     Index:Byte=@n78.TimerState
    ) -> (ReturnValue:Linear Color Structure)
   ) -> (then)
   > SetActiveWidget(
    self=Get WidgetSwitcher_Urgent() -> (WidgetSwitcher_Urgent:Widget Switcher Object Reference)
    Widget=Select!(
     Option 0:Widget Object Reference=Get DateTimeText_TimeRemaining() -> (DateTimeText_TimeRemaining:Athena Challenge Date Time Text Block Object Reference)
     Option 1:Widget Object Reference=Get Text_TimeRemainingUrgent() -> (Text_TimeRemainingUrgent:Common Text Object Reference)
     Index:bool="Equal (Enum)"(A=@n78.TimerState, B=Urgent)
    ) -> (ReturnValue:Widget Object Reference)
   )
   Event On Initialize Quest Notification Finished() {is=title K2Node_Event}
   > SetText(
    self=Get Rich_QuestText() -> (Rich_QuestText:Rich Text Block Object Reference)
    InText="GetText (Text)"!(self:Text Object Reference=Get Text_Description() -> (Text_Description:Common Text Object Reference)) -> (ReturnValue:Text)
   )
   > Branch(Condition=Get bIsEndOfMatch() -> (bIsEndOfMatch:bool))
   n81 = Get Text_CategoryDisplay() -> (Text_CategoryDisplay:Common Text Object Reference)
   > Branch(Condition=TextIsEmpty(InText="GetText (Text)"(self=@n81)))
   > SetVisibility(self=@n81, InVisibility=Collapsed)
   BindToAnimationFinished(Animation=Get FillProgress())
   n82 = Get Text_MapPrompt()
   SetActiveWidget(
    self:Common Visibility Switcher Object Reference=Get Switcher_TrackText() -> (Switcher_TrackText:Common Visibility Switcher Object Reference)
    Widget=Select!(
     MouseAndKeyboard:Widget Object Reference=@n82
     Gamepad:Widget Object Reference=@n82
     Touch:Widget Object Reference=Get Text_MobileCollectionPrompt() -> (Text_MobileCollectionPrompt:Common Text Object Reference)
     Count:Widget Object Reference
     Index:ECommonInputType Enum=MouseAndKeyboard
    ) -> (ReturnValue)
   )
   SetScalarParameterValue(
    self=GetDynamicMaterial(self:Border Object Reference=Get CommonBorder_MainContent() -> (CommonBorder_MainContent:Common Border Object Reference))
    ParameterName=Opacity
    Value=0.85
   )
  ]}
 GetBindActions {nodes [
   GetBindActions() {is=title K2Node_FunctionEntry}
   > Return Node!(execute, OutActions:Array of Names)
  ]}
 ProgressCompleted {nodes [ProgressCompleted() {is=title K2Node_FunctionEntry}]}
 SetCountLerp {nodes [
   n1 = SetCountLerp!() -> (then, Value:double=~) {is=title K2Node_FunctionEntry}
   > n2 = AddedProgress!(
    execute
    self
    StartValue:int=Get PreviousCount() -> (PreviousCount=0)
    TargetValue:int=Get CurrentCount() -> (CurrentCount=0)
    ProgressTime:double=@n1
   ) -> (then, TextProgress:Text) {is=title K2Node_CallFunction}
   > Set Text(Text=@n2, self=Get RichText_Objectives_Current())
  ]}
 SetIconParameters {nodes [
   SetIconParameters() {is=title K2Node_FunctionEntry}
   > n1 = GetDynamicMaterial(self:Image Object Reference=Get Image_Icon())
   > SetScalarParameterValue(self=@n1, ParameterName=Corner roundness, Value=0.15)
   > SetScalarParameterValue(self=@n1, ParameterName=Left aligned, Value=1)
  ]}
 SetMobileVisuals {nodes [
   SetMobileVisuals() {is=title K2Node_FunctionEntry}
   > Branch(Condition=IsUsingCompactUI())
   > SetTextStyleSet!(
    execute
    self:Rich Text Block Object Reference=Get Rich_QuestText()
    NewTextStyleSet:Data Table Object Reference=/Game/Athena/HUD/QuestIcons/DT_RichText_QuestNotification_Mobile
   ) -> (then)
   > SetMaxDesiredHeight(self=Get SizeBox_Description(), InMaxDesiredHeight=200)
   > n1 = Cast To FortPlayerControllerFrontEnd!(execute, Object=GetOwningPlayer()) -> (then, CastFailed, AsFort Player Controller Front End:Fort Player Controller Front End Object Reference) {is=title K2Node_DynamicCast}
   n2 = Select!(Option 0:ESlateVisibility Enum=SelfHitTestInvisible, Option 1:ESlateVisibility Enum=Visible, Index:bool=IsValid(Object=@n1)) -> (ReturnValue:ESlateVisibility Enum)
   SetMobileClickCatcherVisibility!(execute=[@n1.CastFailed, @n1.then], self:Fort Notification Entry Quest Object Reference, ButtonVisibility:ESlateVisibility Enum=@n2) -> (then)
   > SetVisibility(self=Get Image_Main_CategoryBG(), InVisibility=@n2)
  ]}
 ShouldShowPromptInHud {nodes [
   ShouldShowPromptInHud() {is=title K2Node_FunctionEntry}
   > Return Node!(execute, ReturnValue:bool=false)
  ]}
 UpdateProgress {nodes [
   UpdateProgress() {is=title K2Node_FunctionEntry}
   > PlayAnimation(InAnimation=Get FillProgress())
  ]}
]
variables [
 AddedCurrent {type=integer}
 bIsEndOfMatch {type=boolean}
 CountLerp {type=number}
 CQ_Material {type=MaterialInstance, default=/FortniteCoreUI/UI/Quests/MaterialInstances/MI_UI_QuestNotification_CommunityQuest}
 CQ_Material_DefaultPortrait_BG {type=MaterialInstance, default=/FortniteCoreUI/UI/Quests/MaterialInstances/MI_UI_QuestNotification_Icon_Background}
 CQ_Material_Portrait_BG {type=MaterialInstance, default=/FortniteCoreUI/UI/Quests/MaterialInstances/MI_UI_CommunityQuestNotification_Icon_Background}
 Default_BG_Material {type=MaterialInstance, default=/FortniteCoreUI/UI/Quests/MaterialInstances/MI_UI_QuestNotification_Quest}
 DefaultIconMaterial {type=MaterialInstance, default=/FortniteCoreUI/UI/Quests/MaterialInstances/MI_UI_Quest_TileBackground_Default}
 Duration {type=number}
 ElapsedTime {type=number}
 Flavor Text Visibility {type=ESlateVisibility, default=SelfHitTestInvisible}
 HideQuestDisplay {type=boolean}
 In Width Override {type=number, default=540}
 IsCQ {type=boolean, default=true}
 IsStorySpecial {type=boolean}
 MID_MissionHeader {type=MaterialInstanceDynamic}
 ProgressCompleted {}
 PunchcardSoftMaterial {type=MaterialInterface, default=&MI_UI_Quest_TileBackground_Default}
 SimpleAnim_Intro {type=string, default=Intro}
 SimpleAnim_IntroDuration {type=number, default=0.5}
 Text_Map_Prompt {type=string, default=Hold to Track Quest}
 Text_Mobile_Prompt {type=string, default=Tap To Track Quest}
 TextMaxHeightOverride {type=number, default=150}
]
widgets {
 info {inheritedWidgetCount=9, parentClass=/Script/FortniteUI.FortNotificationEntry_Quest, rootWidgetClass=&GridPanel, widgetCount=59}
 widgets [
  Root {widgetClassPath=&GridPanel, widget=Root, properties {columnFill [1], rowFill [1]}}
  SizeBox_main {
   widgetClassPath=&SizeBox
   bIsVariable=true
   parent=Root
   widget=SizeBox_main
   properties {bOverride_WidthOverride=true, heightOverride=80, maxDesiredHeight=150, minDesiredHeight=80, widthOverride=550}
   slot {class=GridSlot}
  }
  SimpleAnimations {
   widgetClassPath=/Script/FortniteUI.FortSimpleWidgetAnimationsPanel
   bIsVariable=true
   parent=SizeBox_main
   widget=SimpleAnimations
   properties {animations [Intro {
      bIsEnabled=true
      onAnimationsFinished=()
      targets [
       {
        alpha=/Game/UI/InGame/EventNotifications/UIAnim_Intro_Alpha
        bResetOnFinish=false
        bShouldAnimateAlpha=true
        bShouldAnimateScale=false
        bShouldAnimateTranslation=true
        scale=null
        targetName=MainContent
        translation=/Game/UI/InGame/EventNotifications/UIAnim_Intro_Translation
        materialScalarAnimations []
       }
       {
        alpha=&UIAnim_Intro_Alpha
        bResetOnFinish=false
        bShouldAnimateAlpha=true
        bShouldAnimateScale=false
        bShouldAnimateTranslation=true
        scale=null
        targetName=HorizontalBoxHotkeyPrompt
        translation=&UIAnim_Intro_Translation
        materialScalarAnimations []
       }
      ]
     }]}
   slot {class=SizeBoxSlot, properties {verticalAlignment=VAlign_Top}}
  }
  MainContent {
   widgetClassPath=&GridPanel
   parent=SimpleAnimations
   widget=MainContent
   properties {columnFill [1], rowFill [1]}
   slot {class=FortSimpleWidgetAnimationsSlot}
  }
  Text_CategoryDisplay {
   widgetClassPath=&CommonTextBlock
   bInherited=true
   bIsVariable=true
   parent=MainContent
   widget=Text_CategoryDisplay
   properties {
    text=Quest Display Category
    font {
     bForceMonospaced=false
     bMaterialIsStencil=false
     fontMaterial=null
     fontObject=&HeadingNow
     letterSpacing=0
     monospacedWidth=1
     size=16.5
     skewAmount=0
     typefaceFontName=86 Bold Italic
     outlineSettings {
      bApplyOutlineToDropShadows=false
      bMiteredCorners=false
      bSeparateFillAlpha=false
      outlineMaterial=null
      outlineSize=0
      outlineColor=<rgba 0 0 0 1>
     }
    }
    shadowOffset=<0 0>
   }
   slot {class=GridSlot, properties {padding=<ltrb 2 0 0 4>}}
  }
  Panel_AssistedBy {
   widgetClassPath=&Overlay
   bInherited=true
   bIsVariable=true
   parent=MainContent
   widget=Panel_AssistedBy
   slot {class=GridSlot, properties {row=1, verticalAlignment=VAlign_Center, padding=<ltrb 0 10 0 7>}}
  }
  Image_Assisted_BG {
   widgetClassPath=&Image
   bIsVariable=true
   parent=Panel_AssistedBy
   widget=Image_Assisted_BG
   properties {brush {
     bIsDynamicallyLoaded=false
     drawAs=Image
     imageType=FullColor
     mirroring=NoMirror
     resourceName=None
     resourceObject=&MI_UI_QuestNotification_Quest
     tiling=NoTile
     imageSize=<32 32>
     margin=<ltrb 0 0 0 0>
     outlineSettings {
      bUseBrushTransparency=false
      roundingType=FixedRadius
      width=0
      color {colorUseRule=UseColor_Specified, specifiedColor=<rgba 0 0 0 0>}
      cornerRadii=<12 12 12 12>
     }
     tintColor {colorUseRule=UseColor_Specified, specifiedColor=<rgba 1 1 1 1>}
     uVRegion {bIsValid=false, max=<0 0>, min=<0 0>}
    }}
   slot {class=OverlaySlot, properties {horizontalAlignment=HAlign_Fill, verticalAlignment=VAlign_Fill}}
  }
  HorizontalBox_Assisted {
   widgetClassPath=&HorizontalBox
   parent=Panel_AssistedBy
   widget=HorizontalBox_Assisted
   slot {class=OverlaySlot, properties {verticalAlignment=VAlign_Center, padding=<ltrb 23 8 0 8>}}
  }
  Image_GotHelped {
   widgetClassPath=&Image
   parent=HorizontalBox_Assisted
   widget=Image_GotHelped
   properties {
    brush {
     bIsDynamicallyLoaded=false
     drawAs=Image
     imageType=FullColor
     mirroring=NoMirror
     resourceName=None
     resourceObject=/FortniteCoreUI/Athena/HUD/Art/T_Icon_Assist
     tiling=NoTile
     imageSize=<12 12>
     margin=<ltrb 0 0 0 0>
     outlineSettings {
      bUseBrushTransparency=false
      roundingType=HalfHeightRadius
      width=0
      color {colorUseRule=UseColor_Specified, specifiedColor=<rgba 0 0 0 0>}
      cornerRadii=<0 0 0 0>
     }
     tintColor {colorUseRule=UseColor_Specified, specifiedColor=<rgba 1 1 1 1>}
     uVRegion {bIsValid=false, max=<0 0>, min=<0 0>}
    }
    renderTransform {angle=0, scale=<2 2>, shear=<0 0>, translation=<0 0>}
   }
   slot {class=HorizontalBoxSlot, properties {horizontalAlignment=HAlign_Left, verticalAlignment=VAlign_Center, padding=<ltrb 4 0 6 0>}}
  }
  Text_AssistBlock {
   widgetClassPath=&CommonRichTextBlock
   bIsVariable=true
   parent=HorizontalBox_Assisted
   widget=Text_AssistBlock
   properties {
    bOverrideDefaultStyle=true
    mobileTextBlockScale=1.1
    scrollStyle=&TextScrollStyle-Base_C
    text="<text case=\"upper\">Assisted by</> {PlayerName}"
    wrapTextAt=400
    defaultTextStyleOverride {
     overflowPolicy=Clip
     transformPolicy=None
     colorAndOpacity {colorUseRule=UseColor_Specified, specifiedColor=<rgba 1 1 1 1>}
     font {
      bForceMonospaced=false
      bMaterialIsStencil=false
      fontMaterial=null
      fontObject=&HeadingNow
      letterSpacing=-50
      monospacedWidth=1
      size=10.5
      skewAmount=0
      typefaceFontName=86 Bold
      outlineSettings {
       bApplyOutlineToDropShadows=false
       bMiteredCorners=false
       bSeparateFillAlpha=false
       outlineMaterial=null
       outlineSize=0
       outlineColor=<rgba 0 0 0 0.3>
      }
     }
     highlightColor {colorUseRule=UseColor_Specified, specifiedColor=<rgba 0 0 0 1>}
     highlightShape {
      bIsDynamicallyLoaded=false
      drawAs=Image
      imageType=NoImage
      mirroring=NoMirror
      resourceName=None
      resourceObject=null
      tiling=NoTile
      imageSize=<32 32>
      margin=<ltrb 0 0 0 0>
      outlineSettings {
       bUseBrushTransparency=false
       roundingType=HalfHeightRadius
       width=0
       color {colorUseRule=UseColor_Specified, specifiedColor=<rgba 0 0 0 0>}
       cornerRadii=<0 0 0 0>
      }
      tintColor {colorUseRule=UseColor_Specified, specifiedColor=<rgba 1 1 1 1>}
      uVRegion {bIsValid=false, max=<0 0>, min=<0 0>}
     }
     selectedBackgroundColor {colorUseRule=UseColor_Foreground, specifiedColor=<rgba 1 0 1 1>}
     shadowColorAndOpacity=<rgba 0 0 0 1>
     shadowOffset=<0 0>
     strikeBrush {
      bIsDynamicallyLoaded=false
      drawAs=Image
      imageType=NoImage
      mirroring=NoMirror
      resourceName=None
      resourceObject=null
      tiling=NoTile
      imageSize=<32 32>
      margin=<ltrb 0 0 0 0>
      outlineSettings {
       bUseBrushTransparency=false
       roundingType=HalfHeightRadius
       width=0
       color {colorUseRule=UseColor_Specified, specifiedColor=<rgba 0 0 0 0>}
       cornerRadii=<0 0 0 0>
      }
      tintColor {colorUseRule=UseColor_Specified, specifiedColor=<rgba 1 1 1 1>}
      uVRegion {bIsValid=false, max=<0 0>, min=<0 0>}
     }
     underlineBrush {
      bIsDynamicallyLoaded=false
      drawAs=Image
      imageType=NoImage
      mirroring=NoMirror
      resourceName=None
      resourceObject=null
      tiling=NoTile
      imageSize=<32 32>
      margin=<ltrb 0 0 0 0>
      outlineSettings {
       bUseBrushTransparency=false
       roundingType=HalfHeightRadius
       width=0
       color {colorUseRule=UseColor_Specified, specifiedColor=<rgba 0 0 0 0>}
       cornerRadii=<0 0 0 0>
      }
      tintColor {colorUseRule=UseColor_Specified, specifiedColor=<rgba 1 1 1 1>}
      uVRegion {bIsValid=false, max=<0 0>, min=<0 0>}
     }
    }
    renderTransform {angle=0, scale=<1 1.15>, shear=<-1 0>, translation=<0 0>}
   }
   slot {class=HorizontalBoxSlot, properties {padding=<ltrb 8 2 4 2>}}
  }
  SizeBox_Description {
   widgetClassPath=&SizeBox
   bIsVariable=true
   parent=MainContent
   widget=SizeBox_Description
   properties {bOverride_MaxDesiredHeight=true, bOverride_MinDesiredHeight=true, heightOverride=90, maxDesiredHeight=155, minDesiredHeight=115}
   slot {class=GridSlot, properties {row=2, padding=<ltrb 0 5 0 5>}}
  }
  CommonBorder_MainContent {
   widgetClassPath=&CommonBorder
   bIsVariable=true
   parent=SizeBox_Description
   widget=CommonBorder_MainContent
   properties {
    background {
     bIsDynamicallyLoaded=false
     drawAs=Box
     imageType=NoImage
     mirroring=NoMirror
     resourceName=None
     resourceObject=/Game/Athena/HUD/MiniMap/Art/M_UI_Quest_Container_S19_Fill
     tiling=NoTile
     imageSize=<64 64>
     margin=<ltrb 0.5 0.5 0.5 0.5>
     outlineSettings {
      bUseBrushTransparency=false
      roundingType=HalfHeightRadius
      width=0
      color {colorUseRule=UseColor_Specified, specifiedColor=<rgba 0 0 0 0>}
      cornerRadii=<0 0 0 0>
     }
     tintColor {colorUseRule=UseColor_Specified, specifiedColor=<rgba 1 1 1 1>}
     uVRegion {bIsValid=false, max=<0 0>, min=<0 0>}
    }
    brushColor=<rgba 1 1 1 0>
    padding=<ltrb 0 0 0 0>
   }
   slot {class=SizeBoxSlot}
  }
  HorizontalBox_4 {
   widgetClassPath=&HorizontalBox
   parent=CommonBorder_MainContent
   widget=HorizontalBox_4
   slot {class=BorderSlot, properties {padding=<ltrb 0 0 0 0>}}
  }
  Border_Icon {
   widgetClassPath=&CommonBorder
   bIsVariable=true
   parent=HorizontalBox_4
   widget=Border_Icon
   properties {
    background {
     bIsDynamicallyLoaded=false
     drawAs=Image
     imageType=FullColor
     mirroring=NoMirror
     resourceName=None
     resourceObject=&MI_UI_QuestNotification_Icon_Background
     tiling=NoTile
     imageSize=<32 32>
     margin=<ltrb 0.1 0.1 0.1 0.1>
     outlineSettings {
      bUseBrushTransparency=false
      roundingType=FixedRadius
      width=0
      color {colorUseRule=UseColor_Specified, specifiedColor=<rgba 0 0 0 0>}
      cornerRadii=<14 14 14 14>
     }
     tintColor {colorUseRule=UseColor_Specified, specifiedColor=<rgba 1 1 1 1>}
     uVRegion {bIsValid=false, max=<0 0>, min=<0 0>}
    }
    padding=<ltrb 0 0 0 0>
   }
   slot {class=HorizontalBoxSlot}
  }
  Image_Icon {
   widgetClassPath=&FortLazyImage
   bIsVariable=true
   parent=Border_Icon
   widget=Image_Icon
   properties {
    materialTextureParamName=NPC-Portrait
    brush {
     bIsDynamicallyLoaded=false
     drawAs=Image
     imageType=FullColor
     mirroring=NoMirror
     resourceName=None
     resourceObject=/FortniteCoreUI/UI/Quests/MaterialInstances/MI_UIKit_Item_QuestNotification
     tiling=NoTile
     imageSize=<100 100>
     margin=<ltrb 0.25 0.25 0.25 0.25>
     outlineSettings {
      bUseBrushTransparency=false
      roundingType=HalfHeightRadius
      width=0
      color {colorUseRule=UseColor_Specified, specifiedColor=<rgba 0 0 0 0>}
      cornerRadii=<0 0 0 0>
     }
     tintColor {colorUseRule=UseColor_Specified, specifiedColor=<rgba 1 1 1 1>}
     uVRegion {bIsValid=false, max=<0 0>, min=<0 0>}
    }
    loadingBackgroundBrush {
     bIsDynamicallyLoaded=false
     drawAs=Image
     imageType=FullColor
     mirroring=NoMirror
     resourceName=None
     resourceObject=null
     tiling=NoTile
     imageSize=<128 128>
     margin=<ltrb 0 0 0 0>
     outlineSettings {
      bUseBrushTransparency=false
      roundingType=HalfHeightRadius
      width=0
      color {colorUseRule=UseColor_Specified, specifiedColor=<rgba 0 0 0 0>}
      cornerRadii=<0 0 0 0>
     }
     tintColor {colorUseRule=UseColor_Specified, specifiedColor=<rgba 1 1 1 1>}
     uVRegion {bIsValid=false, max=<0 0>, min=<0 0>}
    }
   }
   slot {class=BorderSlot, properties {padding=<ltrb 0 0 0 0>}}
  }
  Overlay_Main_BG {
   widgetClassPath=&Overlay
   parent=HorizontalBox_4
   widget=Overlay_Main_BG
   slot {class=HorizontalBoxSlot, properties {size {sizeRule=Fill, value=1}}}
  }
  Switcher_MainBG {
   widgetClassPath=&CommonVisibilitySwitcher
   bIsVariable=true
   parent=Overlay_Main_BG
   widget=Switcher_MainBG
   slot {class=OverlaySlot, properties {horizontalAlignment=HAlign_Fill, verticalAlignment=VAlign_Fill}}
  }
  Overlay_0 {
   widgetClassPath=&Overlay
   parent=Switcher_MainBG
   widget=Overlay_0
   properties {bOverride_Cursor=true}
   slot {class=CommonVisibilitySwitcherSlot, properties {padding=<ltrb 12 0 0 0>}}
  }
  Image_Main_DefaultBG {
   widgetClassPath=&Image
   bIsVariable=true
   parent=Overlay_0
   widget=Image_Main_DefaultBG
   properties {brush {
     bIsDynamicallyLoaded=false
     drawAs=Image
     imageType=FullColor
     mirroring=NoMirror
     resourceName=None
     resourceObject=&MI_UI_QuestNotification_Quest
     tiling=NoTile
     imageSize=<32 36>
     margin=<ltrb 0 0 0 0>
     outlineSettings {
      bUseBrushTransparency=false
      roundingType=FixedRadius
      width=0
      color {colorUseRule=UseColor_Specified, specifiedColor=<rgba 0 0 0 0>}
      cornerRadii=<15 15 15 15>
     }
     tintColor {colorUseRule=UseColor_Specified, specifiedColor=<rgba 1 1 1 0.9>}
     uVRegion {bIsValid=false, max=<0 0>, min=<0 0>}
    }}
   slot {class=OverlaySlot, properties {horizontalAlignment=HAlign_Fill, verticalAlignment=VAlign_Fill}}
  }
  Image_Main_CategoryBG {
   widgetClassPath=&Image
   bIsVariable=true
   parent=Switcher_MainBG
   widget=Image_Main_CategoryBG
   properties {brush {
     bIsDynamicallyLoaded=false
     drawAs=Image
     imageType=FullColor
     mirroring=NoMirror
     resourceName=None
     resourceObject=/FortniteCoreUI/Items/QuestItems/Categories/Art/MI_UI_QB_Weekly
     tiling=NoTile
     imageSize=<0 0>
     margin=<ltrb 0 0 0 0>
     outlineSettings {
      bUseBrushTransparency=false
      roundingType=HalfHeightRadius
      width=0
      color {colorUseRule=UseColor_Specified, specifiedColor=<rgba 0 0 0 0>}
      cornerRadii=<0 0 0 0>
     }
     tintColor {colorUseRule=UseColor_Specified, specifiedColor=<rgba 1 1 1 1>}
     uVRegion {bIsValid=false, max=<0 0>, min=<0 0>}
    }}
   slot {class=CommonVisibilitySwitcherSlot}
  }
  GridPanel_TextContent {
   widgetClassPath=&GridPanel
   parent=Overlay_Main_BG
   widget=GridPanel_TextContent
   properties {columnFill [1], rowFill [3]}
   slot {class=OverlaySlot, properties {horizontalAlignment=HAlign_Fill, verticalAlignment=VAlign_Fill, padding=<ltrb 30 18 20 25>}}
  }
  HorizontalBox_Description {
   widgetClassPath=&HorizontalBox
   parent=GridPanel_TextContent
   widget=HorizontalBox_Description
   slot {class=GridSlot, properties {verticalAlignment=VAlign_Center}}
  }
  VerticalBox_0 {
   widgetClassPath=&VerticalBox
   parent=HorizontalBox_Description
   widget=VerticalBox_0
   slot {class=HorizontalBoxSlot, properties {verticalAlignment=VAlign_Center, size {sizeRule=Fill, value=1}}}
  }
  HorizontalBox_0 {widgetClassPath=&HorizontalBox, parent=VerticalBox_0, widget=HorizontalBox_0, slot {class=VerticalBoxSlot}}
  OverlayCQIcon {
   widgetClassPath=&Overlay
   parent=HorizontalBox_0
   widget=OverlayCQIcon
   properties {visibility=HitTestInvisible}
   slot {class=HorizontalBoxSlot, properties {horizontalAlignment=HAlign_Center, verticalAlignment=VAlign_Center}}
  }
  Image_Icon_CQ {
   widgetClassPath=&Image
   bIsVariable=true
   parent=OverlayCQIcon
   widget=Image_Icon_CQ
   properties {
    visibility=Collapsed
    brush {
     bIsDynamicallyLoaded=false
     drawAs=Image
     imageType=FullColor
     mirroring=NoMirror
     resourceName=None
     resourceObject=/FortniteCoreUI/UI/Quests/Materials/M_UI_Icon_Toast_Backplate_CommunityQuest
     tiling=NoTile
     imageSize=<32 32>
     margin=<ltrb 0 0 0 0>
     outlineSettings {
      bUseBrushTransparency=false
      roundingType=HalfHeightRadius
      width=0
      color {colorUseRule=UseColor_Specified, specifiedColor=<rgba 0 0 0 0>}
      cornerRadii=<0 0 0 0>
     }
     tintColor {colorUseRule=UseColor_Specified, specifiedColor=<rgba 1 1 1 1>}
     uVRegion {bIsValid=false, max=<0 0>, min=<0 0>}
    }
    renderTransform {angle=0, scale=<2 2>, shear=<0 0>, translation=<-4 0>}
   }
   slot {class=OverlaySlot, properties {horizontalAlignment=HAlign_Center, verticalAlignment=VAlign_Center, padding=<ltrb 16 0 16 0>}}
  }
  CommonVisibilitySwitcher_0 {
   widgetClassPath=&CommonVisibilitySwitcher
   parent=HorizontalBox_0
   widget=CommonVisibilitySwitcher_0
   properties {activeWidgetIndex=1}
   slot {class=HorizontalBoxSlot, properties {size {sizeRule=Fill, value=1}}}
  }
  Text_Description {
   widgetClassPath=&CommonTextBlock
   bInherited=true
   bIsVariable=true
   parent=CommonVisibilitySwitcher_0
   widget=Text_Description
   properties {
    autoWrapText=true
    bAutoCollapseWithEmptyText=true
    bIsScrollingEnabled=false
    clipping=ClipToBounds
    lineHeightPercentage=1.2
    mobileFontSizeMultiplier=1.4
    text=Do a very descriptive task with lots of text that fills up to three lines of description
    textOverflowPolicy=MultilineEllipsis
    visibility=HitTestInvisible
    font {
     bForceMonospaced=false
     bMaterialIsStencil=false
     fontMaterial=null
     fontObject=&HeadingNow
     letterSpacing=50
     monospacedWidth=0.7333
     size=12.75
     skewAmount=0
     typefaceFontName=75 Medium
     outlineSettings {
      bApplyOutlineToDropShadows=false
      bMiteredCorners=false
      bSeparateFillAlpha=false
      outlineMaterial=null
      outlineSize=0
      outlineColor=<rgba 0 0 0 0.1>
     }
    }
    shadowOffset=<0 0>
   }
   slot {class=CommonVisibilitySwitcherSlot}
  }
  Rich_QuestText {
   widgetClassPath=/Script/UMG.RichTextBlock
   bIsVariable=true
   parent=CommonVisibilitySwitcher_0
   widget=Rich_QuestText
   properties {
    autoWrapText=true
    text=Do a very descriptive task with lots of text that fills up to three lines of description
    textStyleSet=/Game/Athena/HUD/QuestIcons/DT_RichText_QuestNotification
    decoratorClasses [/Game/Athena/HUD/QuestIcons/BP_QuestTile_IconDecorator.BP_QuestTile_IconDecorator_C]
   }
   slot {class=CommonVisibilitySwitcherSlot}
  }
  Overlay_1 {
   widgetClassPath=&Overlay
   parent=VerticalBox_0
   widget=Overlay_1
   slot {class=VerticalBoxSlot, properties {padding=<ltrb 0 0 0 10>}}
  }
  Text_FlavorText {
   widgetClassPath=&CommonTextBlock
   bInherited=true
   bIsVariable=true
   parent=Overlay_1
   widget=Text_FlavorText
   properties {
    autoWrapText=true
    bAutoCollapseWithEmptyText=true
    bIsScrollingEnabled=false
    clipping=ClipToBounds
    mobileFontSizeMultiplier=1.4
    text=Completion text that is extremely long and fills at least three lines of text and this is to test that and more
    textOverflowPolicy=Ellipsis
    colorAndOpacity {colorUseRule=UseColor_Specified, specifiedColor=<rgba 1 1 1 0.8>}
    font {
     bForceMonospaced=false
     bMaterialIsStencil=false
     fontMaterial=null
     fontObject=&HeadingNow
     letterSpacing=-60
     monospacedWidth=1
     size=12.75
     skewAmount=0
     typefaceFontName=74 Regular
     outlineSettings {
      bApplyOutlineToDropShadows=false
      bMiteredCorners=false
      bSeparateFillAlpha=false
      outlineMaterial=null
      outlineSize=0
      outlineColor=<rgba 0 0 0 0.5>
     }
    }
    shadowOffset=<0 0>
   }
   slot {class=OverlaySlot, properties {horizontalAlignment=HAlign_Fill}}
  }
  Spacer_Flavor {
   widgetClassPath=&Spacer
   bIsVariable=true
   parent=Overlay_1
   widget=Spacer_Flavor
   properties {visibility=Collapsed, size=<1 40>}
   slot {class=OverlaySlot, properties {horizontalAlignment=HAlign_Fill, verticalAlignment=VAlign_Fill}}
  }
  Spacer_Story {
   widgetClassPath=&Spacer
   bIsVariable=true
   parent=HorizontalBox_Description
   widget=Spacer_Story
   properties {visibility=Visible, size=<110 60>}
   slot {class=HorizontalBoxSlot}
  }
  HorizontalBox_BottomBar {
   widgetClassPath=&HorizontalBox
   parent=GridPanel_TextContent
   widget=HorizontalBox_BottomBar
   properties {visibility=HitTestInvisible, renderTransform {angle=0, scale=<1 1>, shear=<0 0>, translation=<0 14>}}
   slot {class=GridSlot, properties {verticalAlignment=VAlign_Bottom, padding=<ltrb 0 0 0 -8>}}
  }
  Border_TimerContainer {
   widgetClassPath=&Border
   bIsVariable=true
   parent=HorizontalBox_BottomBar
   widget=Border_TimerContainer
   properties {
    visibility=Collapsed
    background {
     bIsDynamicallyLoaded=false
     drawAs=NoDrawType
     imageType=NoImage
     mirroring=NoMirror
     resourceName=None
     resourceObject=null
     tiling=NoTile
     imageSize=<32 32>
     margin=<ltrb 0 0 0 0>
     outlineSettings {
      bUseBrushTransparency=false
      roundingType=HalfHeightRadius
      width=0
      color {colorUseRule=UseColor_Specified, specifiedColor=<rgba 0 0 0 0>}
      cornerRadii=<0 0 0 0>
     }
     tintColor {colorUseRule=UseColor_Specified, specifiedColor=<rgba 1 1 1 1>}
     uVRegion {bIsValid=false, max=<0 0>, min=<0 0>}
    }
    contentColorAndOpacity=<rgba 0.2195 0.6584 1 1>
    padding=<ltrb 0 0 0 0>
   }
   slot {class=HorizontalBoxSlot, properties {verticalAlignment=VAlign_Bottom, size {sizeRule=Fill, value=1}}}
  }
  TimerContainer {
   widgetClassPath=&HorizontalBox
   parent=Border_TimerContainer
   widget=TimerContainer
   properties {flowDirectionPreference=Culture}
   slot {class=BorderSlot, properties {padding=<ltrb 0 0 0 0>}}
  }
  ImageTimer {
   widgetClassPath=&Image
   parent=TimerContainer
   widget=ImageTimer
   properties {brush {
     bIsDynamicallyLoaded=false
     drawAs=Image
     imageType=FullColor
     mirroring=NoMirror
     resourceName=None
     resourceObject=/Game/Athena/HUD/Materials/M_UI_Timer
     tiling=NoTile
     imageSize=<32 32>
     margin=<ltrb 0 0 0 0>
     outlineSettings {
      bUseBrushTransparency=false
      roundingType=HalfHeightRadius
      width=0
      color {colorUseRule=UseColor_Specified, specifiedColor=<rgba 0 0 0 0>}
      cornerRadii=<0 0 0 0>
     }
     tintColor {colorUseRule=UseColor_Specified, specifiedColor=<rgba 1 1 1 1>}
     uVRegion {bIsValid=false, max=<0 0>, min=<0 0>}
    }}
   slot {class=HorizontalBoxSlot, properties {horizontalAlignment=HAlign_Left, verticalAlignment=VAlign_Center, padding=<ltrb -6 -12 0 -12>}}
  }
  WidgetSwitcher_Urgent {
   widgetClassPath=&WidgetSwitcher
   bIsVariable=true
   parent=TimerContainer
   widget=WidgetSwitcher_Urgent
   slot {class=HorizontalBoxSlot, properties {verticalAlignment=VAlign_Center, size {sizeRule=Fill, value=1}}}
  }
  DateTimeText_TimeRemaining {
   widgetClassPath=/Script/FortniteUI.AthenaChallengeDateTimeTextBlock
   bInherited=true
   bIsVariable=true
   parent=WidgetSwitcher_Urgent
   widget=DateTimeText_TimeRemaining
   properties {
    mobileFontSizeMultiplier=1.2
    text=Text Block
    font {
     bForceMonospaced=false
     bMaterialIsStencil=false
     fontMaterial=/FortniteCoreUI/Athena/HUD/Quests/Art/M_UI_Timer_Font_Quest
     fontObject=&HeadingNow
     letterSpacing=0
     monospacedWidth=1
     size=12
     skewAmount=0
     typefaceFontName=75 Medium
     outlineSettings {
      bApplyOutlineToDropShadows=false
      bMiteredCorners=false
      bSeparateFillAlpha=false
      outlineMaterial=null
      outlineSize=0
      outlineColor=<rgba 0 0 0 0.3>
     }
    }
    shadowOffset=<0 0>
   }
   slot {class=WidgetSwitcherSlot, properties {verticalAlignment=VAlign_Bottom}}
  }
  Text_TimeRemainingUrgent {
   widgetClassPath=&CommonTextBlock
   bIsVariable=true
   parent=WidgetSwitcher_Urgent
   widget=Text_TimeRemainingUrgent
   properties {
    mobileFontSizeMultiplier=1.2
    text="This Match Only!"
    textTransformPolicy=ToUpper
    visibility=Collapsed
    font {
     bForceMonospaced=false
     bMaterialIsStencil=false
     fontMaterial=&M_UI_Timer_Font_Quest
     fontObject=&HeadingNow
     letterSpacing=0
     monospacedWidth=1
     size=12
     skewAmount=0
     typefaceFontName=75 Medium
     outlineSettings {
      bApplyOutlineToDropShadows=false
      bMiteredCorners=false
      bSeparateFillAlpha=false
      outlineMaterial=null
      outlineSize=0
      outlineColor=<rgba 0 0 0 0.3>
     }
    }
    shadowOffset=<0 0>
   }
   slot {class=WidgetSwitcherSlot}
  }
  WidgetSwitcher_Complete {
   widgetClassPath=&WidgetSwitcher
   bIsVariable=true
   parent=HorizontalBox_BottomBar
   widget=WidgetSwitcher_Complete
   properties {renderTransform {angle=0, scale=<1 1>, shear=<0 0>, translation=<6 0>}}
   slot {class=HorizontalBoxSlot, properties {horizontalAlignment=HAlign_Right, verticalAlignment=VAlign_Bottom, size {sizeRule=Fill, value=1}}}
  }
  Overlay_Progress {
   widgetClassPath=&Overlay
   bIsVariable=true
   parent=WidgetSwitcher_Complete
   widget=Overlay_Progress
   slot {class=WidgetSwitcherSlot, properties {horizontalAlignment=HAlign_Right}}
  }
  RichText_ObjectivesCompleted {
   widgetClassPath=&CommonRichTextBlock
   bInherited=true
   bIsVariable=true
   parent=Overlay_Progress
   widget=RichText_ObjectivesCompleted
   properties {
    justification=Center
    mobileTextBlockScale=1.2
    text=100 / 100
    textStyleSet=/FortniteCoreUI/HUD/MiniMap/AthenaChallengeObjectiveProgressStyles
    defaultTextStyleOverride {
     overflowPolicy=Clip
     transformPolicy=None
     colorAndOpacity {colorUseRule=UseColor_Specified, specifiedColor=<rgba 0.2122 0.6514 1 1>}
     font {
      bForceMonospaced=false
      bMaterialIsStencil=false
      fontMaterial=null
      fontObject=&HeadingNow
      letterSpacing=20
      monospacedWidth=1
      size=12.75
      skewAmount=0
      typefaceFontName=76 Bold
      outlineSettings {
       bApplyOutlineToDropShadows=false
       bMiteredCorners=false
       bSeparateFillAlpha=false
       outlineMaterial=null
       outlineSize=1
       outlineColor=<rgba 0.026 0.026 0.026 0.5>
      }
     }
     highlightColor {colorUseRule=UseColor_Specified, specifiedColor=<rgba 0 0 0 1>}
     highlightShape {
      bIsDynamicallyLoaded=false
      drawAs=Image
      imageType=NoImage
      mirroring=NoMirror
      resourceName=None
      resourceObject=null
      tiling=NoTile
      imageSize=<32 32>
      margin=<ltrb 0 0 0 0>
      outlineSettings {
       bUseBrushTransparency=false
       roundingType=HalfHeightRadius
       width=0
       color {colorUseRule=UseColor_Specified, specifiedColor=<rgba 0 0 0 0>}
       cornerRadii=<0 0 0 0>
      }
      tintColor {colorUseRule=UseColor_Specified, specifiedColor=<rgba 1 1 1 1>}
      uVRegion {bIsValid=false, max=<0 0>, min=<0 0>}
     }
     selectedBackgroundColor {colorUseRule=UseColor_Specified, specifiedColor=<rgba 1 0 1 1>}
     shadowColorAndOpacity=<rgba 0.0314 0.0314 0.0314 0>
     shadowOffset=<2 2>
     strikeBrush {
      bIsDynamicallyLoaded=false
      drawAs=Image
      imageType=NoImage
      mirroring=NoMirror
      resourceName=None
      resourceObject=null
      tiling=NoTile
      imageSize=<32 32>
      margin=<ltrb 0 0 0 0>
      outlineSettings {
       bUseBrushTransparency=false
       roundingType=HalfHeightRadius
       width=0
       color {colorUseRule=UseColor_Specified, specifiedColor=<rgba 0 0 0 0>}
       cornerRadii=<0 0 0 0>
      }
      tintColor {colorUseRule=UseColor_Specified, specifiedColor=<rgba 1 1 1 1>}
      uVRegion {bIsValid=false, max=<0 0>, min=<0 0>}
     }
     underlineBrush {
      bIsDynamicallyLoaded=false
      drawAs=Image
      imageType=NoImage
      mirroring=NoMirror
      resourceName=None
      resourceObject=null
      tiling=NoTile
      imageSize=<32 32>
      margin=<ltrb 0 0 0 0>
      outlineSettings {
       bUseBrushTransparency=false
       roundingType=HalfHeightRadius
       width=0
       color {colorUseRule=UseColor_Specified, specifiedColor=<rgba 0 0 0 0>}
       cornerRadii=<0 0 0 0>
      }
      tintColor {colorUseRule=UseColor_Specified, specifiedColor=<rgba 1 1 1 1>}
      uVRegion {bIsValid=false, max=<0 0>, min=<0 0>}
     }
    }
   }
   slot {class=OverlaySlot, properties {horizontalAlignment=HAlign_Fill, verticalAlignment=VAlign_Bottom}}
  }
  Image_textglow {
   widgetClassPath=&Image
   parent=Overlay_Progress
   widget=Image_textglow
   properties {
    visibility=HitTestInvisible
    brush {
     bIsDynamicallyLoaded=false
     drawAs=Image
     imageType=NoImage
     mirroring=NoMirror
     resourceName=None
     resourceObject=/Game/Athena/HUD/LTM/Wax/Materials/M_UI_TextGlow_Add
     tiling=NoTile
     imageSize=<16 16>
     margin=<ltrb 0 0 0 0>
     outlineSettings {
      bUseBrushTransparency=false
      roundingType=HalfHeightRadius
      width=0
      color {colorUseRule=UseColor_Specified, specifiedColor=<rgba 0 0 0 0>}
      cornerRadii=<0 0 0 0>
     }
     tintColor {colorUseRule=UseColor_Specified, specifiedColor=<rgba 0.25 0.25 0.25 1>}
     uVRegion {bIsValid=false, max=<0 0>, min=<0 0>}
    }
    colorAndOpacity=<rgba 0 0.3465 0.9911 1>
    renderTransform {angle=0, scale=<0 0>, shear=<0 0>, translation=<0 0>}
   }
   slot {class=OverlaySlot, properties {horizontalAlignment=HAlign_Center, verticalAlignment=VAlign_Center}}
  }
  Image_textglow_2 {
   widgetClassPath=&Image
   parent=Overlay_Progress
   widget=Image_textglow_2
   properties {
    renderOpacity=0
    visibility=SelfHitTestInvisible
    brush {
     bIsDynamicallyLoaded=false
     drawAs=Image
     imageType=NoImage
     mirroring=NoMirror
     resourceName=None
     resourceObject=&M_UI_TextGlow_Add
     tiling=NoTile
     imageSize=<16 10>
     margin=<ltrb 0 0 0 0>
     outlineSettings {
      bUseBrushTransparency=false
      roundingType=HalfHeightRadius
      width=0
      color {colorUseRule=UseColor_Specified, specifiedColor=<rgba 0 0 0 0>}
      cornerRadii=<0 0 0 0>
     }
     tintColor {colorUseRule=UseColor_Specified, specifiedColor=<rgba 0.75 0.75 0.75 1>}
     uVRegion {bIsValid=false, max=<0 0>, min=<0 0>}
    }
    colorAndOpacity=<rgba 0 0.7053 0.9911 1>
    renderTransform {angle=7, scale=<1 1>, shear=<0 0>, translation=<0 0>}
   }
   slot {class=OverlaySlot, properties {horizontalAlignment=HAlign_Fill, verticalAlignment=VAlign_Fill, padding=<ltrb -40 2 -40 2>}}
  }
  SizeBox_Complete {
   widgetClassPath=&SizeBox
   bIsVariable=true
   parent=WidgetSwitcher_Complete
   widget=SizeBox_Complete
   properties {
    bOverride_WidthOverride=true
    heightOverride=50
    minDesiredWidth=72
    widthOverride=75
    renderTransform {angle=0, scale=<1 1>, shear=<0 0>, translation=<11 -7>}
   }
   slot {class=WidgetSwitcherSlot, properties {horizontalAlignment=HAlign_Center, verticalAlignment=VAlign_Bottom}}
  }
  Overlay_2 {
   widgetClassPath=&Overlay
   parent=SizeBox_Complete
   widget=Overlay_2
   properties {renderTransform {angle=0, scale=<1 1>, shear=<0 0>, translation=<0 -10>}}
   slot {class=SizeBoxSlot}
  }
  Image_Check {
   widgetClassPath=&Image
   bIsVariable=true
   parent=Overlay_2
   widget=Image_Check
   properties {
    renderOpacity=0
    brush {
     bIsDynamicallyLoaded=false
     drawAs=Image
     imageType=FullColor
     mirroring=NoMirror
     resourceName=None
     resourceObject=/FortniteCoreUI/UI/Foundation/Textures/Icons/Checks/T_UI_Icon_Check_MP_64
     tiling=NoTile
     imageSize=<64 64>
     margin=<ltrb 0 0 0 0>
     outlineSettings {
      bUseBrushTransparency=false
      roundingType=HalfHeightRadius
      width=0
      color {colorUseRule=UseColor_Specified, specifiedColor=<rgba 0 0 0 0>}
      cornerRadii=<0 0 0 0>
     }
     tintColor {colorUseRule=UseColor_Specified, specifiedColor=<rgba 1 1 1 1>}
     uVRegion {bIsValid=false, max=<0 0>, min=<0 0>}
    }
    colorAndOpacity=<rgba 0.0343 1 0.227 1>
    renderTransform {angle=0, scale=<0.6 0.6>, shear=<0 0>, translation=<9 0>}
   }
   slot {class=OverlaySlot, properties {horizontalAlignment=HAlign_Center, verticalAlignment=VAlign_Center, padding=<ltrb -10 -32 -10 -32>}}
  }
  Image_Glow {
   widgetClassPath=&Image
   bIsVariable=true
   parent=Overlay_2
   widget=Image_Glow
   properties {
    renderOpacity=0
    brush {
     bIsDynamicallyLoaded=false
     drawAs=Image
     imageType=NoImage
     mirroring=NoMirror
     resourceName=None
     resourceObject=&M_UI_TextGlow_Add
     tiling=NoTile
     imageSize=<130 130>
     margin=<ltrb 0 0 0 0>
     outlineSettings {
      bUseBrushTransparency=false
      roundingType=HalfHeightRadius
      width=0
      color {colorUseRule=UseColor_Specified, specifiedColor=<rgba 0 0 0 0>}
      cornerRadii=<0 0 0 0>
     }
     tintColor {colorUseRule=UseColor_Specified, specifiedColor=<rgba 0 1 0.1443 1>}
     uVRegion {bIsValid=false, max=<0 0>, min=<0 0>}
    }
    renderTransform {angle=0, scale=<1.3 1.3>, shear=<0 0>, translation=<0 0>}
   }
   slot {class=OverlaySlot, properties {horizontalAlignment=HAlign_Center, verticalAlignment=VAlign_Center, padding=<ltrb -100 -100 -100 -100>}}
  }
  Image_Spark {
   widgetClassPath=&Image
   bIsVariable=true
   parent=Overlay_2
   widget=Image_Spark
   properties {
    renderOpacity=0
    brush {
     bIsDynamicallyLoaded=false
     drawAs=Image
     imageType=NoImage
     mirroring=NoMirror
     resourceName=None
     resourceObject=/Game/UI/InGame/Quest/Materials/M_UI_SparkStar
     tiling=NoTile
     imageSize=<120 120>
     margin=<ltrb 0 0 0 0>
     outlineSettings {
      bUseBrushTransparency=false
      roundingType=HalfHeightRadius
      width=0
      color {colorUseRule=UseColor_Specified, specifiedColor=<rgba 0 0 0 0>}
      cornerRadii=<0 0 0 0>
     }
     tintColor {colorUseRule=UseColor_Specified, specifiedColor=<rgba 1 1 1 1>}
     uVRegion {bIsValid=false, max=<0 0>, min=<0 0>}
    }
   }
   slot {class=OverlaySlot, properties {horizontalAlignment=HAlign_Center, verticalAlignment=VAlign_Center, padding=<ltrb -75 -75 -75 -75>}}
  }
  Overlay_AddedProgress {
   widgetClassPath=&Overlay
   bIsVariable=true
   parent=WidgetSwitcher_Complete
   widget=Overlay_AddedProgress
   slot {class=WidgetSwitcherSlot}
  }
  RichText_ObjectiveAdded {
   widgetClassPath=&CommonRichTextBlock
   bIsVariable=true
   parent=Overlay_AddedProgress
   widget=RichText_ObjectiveAdded
   properties {
    justification=Center
    mobileTextBlockScale=1.2
    renderOpacity=0
    text="+50 "
    textStyleSet=&AthenaChallengeObjectiveProgressStyles
    defaultTextStyleOverride {
     overflowPolicy=Clip
     transformPolicy=None
     colorAndOpacity {colorUseRule=UseColor_Specified, specifiedColor=<rgba 0.2122 0.6514 1 1>}
     font {
      bForceMonospaced=false
      bMaterialIsStencil=false
      fontMaterial=null
      fontObject=&HeadingNow
      letterSpacing=20
      monospacedWidth=1
      size=12.75
      skewAmount=0
      typefaceFontName=76 Bold
      outlineSettings {
       bApplyOutlineToDropShadows=false
       bMiteredCorners=false
       bSeparateFillAlpha=false
       outlineMaterial=null
       outlineSize=1
       outlineColor=<rgba 0.026 0.026 0.026 0.5>
      }
     }
     highlightColor {colorUseRule=UseColor_Specified, specifiedColor=<rgba 0 0 0 1>}
     highlightShape {
      bIsDynamicallyLoaded=false
      drawAs=Image
      imageType=NoImage
      mirroring=NoMirror
      resourceName=None
      resourceObject=null
      tiling=NoTile
      imageSize=<32 32>
      margin=<ltrb 0 0 0 0>
      outlineSettings {
       bUseBrushTransparency=false
       roundingType=HalfHeightRadius
       width=0
       color {colorUseRule=UseColor_Specified, specifiedColor=<rgba 0 0 0 0>}
       cornerRadii=<0 0 0 0>
      }
      tintColor {colorUseRule=UseColor_Specified, specifiedColor=<rgba 1 1 1 1>}
      uVRegion {bIsValid=false, max=<0 0>, min=<0 0>}
     }
     selectedBackgroundColor {colorUseRule=UseColor_Specified, specifiedColor=<rgba 1 0 1 1>}
     shadowColorAndOpacity=<rgba 0.0314 0.0314 0.0314 0>
     shadowOffset=<2 2>
     strikeBrush {
      bIsDynamicallyLoaded=false
      drawAs=Image
      imageType=NoImage
      mirroring=NoMirror
      resourceName=None
      resourceObject=null
      tiling=NoTile
      imageSize=<32 32>
      margin=<ltrb 0 0 0 0>
      outlineSettings {
       bUseBrushTransparency=false
       roundingType=HalfHeightRadius
       width=0
       color {colorUseRule=UseColor_Specified, specifiedColor=<rgba 0 0 0 0>}
       cornerRadii=<0 0 0 0>
      }
      tintColor {colorUseRule=UseColor_Specified, specifiedColor=<rgba 1 1 1 1>}
      uVRegion {bIsValid=false, max=<0 0>, min=<0 0>}
     }
     underlineBrush {
      bIsDynamicallyLoaded=false
      drawAs=Image
      imageType=NoImage
      mirroring=NoMirror
      resourceName=None
      resourceObject=null
      tiling=NoTile
      imageSize=<32 32>
      margin=<ltrb 0 0 0 0>
      outlineSettings {
       bUseBrushTransparency=false
       roundingType=HalfHeightRadius
       width=0
       color {colorUseRule=UseColor_Specified, specifiedColor=<rgba 0 0 0 0>}
       cornerRadii=<0 0 0 0>
      }
      tintColor {colorUseRule=UseColor_Specified, specifiedColor=<rgba 1 1 1 1>}
      uVRegion {bIsValid=false, max=<0 0>, min=<0 0>}
     }
    }
    renderTransform {angle=0, scale=<0.9 0.9>, shear=<0 0>, translation=<12.6766 0>}
   }
   slot {class=OverlaySlot, properties {horizontalAlignment=HAlign_Right, verticalAlignment=VAlign_Bottom, padding=<ltrb 0 0 50 0>}}
  }
  RichText_Objectives_Current {
   widgetClassPath=&CommonRichTextBlock
   bIsVariable=true
   parent=Overlay_AddedProgress
   widget=RichText_Objectives_Current
   properties {
    justification=Center
    mobileTextBlockScale=1.2
    renderOpacity=0
    text=100
    textStyleSet=&AthenaChallengeObjectiveProgressStyles
    defaultTextStyleOverride {
     overflowPolicy=Clip
     transformPolicy=None
     colorAndOpacity {colorUseRule=UseColor_Specified, specifiedColor=<rgba 0.2122 0.6514 1 1>}
     font {
      bForceMonospaced=false
      bMaterialIsStencil=false
      fontMaterial=null
      fontObject=&HeadingNow
      letterSpacing=20
      monospacedWidth=1
      size=12.75
      skewAmount=0
      typefaceFontName=76 Bold
      outlineSettings {
       bApplyOutlineToDropShadows=false
       bMiteredCorners=false
       bSeparateFillAlpha=false
       outlineMaterial=null
       outlineSize=1
       outlineColor=<rgba 0.026 0.026 0.026 0.5>
      }
     }
     highlightColor {colorUseRule=UseColor_Specified, specifiedColor=<rgba 0 0 0 1>}
     highlightShape {
      bIsDynamicallyLoaded=false
      drawAs=Image
      imageType=NoImage
      mirroring=NoMirror
      resourceName=None
      resourceObject=null
      tiling=NoTile
      imageSize=<32 32>
      margin=<ltrb 0 0 0 0>
      outlineSettings {
       bUseBrushTransparency=false
       roundingType=HalfHeightRadius
       width=0
       color {colorUseRule=UseColor_Specified, specifiedColor=<rgba 0 0 0 0>}
       cornerRadii=<0 0 0 0>
      }
      tintColor {colorUseRule=UseColor_Specified, specifiedColor=<rgba 1 1 1 1>}
      uVRegion {bIsValid=false, max=<0 0>, min=<0 0>}
     }
     selectedBackgroundColor {colorUseRule=UseColor_Specified, specifiedColor=<rgba 1 0 1 1>}
     shadowColorAndOpacity=<rgba 0.0314 0.0314 0.0314 0>
     shadowOffset=<2 2>
     strikeBrush {
      bIsDynamicallyLoaded=false
      drawAs=Image
      imageType=NoImage
      mirroring=NoMirror
      resourceName=None
      resourceObject=null
      tiling=NoTile
      imageSize=<32 32>
      margin=<ltrb 0 0 0 0>
      outlineSettings {
       bUseBrushTransparency=false
       roundingType=HalfHeightRadius
       width=0
       color {colorUseRule=UseColor_Specified, specifiedColor=<rgba 0 0 0 0>}
       cornerRadii=<0 0 0 0>
      }
      tintColor {colorUseRule=UseColor_Specified, specifiedColor=<rgba 1 1 1 1>}
      uVRegion {bIsValid=false, max=<0 0>, min=<0 0>}
     }
     underlineBrush {
      bIsDynamicallyLoaded=false
      drawAs=Image
      imageType=NoImage
      mirroring=NoMirror
      resourceName=None
      resourceObject=null
      tiling=NoTile
      imageSize=<32 32>
      margin=<ltrb 0 0 0 0>
      outlineSettings {
       bUseBrushTransparency=false
       roundingType=HalfHeightRadius
       width=0
       color {colorUseRule=UseColor_Specified, specifiedColor=<rgba 0 0 0 0>}
       cornerRadii=<0 0 0 0>
      }
      tintColor {colorUseRule=UseColor_Specified, specifiedColor=<rgba 1 1 1 1>}
      uVRegion {bIsValid=false, max=<0 0>, min=<0 0>}
     }
    }
   }
   slot {class=OverlaySlot, properties {horizontalAlignment=HAlign_Right, verticalAlignment=VAlign_Bottom}}
  }
  HorizontalBoxHotkeyPrompt {
   widgetClassPath=&HorizontalBox
   bIsVariable=true
   parent=MainContent
   widget=HorizontalBoxHotkeyPrompt
   properties {visibility=Collapsed}
   slot {class=GridSlot, properties {horizontalAlignment=HAlign_Right, row=7, padding=<ltrb 4 4 4 4>}}
  }
  Text_MapPrompt {
   widgetClassPath=&CommonTextBlock
   bIsVariable=true
   parent=HorizontalBoxHotkeyPrompt
   widget=Text_MapPrompt
   properties {
    mobileFontSizeMultiplier=1.2
    text="Test  to Track Quest"
    textTransformPolicy=ToUpper
    font {
     bForceMonospaced=false
     bMaterialIsStencil=false
     fontMaterial=null
     fontObject=&HeadingNow
     letterSpacing=-100
     monospacedWidth=0.7333
     size=10.5
     skewAmount=0
     typefaceFontName=75 Medium
     outlineSettings {
      bApplyOutlineToDropShadows=false
      bMiteredCorners=false
      bSeparateFillAlpha=false
      outlineMaterial=null
      outlineSize=1
      outlineColor=<rgba 0 0 0 0.3>
     }
    }
    shadowOffset=<0 0>
   }
   slot {class=HorizontalBoxSlot, properties {horizontalAlignment=HAlign_Right, verticalAlignment=VAlign_Center, padding=<ltrb 0 0 8 0>}}
  }
  KeybindWidget {
   widgetClassPath=/FortUILibrary/Input/KeybindWidget
   bIsVariable=true
   parent=HorizontalBoxHotkeyPrompt
   widget=KeybindWidget
   properties {forcedHoldKeybindStatus=ForcedHold}
   slot {class=HorizontalBoxSlot, properties {horizontalAlignment=HAlign_Right, verticalAlignment=VAlign_Top}}
  }
  XpSubNotification {
   widgetClassPath=&/XpSubNotification
   bInherited=true
   bIsVariable=true
   parent=MainContent
   widget=XpSubNotification
   properties {visibility=Collapsed}
   slot {class=GridSlot, properties {row=3, padding=<ltrb 0 5 0 0>}}
  }
  Button_MobileClickCatcher {
   widgetClassPath=&Button
   bInherited=true
   parent=Root
   widget=Button_MobileClickCatcher
   properties {isFocusable=false, backgroundColor=<rgba 1 1 1 0>, colorAndOpacity=<rgba 1 1 1 0>}
   slot {class=GridSlot}
  }
  Text_Header {widgetClassPath=&CommonTextBlock, bInherited=true, widget=null}
 ]
}
