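# Post-game widget that shows a "NEW QUESTS!" heading above a vertical entry box of QuestNotification widgets.
# C++ parent FortPostGameNewQuestsWidget raises play events that start short fade-and-slide animations for the heading and entry box.
# OnInitialized binds heading intro and entry box intro/outro completion to parent handlers; heading Outro has no callback.
# Destruct unbinds only the two entry box animations, leaving Intro bound.
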WidgetBlueprint /FortniteCoreUI/HUD/EndGame/FortPostGameNewQuestsWidget
animations [
 EntryBoxIntro {
  binding_count=1
  length_seconds=0.25
  object=&self:EntryBoxIntro
  sequence {bindings [EntryBox_NewQuests {
     id=#1
     possessed_class=DynamicEntryBox
     tracks [
      {
       type=MovieSceneFloatTrack
       display_name=Render Opacity
       sections [{type=MovieSceneFloatSection, channels [None {type=MovieSceneScriptingFloatChannel, keys [[0,0],[5,1]]}]}]
      }
      {
       type=MovieScene2DTransformTrack
       sections [{
         type=MovieScene2DTransformSection
         channels [
          Translation.X {type=MovieSceneScriptingFloatChannel, keys [[0,-64],[5.0003,0]]}
          Translation.Y {type=MovieSceneScriptingFloatChannel}
          Angle {type=MovieSceneScriptingFloatChannel}
          Scale.X {type=MovieSceneScriptingFloatChannel}
          Scale.Y {type=MovieSceneScriptingFloatChannel}
          Shear.X {type=MovieSceneScriptingFloatChannel, keys [[0,-4],[5.0003,0]]}
          Shear.Y {type=MovieSceneScriptingFloatChannel}
         ]
        }]
      }
     ]
    }]}
 }
 EntryBoxOutro {
  binding_count=1
  length_seconds=0.25
  object=&self:EntryBoxOutro
  sequence {bindings [EntryBox_NewQuests {
     id=#1
     possessed_class=DynamicEntryBox
     tracks [
      {
       type=MovieSceneFloatTrack
       display_name=Render Opacity
       sections [{type=MovieSceneFloatSection, channels [None {type=MovieSceneScriptingFloatChannel, keys [[0,1],[5,0]]}]}]
      }
      {
       type=MovieScene2DTransformTrack
       sections [{
         type=MovieScene2DTransformSection
         channels [
          Translation.X {type=MovieSceneScriptingFloatChannel, keys [[0,0],[5.0003,64]]}
          Translation.Y {type=MovieSceneScriptingFloatChannel}
          Angle {type=MovieSceneScriptingFloatChannel}
          Scale.X {type=MovieSceneScriptingFloatChannel}
          Scale.Y {type=MovieSceneScriptingFloatChannel}
          Shear.X {type=MovieSceneScriptingFloatChannel, keys [[0,0],[5.0003,-4]]}
          Shear.Y {type=MovieSceneScriptingFloatChannel, const [0,1]}
         ]
        }]
      }
     ]
    }]}
 }
 Intro {
  binding_count=1
  length_seconds=0.5
  object=&self:Intro
  sequence {bindings [NewQuestsText {
     id=#2
     possessed_class=CommonTextBlock
     tracks [
      {
       type=MovieScene2DTransformTrack
       sections [{
         type=MovieScene2DTransformSection
         channels [
          Translation.X {type=MovieSceneScriptingFloatChannel, const [0,3]}
          Translation.Y {type=MovieSceneScriptingFloatChannel, keys [[0,40],[5,0],[10,0]]}
          Angle {type=MovieSceneScriptingFloatChannel}
          Scale.X {type=MovieSceneScriptingFloatChannel}
          Scale.Y {type=MovieSceneScriptingFloatChannel}
          Shear.X {type=MovieSceneScriptingFloatChannel}
          Shear.Y {type=MovieSceneScriptingFloatChannel}
         ]
        }]
      }
      {
       type=MovieSceneFloatTrack
       display_name=Render Opacity
       sections [{type=MovieSceneFloatSection, channels [None {type=MovieSceneScriptingFloatChannel, keys [[0,0],[5,1],[10,1]]}]}]
      }
     ]
    }]}
 }
 Outro {
  binding_count=1
  length_seconds=0.25
  object=&self:Outro
  sequence {bindings [NewQuestsText {
     id=#3
     possessed_class=CommonTextBlock
     tracks [
      {
       type=MovieScene2DTransformTrack
       sections [{
         type=MovieScene2DTransformSection
         channels [
          Translation.X {type=MovieSceneScriptingFloatChannel, keys [[0,0],[5.0003,50]]}
          Translation.Y {type=MovieSceneScriptingFloatChannel}
          Angle {type=MovieSceneScriptingFloatChannel}
          Scale.X {type=MovieSceneScriptingFloatChannel}
          Scale.Y {type=MovieSceneScriptingFloatChannel}
          Shear.X {type=MovieSceneScriptingFloatChannel}
          Shear.Y {type=MovieSceneScriptingFloatChannel}
         ]
        }]
      }
      {
       type=MovieSceneFloatTrack
       display_name=Render Opacity
       sections [{type=MovieSceneFloatSection, channels [None {type=MovieSceneScriptingFloatChannel, keys [[0,1],[5,0]]}]}]
      }
     ]
    }]}
 }
]
graphs [EventGraph {nodes [
   n1 = HandleEntryBoxIntroFinished() {is=title K2Node_CustomEvent}
   > HandlePlayNewQuestsEntryBoxIntroAnimFinished(self:Fort Post Game New Quests Widget Object Reference)
   n2 = HandleEntryBoxOutroFinished() {is=title K2Node_CustomEvent}
   > HandlePlayNewQuestsEntryBoxOutroAnimFinished(self:Fort Post Game New Quests Widget Object Reference)
   n3 = HandleIntroFinished() {is=title K2Node_CustomEvent}
   > HandlePlayNewQuestsIntroAnimFinished(self:Fort Post Game New Quests Widget Object Reference)
   Event OnPlayNewQuestsEntryBoxIntro() {is=title K2Node_Event}
   > PlayAnimation(InAnimation=Get EntryBoxIntro() -> (EntryBoxIntro:Widget Animation Object Reference))
   Event OnInitialized()
   > Parent: OnInitialized() {is=title K2Node_CallParentFunction}
   > BindToAnimationFinished(Animation=Get EntryBoxIntro(), Delegate=@n1)
   > BindToAnimationFinished(Animation=Get EntryBoxOutro() -> (EntryBoxOutro:Widget Animation Object Reference), Delegate=@n2)
   > BindToAnimationFinished(Animation=Get Intro(), Delegate=@n3)
   Event Destruct()
   > Parent: Destruct()
   > UnbindAllFromAnimationFinished!(execute, self:User Widget Object Reference, Animation:Widget Animation Object Reference=Get EntryBoxIntro()) -> (then)
   > UnbindAllFromAnimationFinished(Animation=Get EntryBoxOutro())
   Event OnPlayNewQuestsEntryBoxOutro() {is=title K2Node_Event}
   > PlayAnimation(InAnimation=Get EntryBoxOutro())
   Event OnPlayIntro() {is=title K2Node_Event}
   > PlayAnimation(InAnimation=Get Intro())
   Event OnPlayOutro() {is=title K2Node_Event}
   > PlayAnimation(InAnimation=Get Outro() -> (Outro:Widget Animation Object Reference))
  ]}]
variables []
widgets {
 info {inheritedWidgetCount=1, parentClass=/Script/FortniteUI.FortPostGameNewQuestsWidget, rootWidgetClass=&Overlay, widgetCount=4}
 widgets [
  Overlay_29 {widgetClassPath=&Overlay, widget=Overlay_29}
  VerticalBox_45 {widgetClassPath=&VerticalBox, parent=Overlay_29, widget=VerticalBox_45, slot {class=OverlaySlot}}
  NewQuestsText {
   widgetClassPath=&CommonTextBlock
   parent=VerticalBox_45
   widget=NewQuestsText
   properties {
    text="NEW QUESTS!"
    font {
     bForceMonospaced=false
     bMaterialIsStencil=false
     fontMaterial=null
     fontObject=&HeadingNow
     letterSpacing=0
     monospacedWidth=1
     size=16.5
     skewAmount=-0.09
     typefaceFontName=86 Bold Italic
     outlineSettings {
      bApplyOutlineToDropShadows=false
      bMiteredCorners=false
      bSeparateFillAlpha=false
      outlineMaterial=null
      outlineSize=0
      outlineColor=<rgba 0 0 0 1>
     }
    }
    renderTransform {angle=0, scale=<1 1>, shear=<-12 0>, translation=<0 0>}
    shadowOffset=<0 0>
   }
   slot {class=VerticalBoxSlot}
  }
  EntryBox_NewQuests {
   widgetClassPath=&DynamicEntryBox
   bInherited=true
   bIsVariable=true
   parent=VerticalBox_45
   widget=EntryBox_NewQuests
   properties {entryBoxType=Vertical, entryWidgetClass=/FortniteCoreUI/UI/InGame/EventNotifications/QuestNotification.QuestNotification_C}
   slot {class=VerticalBoxSlot}
  }
 ]
}
