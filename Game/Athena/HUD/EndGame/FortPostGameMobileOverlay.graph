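# Mobile post-game overlay widget with a right-aligned row of spectate, report player, map, push-to-talk, emote and hamburger menu buttons, plus a playlist extension slot.
# Spectate and report clicks broadcast events through the global event router; map toggles the fullscreen map.
# UpdateButtonsAccordingToPostGameScreen sets visibility from screen flags and ShouldShowSpectate, hiding spectate and report on the elimination screen.
# On Construct, an S39 console variable selects the row padding.
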WidgetBlueprint /Game/Athena/HUD/EndGame/FortPostGameMobileOverlay
graphs [
 EventGraph {nodes [
   "On Clicked (Button_Spectate)"!() -> (OutputDelegate, then, Button:Common Button Base Object Reference) {is=title K2Node_ComponentBoundEvent}
   > Broadcast Event!(
    execute
    self:Gameplay Event Router Component Object Reference=GetGlobalEventRouter()
    EventData:Match XPScreen Event Toggle Spectating Structure=Make Match XPScreen Event Toggle Spectating!() -> (MatchXPScreenEvent_ToggleSpectating:Match XPScreen Event Toggle Spectating Structure) {is=title K2Node_MakeStruct}
    EventContext:Object Reference=Self-Reference()
   ) -> (then, ReturnValue:bool)
   "On Clicked (Button_OpenMap)"!() -> (OutputDelegate, then, Button:Common Button Base Object Reference) {is=title K2Node_ComponentBoundEvent}
   > ToggleFullscreenMap(self:Fort HUDContext Object Reference=Get FortHUDContext!() -> (ReturnValue:Fort HUDContext Object Reference) {is=title K2Node_GetBlueprintContextEx})
   "On Clicked (Button_ReportPlayer)"!() -> (OutputDelegate, then, Button:Common Button Base Object Reference) {is=title K2Node_ComponentBoundEvent}
   > Broadcast Event(
    self=GetGlobalEventRouter()
    EventData:Match XPScreen Event Report Player Structure=Make Match XPScreen Event Report Player!() -> (MatchXPScreenEvent_ReportPlayer:Match XPScreen Event Report Player Structure) {is=title K2Node_MakeStruct}
    EventContext=Self-Reference()
   )
   Event Construct()
   > UpdateS39PostMatchVisibility()
   n1 = Event UpdateButtonsAccordingToPostGameScreen!() -> (OutputDelegate, then, PostGameScreenWidget:Fort Post Game Screen Object Reference) {is=title K2Node_Event}
   > Is Valid(InputObject=@n1.PostGameScreenWidget)
   n2 = Get bBlockFullscreenMapOpening!(self:Fort Post Game Screen Object Reference=@n1.PostGameScreenWidget) -> (bBlockFullscreenMapOpening:bool)
   > SetVisibility(
    self=Get Button_OpenMap() -> (Button_OpenMap:WBP UIKit Button Generic Object Reference)
    InVisibility=Select(Option 0=Collapsed, Option 1=Visible, Index=NOT Boolean(A=@n2))
   )
   > SetVisibility(
    self=Get FortPlaylistUIExtensionSlot_LeftOfPostGameButtons() -> (FortPlaylistUIExtensionSlot_LeftOfPostGameButtons:Fort Playlist UIExtension Slot Object Reference)
    InVisibility=Select(
     Option 0=Collapsed
     Option 1=Visible
     Index=Get bDoesShowSocialTextShat!(self:Fort Post Game Screen Object Reference=@n1.PostGameScreenWidget) -> (bDoesShowSocialTextShat:bool)
    )
   )
   > n3 = "IsA ( soft )"!(
    execute
    Object:Object Reference=@n1.PostGameScreenWidget
    SoftClass:Object Soft Class Reference=&/FortEliminationScreen.FortEliminationScreen_C
   ) -> (True:Exec, False:Exec)
   n4 = ShouldShowSpectate!(execute=@n3.False) -> (then, ReturnValue:bool)
   > SetVisibility(
    self=Get Button_Spectate() -> (Button_Spectate:WBP UIKit Button Generic Object Reference)
    InVisibility=Select(Option 0=Collapsed, Option 1=Visible, Index=@n4)
   )
   > SetVisibility(self=Get Button_ReportPlayer() -> (Button_ReportPlayer:WBP UIKit Button Generic Object Reference), InVisibility=Visible)
   SetVisibility(execute=@n3.True, self=[Get Button_Spectate(), Get Button_ReportPlayer()], InVisibility=Collapsed)
  ]}
 UpdateS39PostMatchVisibility {nodes [
   UpdateS39PostMatchVisibility() {is=title K2Node_FunctionEntry}
   > n1 = GetConsoleVariableBoolValue!(execute, VariableName:String=Athena.MatchXPScreenWidget.S39EnablePostGameUpdates) -> (then, ReturnValue:bool)
   > Set IsUsingS39PostmatchUpdates(IsUsingS39PostmatchUpdates:bool=@n1) -> (Output_Get:bool)
   > SetPadding!(
    execute
    self:Overlay Slot Object Reference=SlotAsOverlaySlot(Widget=Get HBox_MobileMenu() -> (HBox_MobileMenu:Horizontal Box Object Reference))
    InPadding:Margin Structure=Select(
     Option 0:Margin Structure=Get Padding_WithoutPostMatchUpdates() -> (Padding_WithoutPostMatchUpdates:Margin Structure)
     Option 1:Margin Structure=Get Padding_WithPostMatchUpdates() -> (Padding_WithPostMatchUpdates:Margin Structure)
     Index=Get IsUsingS39PostmatchUpdates() -> (IsUsingS39PostmatchUpdates:bool)
    ) -> (ReturnValue:Margin Structure)
   ) -> (then)
  ]}
]
variables [
 IsUsingS39PostmatchUpdates {type=boolean}
 Padding_WithoutPostMatchUpdates {type=Margin, default=<ltrb 0 280 120 0>}
 Padding_WithPostMatchUpdates {type=Margin, default=<ltrb 0 0 120 0>}
]
widgets {
 info {inheritedWidgetCount=1, parentClass=/Script/FortniteUI.FortPostGameMobileButtonOverlay, rootWidgetClass=&Overlay, widgetCount=11}
 widgets [
  Overlay_30 {widgetClassPath=&Overlay, widget=Overlay_30}
  HBox_MobileMenu {
   widgetClassPath=&HorizontalBox
   bIsVariable=true
   parent=Overlay_30
   widget=HBox_MobileMenu
   slot {class=OverlaySlot, properties {horizontalAlignment=HAlign_Right, padding=<ltrb 0 280 120 0>}}
  }
  FortPlaylistUIExtensionSlot_LeftOfPostGameButtons {
   widgetClassPath=/Script/FortniteUI.FortPlaylistUIExtensionSlot
   bIsVariable=true
   parent=HBox_MobileMenu
   widget=FortPlaylistUIExtensionSlot_LeftOfPostGameButtons
   properties {
    entryWidgetClass=/Game/Athena/HUD/FortHUDElementWrapper.FortHUDElementWrapper_C
    numDesignerPreviewEntries=0
    extensionSlots [LeftPostGameMobileOverlay]
   }
   slot {class=HorizontalBoxSlot}
  }
  Button_Spectate {
   widgetClassPath=&WBP_UIKit_Button_Generic
   bIsVariable=true
   parent=HBox_MobileMenu
   widget=Button_Spectate
   properties {
    backgroundClass=null
    backgroundMaterial=null
    icon=/Game/Athena/HUD/Art/T_UI_Icon_Spectate
    iconMaterial=/FortUILibrary/UIKit/Buttons/MaterialInstances/GenericButtons/MI_UIKit_GenericIcon_Home
    maxHeight=102
    maxWidth=102
    style=/FortUILibrary/Foundation/Buttons/ButtonStyles/ButtonStyle-Mobile.ButtonStyle-Mobile_C
    text=""
    buttonPadding=<ltrb 0 0 0 0>
    iconPadding=<ltrb -100 -100 -100 -100>
    iconSize=<100 100>
    secondTextPadding=<ltrb 0 0 0 0>
   }
   slot {class=HorizontalBoxSlot, properties {horizontalAlignment=HAlign_Center, verticalAlignment=VAlign_Center, padding=<ltrb 10 0 0 0>}}
  }
  Button_ReportPlayer {
   widgetClassPath=&WBP_UIKit_Button_Generic
   bIsVariable=true
   parent=HBox_MobileMenu
   widget=Button_ReportPlayer
   properties {
    backgroundClass=null
    backgroundMaterial=null
    icon=/Game/Athena/HUD/Art/T_UI_Icon_ReportPlayer
    iconMaterial=&MI_UIKit_GenericIcon_Home
    maxHeight=102
    maxWidth=102
    style=&ButtonStyle-Mobile_C
    text=""
    buttonPadding=<ltrb 0 0 0 0>
    iconPadding=<ltrb -100 -100 -100 -100>
    iconSize=<80 80>
    secondTextPadding=<ltrb 0 0 0 0>
   }
   slot {class=HorizontalBoxSlot, properties {horizontalAlignment=HAlign_Center, verticalAlignment=VAlign_Center, padding=<ltrb 10 0 0 0>}}
  }
  Button_OpenMap {
   widgetClassPath=&WBP_UIKit_Button_Generic
   bIsVariable=true
   parent=HBox_MobileMenu
   widget=Button_OpenMap
   properties {
    backgroundClass=null
    backgroundMaterial=null
    icon=/Game/Athena/HUD/Art/T_UI_Icon_Map
    iconMaterial=&MI_UIKit_GenericIcon_Home
    maxHeight=102
    maxWidth=102
    style=&ButtonStyle-Mobile_C
    text=""
    buttonPadding=<ltrb 0 0 0 0>
    iconPadding=<ltrb -100 -100 -100 -100>
    iconSize=<80 80>
    secondTextPadding=<ltrb 0 0 0 0>
   }
   slot {class=HorizontalBoxSlot, properties {horizontalAlignment=HAlign_Center, verticalAlignment=VAlign_Center, padding=<ltrb 10 0 0 0>}}
  }
  Button_PushToTalk {
   widgetClassPath=/Game/Athena/HUD/Bacchus/Button_PushToTalk
   bInherited=true
   bIsVariable=true
   parent=HBox_MobileMenu
   widget=Button_PushToTalk
   properties {buttonSize=100}
   slot {class=HorizontalBoxSlot, properties {horizontalAlignment=HAlign_Right, verticalAlignment=VAlign_Top, padding=<ltrb 10 0 0 0>}}
  }
  Button_EmoteMenu {
   widgetClassPath=/Game/Athena/HUD/Bacchus/Button_EmoteMenu
   bIsVariable=true
   parent=HBox_MobileMenu
   widget=Button_EmoteMenu
   properties {
    bUpdatesWithHUDState=false
    buttonSize=100
    defaultIcon=/Game/Athena/HUD/Atlas/Bacchus/T_UI_Emote_Sprite
    elementTag {tagName=HUD.Bacchus.TouchButton.EmoteMenu}
   }
   slot {class=HorizontalBoxSlot, properties {horizontalAlignment=HAlign_Right, verticalAlignment=VAlign_Top, padding=<ltrb 10 0 0 0>}}
  }
  Spacer_54 {widgetClassPath=&Spacer, parent=HBox_MobileMenu, widget=Spacer_54, properties {size=<100 1>}, slot {class=HorizontalBoxSlot}}
  FortVisualAttachment_57 {
   widgetClassPath=&FortVisualAttachment
   parent=HBox_MobileMenu
   widget=FortVisualAttachment_57
   properties {bOverride_HeightOverride=true, bOverride_WidthOverride=true, heightOverride=200, widthOverride=200, contentAnchor=<0.5 0.5>}
   slot {class=HorizontalBoxSlot, properties {verticalAlignment=VAlign_Center}}
  }
  Button_HamburgerMenu {
   widgetClassPath=/Game/Athena/HUD/Bacchus/HamburgerButton
   bIsVariable=true
   parent=FortVisualAttachment_57
   widget=Button_HamburgerMenu
   properties {elementTag {tagName=HUD.Bacchus.Hamburger}}
   slot {class=SizeBoxSlot, properties {horizontalAlignment=HAlign_Right, verticalAlignment=VAlign_Top}}
  }
 ]
}
