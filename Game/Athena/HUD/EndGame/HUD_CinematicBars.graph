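# HUD widget for end-game cinematic letterbox bars: dark top and bottom strips, sized by spacers and top-padded safe zones.
# Construct hides the widget; EventIntro makes it hit-test invisible and plays Intro, which grows both bars vertically by animating desired size scale.
# Despite its name, EventToggleBlur only fades a translucent black Darken layer, playing BGBlur forward or reverse via the isBlurred flag.
# A Subtract Safezone macro on the container has no incoming exec connection.
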WidgetBlueprint /Game/Athena/HUD/EndGame/HUD_CinematicBars
animations [
 Intro {
  binding_count=2
  length_seconds=1
  object=&self:Intro
  sequence {bindings [
    CommonBorder_1 {
     id=#1
     possessed_class=CommonBorder
     tracks [{
       type=MovieSceneFloatVectorTrack
       display_name=Desired Size Scale
       sections [{
         type=MovieSceneFloatVectorSection
         channels [
          Vector.X {type=MovieSceneScriptingFloatChannel}
          Vector.Y {type=MovieSceneScriptingFloatChannel, keys [[0,0],[20,1]]}
         ]
        }]
      }]
    }
    CommonBorder_0 {
     id=#2
     possessed_class=CommonBorder
     tracks [{
       type=MovieSceneFloatVectorTrack
       display_name=Desired Size Scale
       sections [{
         type=MovieSceneFloatVectorSection
         channels [
          Vector.X {type=MovieSceneScriptingFloatChannel}
          Vector.Y {type=MovieSceneScriptingFloatChannel, keys [[0,0],[20,1]]}
         ]
        }]
      }]
    }
   ]}
 }
 BGBlur {
  binding_count=1
  length_seconds=0.4
  object=&self:BGBlur
  sequence {bindings [Darken {
     id=#3
     possessed_class=Image
     tracks [{
       type=MovieSceneFloatTrack
       display_name=Render Opacity
       sections [{type=MovieSceneFloatSection, channels [None {type=MovieSceneScriptingFloatChannel, keys [[0,0],[8,1]]}]}]
      }]
    }]}
 }
]
graphs [EventGraph {nodes [
   EventIntro() {is=title K2Node_CustomEvent}
   > SetVisibility(InVisibility=HitTestInvisible)
   > PlayAnimationForward(InAnimation=Get Intro())
   EventToggleBlur() {is=title K2Node_CustomEvent}
   > n1 = Branch(Condition=Get isBlurred() -> (isBlurred:bool))
   n2 = Get BGBlur() -> (BGBlur:Widget Animation Object Reference)
   > PlayAnimationReverse(InAnimation=@n2)
   > Set isBlurred(isBlurred:bool=false) -> (Output_Get:bool)
   PlayAnimationForward(execute=@n1.else, InAnimation=@n2)
   > Set isBlurred(isBlurred=true)
   n3 = Event Construct()
   SetVisibility(execute=[@n3.then, @n4.Out], InVisibility=Hidden)
   n4 = Subtract Safezone!(In:Exec, OverlaySlot:Widget Object Reference=Get Overlay_Container() -> (Overlay_Container:Overlay Object Reference), ScaleUp:bool) -> (Out:Exec) {is=title K2Node_MacroInstance}
  ]}]
variables [isBlurred {type=boolean}]
widgets {
 info {inheritedWidgetCount=0, parentClass=&UserWidget, rootWidgetClass=&Overlay, widgetCount=14}
 widgets [
  Overlay_Container {widgetClassPath=&Overlay, bIsVariable=true, widget=Overlay_Container}
  Darken {
   widgetClassPath=&Image
   parent=Overlay_Container
   widget=Darken
   properties {
    renderOpacity=0
    visibility=SelfHitTestInvisible
    brush {
     bIsDynamicallyLoaded=false
     drawAs=Image
     imageType=NoImage
     mirroring=NoMirror
     resourceName=None
     resourceObject=null
     tiling=NoTile
     imageSize=<32 32>
     margin=<ltrb 0 0 0 0>
     outlineSettings {
      bUseBrushTransparency=false
      roundingType=HalfHeightRadius
      width=0
      color {colorUseRule=UseColor_Specified, specifiedColor=<rgba 0 0 0 0>}
      cornerRadii=<0 0 0 0>
     }
     tintColor {colorUseRule=UseColor_Specified, specifiedColor=<rgba 0 0 0 0.25>}
     uVRegion {bIsValid=false, max=<0 0>, min=<0 0>}
    }
   }
   slot {class=OverlaySlot, properties {horizontalAlignment=HAlign_Fill, verticalAlignment=VAlign_Fill}}
  }
  CommonBorder_0 {
   widgetClassPath=&CommonBorder
   parent=Overlay_Container
   widget=CommonBorder_0
   properties {padding=<ltrb 0 0 0 0>}
   slot {class=OverlaySlot, properties {horizontalAlignment=HAlign_Fill}}
  }
  Overlay_0 {widgetClassPath=&Overlay, parent=CommonBorder_0, widget=Overlay_0, slot {class=BorderSlot, properties {padding=<ltrb 0 0 0 0>}}}
  VerticalBox_0 {
   widgetClassPath=&VerticalBox
   parent=Overlay_0
   widget=VerticalBox_0
   slot {class=OverlaySlot, properties {horizontalAlignment=HAlign_Fill, verticalAlignment=VAlign_Fill}}
  }
  SafeZone_194 {
   widgetClassPath=&SafeZone
   bIsVariable=true
   parent=VerticalBox_0
   widget=SafeZone_194
   properties {padBottom=false, padLeft=false, padRight=false}
   slot {class=VerticalBoxSlot}
  }
  Spacer_67 {widgetClassPath=&Spacer, parent=VerticalBox_0, widget=Spacer_67, properties {size=<1 45>}, slot {class=VerticalBoxSlot}}
  Image_topbar {
   widgetClassPath=&Image
   bIsVariable=true
   parent=Overlay_0
   widget=Image_topbar
   properties {
    brush {
     bIsDynamicallyLoaded=false
     drawAs=Image
     imageType=NoImage
     mirroring=NoMirror
     resourceName=None
     resourceObject=null
     tiling=NoTile
     imageSize=<32 80>
     margin=<ltrb 0 0 0 0>
     outlineSettings {
      bUseBrushTransparency=false
      roundingType=HalfHeightRadius
      width=0
      color {colorUseRule=UseColor_Specified, specifiedColor=<rgba 0 0 0 0>}
      cornerRadii=<0 0 0 0>
     }
     tintColor {colorUseRule=UseColor_Specified, specifiedColor=<rgba 1 1 1 1>}
     uVRegion {bIsValid=false, max=<0 0>, min=<0 0>}
    }
    colorAndOpacity=<rgba 0.0003 0.0007 0.013 0.9>
   }
   slot {class=OverlaySlot, properties {horizontalAlignment=HAlign_Fill, verticalAlignment=VAlign_Fill}}
  }
  CommonBorder_1 {
   widgetClassPath=&CommonBorder
   parent=Overlay_Container
   widget=CommonBorder_1
   properties {padding=<ltrb 0 0 0 0>}
   slot {class=OverlaySlot, properties {horizontalAlignment=HAlign_Fill, verticalAlignment=VAlign_Bottom}}
  }
  Overlay_1 {widgetClassPath=&Overlay, parent=CommonBorder_1, widget=Overlay_1, slot {class=BorderSlot, properties {padding=<ltrb 0 0 0 0>}}}
  Image_bottombar {
   widgetClassPath=&Image
   bIsVariable=true
   parent=Overlay_1
   widget=Image_bottombar
   properties {
    brush {
     bIsDynamicallyLoaded=false
     drawAs=Image
     imageType=NoImage
     mirroring=NoMirror
     resourceName=None
     resourceObject=null
     tiling=NoTile
     imageSize=<32 48>
     margin=<ltrb 0 0 0 0>
     outlineSettings {
      bUseBrushTransparency=false
      roundingType=HalfHeightRadius
      width=0
      color {colorUseRule=UseColor_Specified, specifiedColor=<rgba 0 0 0 0>}
      cornerRadii=<0 0 0 0>
     }
     tintColor {colorUseRule=UseColor_Specified, specifiedColor=<rgba 1 1 1 1>}
     uVRegion {bIsValid=false, max=<0 0>, min=<0 0>}
    }
    colorAndOpacity=<rgba 0.0003 0.0007 0.013 0.9>
   }
   slot {class=OverlaySlot, properties {horizontalAlignment=HAlign_Fill, verticalAlignment=VAlign_Fill}}
  }
  VerticalBox {
   widgetClassPath=&VerticalBox
   parent=Overlay_1
   widget=VerticalBox
   slot {class=OverlaySlot, properties {horizontalAlignment=HAlign_Fill, verticalAlignment=VAlign_Fill}}
  }
  SafeZone {
   widgetClassPath=&SafeZone
   bIsVariable=true
   parent=VerticalBox
   widget=SafeZone
   properties {padBottom=false, padLeft=false, padRight=false}
   slot {class=VerticalBoxSlot}
  }
  Spacer {widgetClassPath=&Spacer, parent=VerticalBox, widget=Spacer, properties {size=<1 80>}, slot {class=VerticalBoxSlot}}
 ]
}
