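# Post-game HUD widget that shows the name and ID of the spectated player, built on PostGameSpectatedPlayerName.
# ShowWidget fills an uppercase, sheared "Spectating: {0}" label and a small user ID line, then reveals the initially collapsed root panel.
# A separate Intro event plays IntroAnim, an opacity fade-in of the container, with no caller in this graph.
# The panel's Alpha animation targets VerticalBox_0, absent here, and its play node is unconnected.
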WidgetBlueprint /Game/Athena/HUD/EndGame/PostGameSpectatingPlayerName
animations [IntroAnim {
  binding_count=1
  length_seconds=0.95
  object=&self:IntroAnim
  sequence {bindings [VerticalBox_Container {
     id=#1
     possessed_class=VerticalBox
     tracks [{
       type=MovieSceneFloatTrack
       display_name=Render Opacity
       sections [{type=MovieSceneFloatSection, channels [None {type=MovieSceneScriptingFloatChannel, keys [[0,0],[19,1]]}]}]
      }]
    }]}
 }]
graphs [EventGraph {nodes [
   Intro() {is=title K2Node_CustomEvent}
   > PlayAnimation(InAnimation=Get IntroAnim() -> (IntroAnim:Widget Animation Object Reference))
   n1 = Event ShowWidget!() -> (OutputDelegate, then, PlayerName:String, PlayerID:String) {is=title K2Node_Event}
   > "SetText (Text)"(
    self=Get Text_PlayerName() -> (Text_PlayerName:Common Text Object Reference)
    InText=Format Text!(Format="LOCTEXT(\"Spectating: {0}\")", 0:Text="To Text (String)"(InString=@n1.PlayerName)) -> (Result)
   )
   > "SetText (Text)"(self=Get Text_UserID() -> (Text_UserID:Text Object Reference), InText="To Text (String)"(InString=@n1.PlayerID))
   > SetVisibility(self=Get SimpleAnimation() -> (SimpleAnimation:Fort Simple Widget Animations Panel Object Reference), InVisibility=HitTestInvisible)
   PlayAnimationByName!(
    execute
    self:Fort Simple Widget Animations Panel Object Reference=Get SimpleAnimation()
    Name:"Name (by ref)"=MakeLiteralName(Value=Alpha)
    Speed:float=1
    StartAtTime:float=0
   ) -> (then, ReturnValue:bool)
   Event PreConstruct()
   Event Tick()
  ]}]
variables []
widgets {
 info {
  inheritedWidgetCount=1
  parentClass=/Script/FortniteUI.PostGameSpectatedPlayerName
  rootWidgetClass=/Script/FortniteUI.FortSimpleWidgetAnimationsPanel
  widgetCount=6
 }
 widgets [
  SimpleAnimation {
   widgetClassPath=&FortSimpleWidgetAnimationsPanel
   bIsVariable=true
   widget=SimpleAnimation
   properties {
    visibility=Collapsed
    animations [Alpha {
      bIsEnabled=true
      onAnimationsFinished=()
      targets [{
        alpha=/Game/Athena/HUD/Discoverability/IntroAnim_RenderOpacity
        bResetOnFinish=false
        bShouldAnimateAlpha=true
        bShouldAnimateScale=false
        bShouldAnimateTranslation=false
        scale=null
        targetName=VerticalBox_0
        translation=null
        materialScalarAnimations []
       }]
     }]
   }
  }
  VerticalBox_Container {widgetClassPath=&VerticalBox, parent=SimpleAnimation, widget=VerticalBox_Container, slot {class=FortSimpleWidgetAnimationsSlot}}
  Text_PlayerName {
   widgetClassPath=&CommonTextBlock
   bIsVariable=true
   parent=VerticalBox_Container
   widget=Text_PlayerName
   properties {
    style=/FortUILibrary/Styles/Typography/Foundation/TextStyle-Base-XS-BBS.TextStyle-Base-XS-BBS_C
    text=Spectating: Playername
    textTransformPolicy=ToUpper
    font {
     bForceMonospaced=false
     bMaterialIsStencil=false
     fontMaterial=null
     fontObject=&BurbankSmall
     letterSpacing=0
     monospacedWidth=1
     size=18
     skewAmount=0
     typefaceFontName=Bold
     outlineSettings {
      bApplyOutlineToDropShadows=false
      bMiteredCorners=false
      bSeparateFillAlpha=false
      outlineMaterial=null
      outlineSize=0
      outlineColor=<rgba 0 0 0 1>
     }
    }
    renderTransform {angle=0, scale=<1 1>, shear=<-8 0>, translation=<0 0>}
    shadowOffset=<0 0>
   }
   slot {class=VerticalBoxSlot, properties {horizontalAlignment=HAlign_Right}}
  }
  CommonBorder_0 {
   widgetClassPath=&CommonBorder
   parent=VerticalBox_Container
   widget=CommonBorder_0
   properties {
    horizontalAlignment=HAlign_Center
    renderOpacity=0.5
    verticalAlignment=VAlign_Center
    background {
     bIsDynamicallyLoaded=false
     drawAs=NoDrawType
     imageType=NoImage
     mirroring=NoMirror
     resourceName=None
     resourceObject=null
     tiling=NoTile
     imageSize=<32 32>
     margin=<ltrb 0 0 0 0>
     outlineSettings {
      bUseBrushTransparency=false
      roundingType=HalfHeightRadius
      width=0
      color {colorUseRule=UseColor_Specified, specifiedColor=<rgba 0 0 0 0>}
      cornerRadii=<0 0 0 0>
     }
     tintColor {colorUseRule=UseColor_Specified, specifiedColor=<rgba 0.0052 0.0052 0.0056 0.25>}
     uVRegion {bIsValid=false, max=<0 0>, min=<0 0>}
    }
    renderTransform {angle=0, scale=<1 1>, shear=<0 0>, translation=<0 -10>}
   }
   slot {class=VerticalBoxSlot, properties {horizontalAlignment=HAlign_Right, verticalAlignment=VAlign_Center, padding=<ltrb 0 0 -4 0>}}
  }
  Text_UserID {
   widgetClassPath=&TextBlock
   bIsVariable=true
   parent=CommonBorder_0
   widget=Text_UserID
   properties {
    text=USER ID
    colorAndOpacity {colorUseRule=UseColor_Specified, specifiedColor=<rgba 0.5365 0.5365 0.5365 1>}
    font {
     bForceMonospaced=false
     bMaterialIsStencil=false
     fontMaterial=null
     fontObject=&NotoSans
     letterSpacing=0
     monospacedWidth=1
     size=8
     skewAmount=0
     typefaceFontName=NotoSans-Regular
     outlineSettings {
      bApplyOutlineToDropShadows=false
      bMiteredCorners=false
      bSeparateFillAlpha=false
      outlineMaterial=null
      outlineSize=0
      outlineColor=<rgba 0 0 0 1>
     }
    }
   }
   slot {class=BorderSlot, properties {horizontalAlignment=HAlign_Center, verticalAlignment=VAlign_Center}}
  }
  TutorialHighlight {widgetClassPath=&TutorialHighlight}
 ]
}
