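# Root HUD widget for the post-game screens, layering cinematic bars, spectated player name, screen container, placement, mobile buttons and new quests in a grid.
# Screen config updates toggle safe-zone padding and name visibility, while events play intros, toggle blur and re-layer placement.
# Touch players off Switch get an async-loaded button overlay that follows the current screen.
# A Dynamic UI scene disallowing text chat lasts from Construct to Destruct.
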WidgetBlueprint /FortniteCoreUI/HUD/EndGame/FortPostGameScreenRoot
graphs [
 Cosmetics {nodes [
   n1 = Event OnPostGameScreenConfigUpdated!() -> (OutputDelegate, then, NewPostGameConfig:"Fort Post Game Screen Config Structure (by ref)") {is=title K2Node_Event}
   n2 = Break Fort Post Game Screen Config!(FortPostGameScreenConfig:"Fort Post Game Screen Config Structure (by ref)"=@n1.NewPostGameConfig) -> (
    ClickCatcherState:EPostGameClickCatcherMode Enum
    PostGameHUDMode:EPostGameHUDMode Enum
    bSohwCinematicBars:bool=~
    bShowPlacementOverlay:bool=~
    bIsContainerSafeZoneEnabled:bool=~
    bShowSpectatingPlayerName:bool=~
    bPutPawnInStasis:bool=~
    bAllowEmotesInStasis:bool=~
   ) {is=title K2Node_BreakStruct}
   > SetSidesToPad!(
    execute
    self:Safe Zone Object Reference=Get SafeZone_PostMatchContent() -> (SafeZone_PostMatchContent:Safe Zone Object Reference)
    InPadLeft:bool=@n2.bIsContainerSafeZoneEnabled
    InPadRight:bool=@n2.bIsContainerSafeZoneEnabled
    InPadTop:bool=@n2.bIsContainerSafeZoneEnabled
    InPadBottom:bool=@n2.bIsContainerSafeZoneEnabled
   ) -> (then)
   > SetVisibility(
    self=Get PostGameSpectatingPlayerName() -> (PostGameSpectatingPlayerName:Post Game Spectating Player Name Object Reference)
    InVisibility=Select(Option 0=Collapsed, Option 1=SelfHitTestInvisible, Index=@n2.bShowSpectatingPlayerName)
   )
   Event AnimateCinematicBars() {is=title K2Node_Event}
   > Do Once()
   > EventIntro(self:HUD Cinematic Bars Object Reference=Cast To HUD_CinematicBars!(Object=Get HUD_CinematicBars() -> (HUD_CinematicBars:User Widget Object Reference)) -> (AsHUD Cinematic Bars:HUD Cinematic Bars Object Reference, bSuccess:bool=~) {is=title K2Node_DynamicCast})
   > Intro(self:Post Game Spectating Player Name Object Reference=Get PostGameSpectatingPlayerName())
   Event OnInitialized()
   n3 = GetOwningPlayer()
   n4 = OR Boolean(A=IsNintendoSwitch!() -> (ReturnValue:bool), B=IsNintendoSwitch2!() -> (ReturnValue:bool))
   > Branch(Condition=AND Boolean(
     A=NOT Boolean(A=@n4)
     B=IsUsingInputType!(InputType:ECommonInputType Enum=Touch, PlayerController:Player Controller Object Reference=@n3) -> (ReturnValue:bool)
    ))
   > n5 = Async Load Class Asset!(execute:Exec, AssetClass:Object Soft Class Reference=Get MobileOverlay() -> (MobileOverlay:User Widget Soft Class Reference)) -> (then:Exec, Completed:Exec, Class:Object Class Reference) {is=title K2Node_LoadAssetClass}
   n6 = Cast To UserWidget Class!(execute:Exec=@n5.Completed, Class:Object Class Reference=@n5) -> (then:Exec, CastFailed:Exec, AsUser Widget:User Widget Class Reference) {is=title K2Node_ClassDynamicCast}
   > n7 = Create Widget!(execute:Exec, Class:User Widget Class Reference=@n6, OwningPlayer:Player Controller Object Reference=GetOwningPlayer()) -> (then:Exec, ReturnValue:User Widget Object Reference)
   > Set MobileOverlayWidget(MobileOverlayWidget:User Widget Object Reference=@n7) -> (Output_Get:User Widget Object Reference)
   > n8 = Sequence()
   > n9 = AddChild(
    self=Get ScaleBox_MobileOverlay() -> (ScaleBox_MobileOverlay:Scale Box Object Reference)
    Content=Get MobileOverlayWidget() -> (MobileOverlayWidget:User Widget Object Reference)
   )
   > n10 = Cast To ScaleBoxSlot!(execute, Object=@n9) -> (then, CastFailed, AsScale Box Slot:Scale Box Slot Object Reference) {is=title K2Node_DynamicCast}
   > SetVerticalAlignment(self:Scale Box Slot Object Reference=@n10, InVerticalAlignment=VAlign_Fill)
   > SetHorizontalAlignment(self:Scale Box Slot Object Reference=@n10, InHorizontalAlignment=HAlign_Fill)
   n11 = Cast To FortPostGameMobileButtonOverlay!(execute=@n8.then_1, Object=Get MobileOverlayWidget()) -> (then, CastFailed, AsFort Post Game Mobile Button Overlay:Fort Post Game Mobile Button Overlay Object Reference) {is=title K2Node_DynamicCast}
   > UpdateButtonsAccordingToPostGameScreen!(
    execute
    self:Fort Post Game Mobile Button Overlay Object Reference=@n11
    PostGameScreenWidget:Fort Post Game Screen Object Reference=Get CurrentPostGameScreen() -> (CurrentPostGameScreen:Fort Post Game Screen Object Reference)
   ) -> (then)
   Event Construct()
   > AddScene!(
    execute
    self:Dynamic UIManager Object Reference=Get DynamicUIManager() -> (ReturnValue:Dynamic UIManager Object Reference) {is=title K2Node_GetSubsystem}
    Scene:Dynamic UIScene Object Reference=Get RemoveChatScene() -> (RemoveChatScene:Dynamic UIScene Object Reference)
    Player:"Player Controller Object Reference (by ref)"=GetOwningPlayer()
   ) -> (then)
   > n12 = Branch(Condition=IsMobileGame())
   > SetPadding!(
    execute
    self:User Widget Object Reference=Get PostGameSpectatingPlayerName()
    InPadding_Left:float=0
    InPadding_Top:float=32
    InPadding_Right:float=72
    InPadding_Bottom:float=0
   ) -> (then)
   SetPadding(execute=@n12.else, self=Get PostGameSpectatingPlayerName(), InPadding_Left=0, InPadding_Top=0, InPadding_Right=0, InPadding_Bottom=0)
   Event ToggleBlur() {is=title K2Node_Event}
   > EventToggleBlur(self:HUD Cinematic Bars Object Reference=Cast To HUD_CinematicBars(Object=Get HUD_CinematicBars()) -> (bSuccess=~))
   n13 = Event SetPlacementZOrder!() -> (OutputDelegate, then, bIsPlacementAboveContent:bool) {is=title K2Node_Event}
   > SetLayer!(
    execute
    self:Grid Slot Object Reference=SlotAsGridSlot(Widget=Get SafeZone_Placement() -> (SafeZone_Placement:Safe Zone Object Reference))
    InLayer:int=Select(Option 0:int=3, Option 1:int=7, Index=@n13.bIsPlacementAboveContent) -> (ReturnValue:int)
   ) -> (then)
   Event Destruct()
   > RemoveScene!(
    execute
    self:Dynamic UIManager Object Reference=Get DynamicUIManager()
    Scene:Dynamic UIScene Object Reference=Get RemoveChatScene()
    Player:"Player Controller Object Reference (by ref)"=GetOwningPlayer()
   ) -> (then)
   n14 = Event OnPostGameScreenIndexChanged!() -> (OutputDelegate, then, NewIndex:int=0, NewScreen:Fort Post Game Screen Object Reference) {is=title K2Node_Event}
   > Set CurrentPostGameScreen(CurrentPostGameScreen:Fort Post Game Screen Object Reference=@n14.NewScreen) -> (Output_Get:Fort Post Game Screen Object Reference)
   > n15 = Cast To FortPostGameMobileButtonOverlay(Object=Get MobileOverlayWidget())
   > UpdateButtonsAccordingToPostGameScreen(self=@n15, PostGameScreenWidget=Get CurrentPostGameScreen())
  ]}
 EventGraph {nodes [
   Event PreConstruct()
   > SetMobileStyle() {is=title K2Node_CallFunction}
  ]}
 OnToggleBlurRequested {nodes [OnToggleBlurRequested() {is=title K2Node_FunctionEntry}]}
 SetMobileStyle {nodes [
   SetMobileStyle() {is=title K2Node_FunctionEntry}
   n1 = GetOwningPlayer()
   > Branch(Condition=AND Boolean(A=IsUsingCompactUI(), B=IsUsingInputType(InputType=Touch, PlayerController=@n1)))
   n2 = Get SafeZone_SpectatingName() -> (SafeZone_SpectatingName:Safe Zone Object Reference)
   n3 = SlotAsGridSlot(Widget=@n2)
   > SetHorizontalAlignment(self:Grid Slot Object Reference=@n3, InHorizontalAlignment=HAlign_Left)
   > SetPadding(self:Grid Slot Object Reference=@n3, InPadding_Left=15, InPadding_Top=-25, InPadding_Right=0, InPadding_Bottom=0)
   > SetSidesToPad(self=@n2, InPadLeft=true, InPadRight=false, InPadTop=true, InPadBottom=false)
  ]}
]
variables [
 CurrentPostGameScreen {type=FortPostGameScreen, default=null}
 MobileOverlay {type=Class, default=/Game/Athena/HUD/EndGame/FortPostGameMobileOverlay.FortPostGameMobileOverlay_C}
 MobileOverlayWidget {type=UserWidget, default=null}
 OnToggleBlurRequested {}
 RemoveChatScene {type=DynamicUIScene, default=&/DUI_TextChatUnallowScene}
]
widgets {
 info {inheritedWidgetCount=7, parentClass=/Script/FortniteUI.FortPostGameScreenRoot, rootWidgetClass=&GridPanel, widgetCount=19}
 widgets [
  GridPanel_38 {widgetClassPath=&GridPanel, widget=GridPanel_38, properties {columnFill [1], rowFill [1]}}
  Spacer_ClickCatcher {
   widgetClassPath=&Spacer
   bInherited=true
   parent=GridPanel_38
   widget=Spacer_ClickCatcher
   properties {visibility=Collapsed}
   slot {class=GridSlot}
  }
  ScaleBox_CinematicBars_UserSpecified {
   widgetClassPath=&ScaleBox
   bInherited=true
   bIsVariable=true
   parent=GridPanel_38
   widget=ScaleBox_CinematicBars_UserSpecified
   properties {stretch=UserSpecified}
   slot {class=GridSlot, properties {layer=1}}
  }
  HUD_CinematicBars {
   widgetClassPath=/Game/Athena/HUD/EndGame/HUD_CinematicBars
   bInherited=true
   bIsVariable=true
   parent=ScaleBox_CinematicBars_UserSpecified
   widget=HUD_CinematicBars
   slot {class=ScaleBoxSlot, properties {horizontalAlignment=HAlign_Fill, verticalAlignment=VAlign_Fill}}
  }
  SafeZone_SpectatingName {
   widgetClassPath=&SafeZone
   bIsVariable=true
   parent=GridPanel_38
   widget=SafeZone_SpectatingName
   properties {padBottom=false, padLeft=false}
   slot {class=GridSlot, properties {horizontalAlignment=HAlign_Right, layer=2, verticalAlignment=VAlign_Top}}
  }
  PostGameSpectatingPlayerName {
   widgetClassPath=/Game/Athena/HUD/EndGame/PostGameSpectatingPlayerName
   bIsVariable=true
   parent=SafeZone_SpectatingName
   widget=PostGameSpectatingPlayerName
   properties {renderTransform {angle=0, scale=<1 1>, shear=<0 0>, translation=<0 9>}}
   slot {class=SafeZoneSlot}
  }
  SafeZone_PostMatchContent {
   widgetClassPath=&SafeZone
   bIsVariable=true
   parent=GridPanel_38
   widget=SafeZone_PostMatchContent
   slot {class=GridSlot, properties {layer=4}}
  }
  ScaleBox_0 {
   widgetClassPath=&ScaleBox
   parent=SafeZone_PostMatchContent
   widget=ScaleBox_0
   properties {stretch=ScaleBySafeZone}
   slot {class=SafeZoneSlot}
  }
  PostGameScreenContainer_Main {
   widgetClassPath=/Script/FortniteUI.FortPostGameScreenContainer
   bInherited=true
   bIsVariable=true
   parent=ScaleBox_0
   widget=PostGameScreenContainer_Main
   properties {visibility=SelfHitTestInvisible}
   slot {class=ScaleBoxSlot, properties {horizontalAlignment=HAlign_Fill, verticalAlignment=VAlign_Fill}}
  }
  SafeZone_Placement {
   widgetClassPath=&SafeZone
   bIsVariable=true
   parent=GridPanel_38
   widget=SafeZone_Placement
   properties {padBottom=false, padLeft=false, padRight=false}
   slot {class=GridSlot, properties {layer=5}}
  }
  ScaleBox_1 {widgetClassPath=&ScaleBox, parent=SafeZone_Placement, widget=ScaleBox_1, properties {stretch=None}, slot {class=SafeZoneSlot}}
  ScaleBox_PlacementOverlay_UserSpecified {
   widgetClassPath=&ScaleBox
   bInherited=true
   parent=ScaleBox_1
   widget=ScaleBox_PlacementOverlay_UserSpecified
   properties {stretch=UserSpecified}
   slot {class=ScaleBoxSlot, properties {horizontalAlignment=HAlign_Fill, verticalAlignment=VAlign_Fill}}
  }
  Overlay_Placement {
   widgetClassPath=&Overlay
   bInherited=true
   bIsVariable=true
   parent=ScaleBox_PlacementOverlay_UserSpecified
   widget=Overlay_Placement
   slot {class=ScaleBoxSlot, properties {horizontalAlignment=HAlign_Fill, verticalAlignment=VAlign_Fill}}
  }
  PlacementOverlay {
   widgetClassPath=/Game/Athena/HUD/EndGame/PlacementOverlay
   parent=Overlay_Placement
   widget=PlacementOverlay
   slot {class=OverlaySlot, properties {horizontalAlignment=HAlign_Fill, verticalAlignment=VAlign_Fill}}
  }
  SafeZone_MobileOverlay {widgetClassPath=&SafeZone, parent=GridPanel_38, widget=SafeZone_MobileOverlay, slot {class=GridSlot, properties {layer=6}}}
  ScaleBox_MobileOverlay {
   widgetClassPath=&ScaleBox
   bIsVariable=true
   parent=SafeZone_MobileOverlay
   widget=ScaleBox_MobileOverlay
   properties {stretch=ScaleBySafeZone}
   slot {class=SafeZoneSlot}
  }
  SafeZone_NewQuests {
   widgetClassPath=&SafeZone
   parent=GridPanel_38
   widget=SafeZone_NewQuests
   properties {padBottom=false, padRight=false}
   slot {class=GridSlot, properties {horizontalAlignment=HAlign_Left, layer=8, verticalAlignment=VAlign_Top, padding=<ltrb 32 80 0 0>}}
  }
  ScaleBox_90 {
   widgetClassPath=&ScaleBox
   parent=SafeZone_NewQuests
   widget=ScaleBox_90
   properties {stretch=UserSpecified}
   slot {class=SafeZoneSlot}
  }
  HUD_NewQuestsWidget {
   widgetClassPath=&/FortPostGameNewQuestsWidget
   bInherited=true
   bIsVariable=true
   parent=ScaleBox_90
   widget=HUD_NewQuestsWidget
   properties {clipping=ClipToBounds, visibility=Collapsed}
   slot {class=ScaleBoxSlot, properties {horizontalAlignment=HAlign_Fill, verticalAlignment=VAlign_Top}}
  }
 ]
}
