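# Widget Blueprint for a single quest entry on the end-of-match screen, derived from the native EndOfMatchQuestEntryWidget class.
# Its root is a QuestNotification widget, the in-game quest event notification, flagged as end-of-match with visible flavor text and overridden width and max text height.
# The event graph is empty, so the Blueprint only configures that widget and leaves behavior to the parent class and QuestNotification.
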WidgetBlueprint /FortniteCoreUI/HUD/EndGame/EndOfMatch_QuestEntry
graphs [EventGraph {links [], nodes []}]
variables []
widgets {
 info {
  inheritedWidgetCount=1
  parentClass=/Script/FortniteUI.EndOfMatchQuestEntryWidget
  rootWidgetClass=/FortniteCoreUI/UI/InGame/EventNotifications/QuestNotification.QuestNotification_C
  widgetCount=1
 }
 widgets [QuestNotification {
   widgetClassPath=/FortniteCoreUI/UI/InGame/EventNotifications/QuestNotification
   bInherited=true
   bIsVariable=true
   widget=QuestNotification
   properties {bIsEndOfMatch=true, flavor Text Visibility=HitTestInvisible, in Width Override=600, textMaxHeightOverride=100}
  }]
}
