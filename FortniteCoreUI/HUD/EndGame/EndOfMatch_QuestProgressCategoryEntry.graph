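# End-of-match HUD widget that shows one category of quest updates under either a "QUEST COMPLETED:" or "QUEST PROGRESS:" title.
# A legacy switcher with zero transition time holds both sheared, outlined titles above a fill-sized SizeBox containing a dynamic entry box.
# That box uses overlay layout and creates EndOfMatch_QuestEntry widgets.
# The switcher and entry containers are inherited from the native EndOfMatchQuestCategoryWidget; the Blueprint's event stubs are empty.
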WidgetBlueprint /FortniteCoreUI/HUD/EndGame/EndOfMatch_QuestProgressCategoryEntry
graphs [EventGraph {nodes [
   Event PreConstruct()
   Event Construct()
   Event Tick()
  ]}]
variables []
widgets {
 info {inheritedWidgetCount=3, parentClass=/Script/FortniteUI.EndOfMatchQuestCategoryWidget, rootWidgetClass=&VerticalBox, widgetCount=6}
 widgets [
  VerticalBox_57 {widgetClassPath=&VerticalBox, widget=VerticalBox_57}
  Switcher_CategoryType {
   widgetClassPath=&CommonWidgetSwitcherLegacy
   bInherited=true
   bIsVariable=true
   parent=VerticalBox_57
   widget=Switcher_CategoryType
   properties {transitionDuration=0}
   slot {class=VerticalBoxSlot}
  }
  QuestsCompletedText {
   widgetClassPath=&CommonTextBlock
   parent=Switcher_CategoryType
   widget=QuestsCompletedText
   properties {
    style=/Game/Athena/HUD/TextStyles/TextStyle-HUD-MapScreen_Title.TextStyle-HUD-MapScreen_Title_C
    text=QUEST COMPLETED:
    font {
     bForceMonospaced=false
     bMaterialIsStencil=false
     fontMaterial=null
     fontObject=&Burbank
     letterSpacing=0
     monospacedWidth=1
     size=18
     skewAmount=0
     typefaceFontName=BurbankRegular-Black
     outlineSettings {
      bApplyOutlineToDropShadows=true
      bMiteredCorners=false
      bSeparateFillAlpha=true
      outlineMaterial=&M-FontShadow
      outlineSize=3
      outlineColor=<rgba 0 0 0 0.5>
     }
    }
    renderTransform {angle=0, scale=<1 1>, shear=<-12 0>, translation=<0 0>}
    shadowOffset=<0 0>
   }
   slot {class=WidgetSwitcherSlot}
  }
  QuestsInProgressText {
   widgetClassPath=&CommonTextBlock
   parent=Switcher_CategoryType
   widget=QuestsInProgressText
   properties {
    style=&TextStyle-HUD-MapScreen_Title_C
    text=QUEST PROGRESS:
    font {
     bForceMonospaced=false
     bMaterialIsStencil=false
     fontMaterial=null
     fontObject=&Burbank
     letterSpacing=0
     monospacedWidth=1
     size=18
     skewAmount=0
     typefaceFontName=BurbankRegular-Black
     outlineSettings {
      bApplyOutlineToDropShadows=true
      bMiteredCorners=false
      bSeparateFillAlpha=true
      outlineMaterial=&M-FontShadow
      outlineSize=3
      outlineColor=<rgba 0 0 0 0.5>
     }
    }
    renderTransform {angle=0, scale=<1 1>, shear=<-12 0>, translation=<0 0>}
    shadowOffset=<0 0>
   }
   slot {class=WidgetSwitcherSlot}
  }
  EntryBox_SizeBox {
   widgetClassPath=&SizeBox
   bInherited=true
   bIsVariable=true
   parent=VerticalBox_57
   widget=EntryBox_SizeBox
   slot {class=VerticalBoxSlot, properties {size {sizeRule=Fill, value=1}}}
  }
  EntryBox_QuestUpdates {
   widgetClassPath=&DynamicEntryBox
   bInherited=true
   bIsVariable=true
   parent=EntryBox_SizeBox
   widget=EntryBox_QuestUpdates
   properties {entryBoxType=Overlay, entryWidgetClass=&/EndOfMatch_QuestEntry.EndOfMatch_QuestEntry_C}
   slot {class=SizeBoxSlot}
  }
 ]
}
